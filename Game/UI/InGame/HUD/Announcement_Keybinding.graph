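# HUD announcement widget that shows a keybinding hint with icon, title, rich-text body and start sound.
# UpdateWidgetData fills it from KeybindingData, then removes it after a positive DisplayTime or on AnnouncementStopped.
# SetBodyText uses DetailText, or on gamepad the first text whose action has a key, falling back to GamepadDetailText.
# An unbound action swaps in a rebind prompt; input-method and keybind changes refresh the body.
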WidgetBlueprint /Game/UI/InGame/HUD/Announcement_Keybinding
graphs [
 EventGraph {nodes [
   OInInputMethodChanged!() -> (OutputDelegate, then, bNewInputType:ECommonInputType Enum) {is=title K2Node_CustomEvent}
   > SetBodyText()
   OnKeybindsChanged() {is=title K2Node_CustomEvent}
   > SetBodyText()
   Event Construct()
   n1 = Create Event()
   n2 = Get CommonInputSubsystem()
   > Bind Event to On Input Method Changed(self=@n2, Delegate=@n1)
   n3 = Create Event()
   n4 = Get FortGlobalUIContext() -> (ReturnValue:Fort Global UIContext Object Reference) {is=title K2Node_GetSubsystem}
   > Bind Event to On Keybinds Changed(self:Fort Global UIContext Object Reference=@n4, Delegate=@n3) {is=title K2Node_AddDelegate}
   Event Destruct()
   > Unbind Event from On Input Method Changed(self=@n2, Delegate=@n1)
   > Unbind Event from On Keybinds Changed(self:Fort Global UIContext Object Reference=@n4, Delegate=@n3) {is=title K2Node_RemoveDelegate}
   n5 = Event UpdateWidgetData!() -> (OutputDelegate, then, Announcement:Fort Client Announcement Object Reference) {is=title K2Node_Event}
   > n6 = Cast To FortClientAnnouncement_Keybinding!(execute, Object=@n5.Announcement) -> (then, CastFailed, AsFort Client Announcement Keybinding:Fort Client Announcement Keybinding Object Reference) {is=title K2Node_DynamicCast}
   > Set Announcement(Announcement:Fort Client Announcement Keybinding Object Reference=@n6) -> (Output_Get:Fort Client Announcement Keybinding Object Reference)
   n7 = Break Fort Client Announcement Data Keybinding!(FortClientAnnouncementData_Keybinding:"Fort Client Announcement Data Keybinding Structure (by ref)"=Get KeybindingData!(self:Fort Client Announcement Keybinding Object Reference=@n6) -> (KeybindingData:Fort Client Announcement Data Keybinding Structure)) -> (
    GamapadActionTextList:Array of Action Text Pair Structures
    Icon:Slate Brush Structure
    TitleText:Text
    DetailText:Text
    GamepadDetailText:Text
    Priority:int
    DisplayTime:float=~
    OnStartSound:Sound Base Object Reference
   ) {is=title K2Node_BreakStruct}
   > n8 = Is Valid(InputObject=GetBrushResource!(Brush:"Slate Brush Structure (by ref)"=@n7.Icon) -> (ReturnValue:Object Reference))
   "SetText (Text)"(execute=[@n12.then, @n8.Is Not Valid], self=Get TextTitleText() -> (TextTitleText:Common Text Object Reference), InText=@n7.TitleText)
   > SetBodyText()
   > PlaySound2D(Sound=@n7.OnStartSound, None=Self-Reference())
   > SetVisibility(self=Get TextTitleText(), InVisibility=Select(Option 0=SelfHitTestInvisible, Option 1=Collapsed, Index=TextIsEmpty(InText=@n7.TitleText)))
   n9 = TextIsEmpty(InText=@n7.TitleText)
   n10 = TextIsEmpty(InText=@n7.DetailText)
   > SetVisibility(InVisibility=Select(Option 0=SelfHitTestInvisible, Option 1=Collapsed, Index=AND Boolean(A=@n9, B=@n10)))
   > Branch(Condition="float > float"(A=@n7.DisplayTime, B=0) {is=title K2Node_CallFunction})
   > n11 = Delay(Duration=@n7.DisplayTime)
   RemoveFromParent(execute=[@n11.then, @n13.then])
   n12 = SetBrush(execute=@n8.Is Valid, self=Get Icon() -> (Icon:Image Object Reference), InBrush=@n7.Icon)
   n13 = Event AnnouncementStopped!() -> (OutputDelegate, then, Announcement:Fort Client Announcement Object Reference) {is=title K2Node_Event}
  ]}
 FindActiveText {nodes [
   FindActiveText!() -> (then, GamepadActionTextList:Array of Action Text Pair Structures) {is=title K2Node_FunctionEntry}
   > Set ReturnText(ReturnText:Text) -> (Output_Get:Text)
   > n1 = For Each Loop with Break!(
    Exec:Exec
    Array:Array of Action Text Pair Structures=Get GamepadActionTextList() -> (GamepadActionTextList:Array of Action Text Pair Structures)
    Break:Exec=@n4.then
   ) -> (LoopBody:Exec, Array Element:Action Text Pair Structure, Array Index:int, Completed:Exec)
   Return Node!(execute=@n1.Completed, Text:Text=Get ReturnText() -> (ReturnText:Text))
   n2 = Break Action Text Pair!(ActionTextPair:"Action Text Pair Structure (by ref)"=@n1) -> (Action:Name, Text:Text) {is=title K2Node_BreakStruct}
   n3 = Get FortGlobalUIContext!() -> (ReturnValue:Fort Global UIContext Object Reference) {is=title K2Node_GetBlueprintContextEx}
   > Branch(Condition=Is Valid Key!(Key:"Key Structure (by ref)"=GetKeyForAction!(self:Fort Global UIContext Object Reference=@n3, Action:Name=@n2, bForceGamepadKey:bool=true, PresetNameOverride:Name=None) -> (ReturnValue:Key Structure)) -> (ReturnValue:bool))
   > n4 = Set ReturnText(ReturnText=@n2.Text)
  ]}
 SetBodyText {nodes [
   SetBodyText() {is=title K2Node_FunctionEntry}
   > n1 = Is Valid(InputObject=Get Announcement() -> (Announcement:Fort Client Announcement Keybinding Object Reference))
   n2 = Set BodyText(execute=@n1.Is Not Valid, BodyText:Text) -> (Output_Get:Text)
   n3 = FindFirstUnboundAction!(
    execute=[@n2.then, @n11.then, @n12.then, @n13.then]
    self:Fort Announcement Widget Object Reference
    AnnouncementBody:Text=Get BodyText() -> (BodyText:Text)
   ) -> (then, ReturnValue:Name=None)
   > n4 = Branch(Condition="Equal (Name)"!(A:Name=@n3, B:Name) -> (ReturnValue:bool) {is=title K2Node_CallFunction})
   SetText(
    execute=[@n4.then, @n5.then]
    self:Fort Rich Text Block Legacy Object Reference=Get TextDetailText() -> (TextDetailText:Fort Rich Text Block Legacy Object Reference)
    InText:Text=Get BodyText()
   )
   n5 = Set BodyText(
    execute=@n4.else
    BodyText=Format Text!(
     Format="LOCTEXT(\"Please rebind the {s} action.\")"
     s:Text=GetTextForAction!(self:Fort Announcement Widget Object Reference, ActionName:"Name (by ref)"=@n3) -> (ReturnValue:Text)
    ) -> (Result)
   )
   n6 = Get CommonInputSubsystem()
   n7 = Branch(execute=@n1.Is Valid, Condition="Equal (Enum)"(A=GetCurrentInputType(self=@n6), B=Gamepad))
   n8 = Break Fort Client Announcement Data Keybinding(FortClientAnnouncementData_Keybinding=Get KeybindingData(self=Get Announcement())) -> (DisplayTime=~)
   > n9 = FindActiveText!(execute, self, GamepadActionTextList:Array of Action Text Pair Structures=@n8) -> (then, Text:Text) {is=title K2Node_CallFunction}
   > n10 = Branch(Condition=TextIsEmpty(InText=@n9))
   > n11 = Set BodyText(BodyText=@n8.GamepadDetailText)
   n12 = Set BodyText(execute=@n10.else, BodyText=@n9)
   n13 = Set BodyText(execute=@n7.else, BodyText=@n8.DetailText)
  ]}
]
variables [Announcement {type=FortClientAnnouncement_Keybinding, default=null}]
widgets {
 info {inheritedWidgetCount=0, parentClass=/Script/FortniteUI.FortAnnouncementWidget, rootWidgetClass=&CommonBorder, widgetCount=14}
 widgets [
  CommonBorder_0 {
   widgetClassPath=&CommonBorder
   widget=CommonBorder_0
   properties {style=/FortUILibrary/Styles/Border/Border-ModalHeader-Dark.Border-ModalHeader-Dark_C, padding=<ltrb 0 0 0 0>}
  }
  Overlay_0 {widgetClassPath=&Overlay, parent=CommonBorder_0, widget=Overlay_0, slot {class=BorderSlot, properties {padding=<ltrb 0 0 0 0>}}}
  VerticalBox_0 {
   widgetClassPath=&VerticalBox
   parent=Overlay_0
   widget=VerticalBox_0
   slot {class=OverlaySlot, properties {horizontalAlignment=HAlign_Fill, verticalAlignment=VAlign_Center, padding=<ltrb 10 10 10 10>}}
  }
  TextTitleText {
   widgetClassPath=&CommonTextBlock
   bIsVariable=true
   parent=VerticalBox_0
   widget=TextTitleText
   properties {
    style=/FortUILibrary/Styles/Typography/Foundation/TextStyle-Base-S-B-S.TextStyle-Base-S-B-S_C
    text=TitleText
    font {
     bForceMonospaced=false
     bMaterialIsStencil=false
     fontMaterial=null
     fontObject=&NotoSans
     letterSpacing=0
     monospacedWidth=1
     size=18
     skewAmount=0
     typefaceFontName=NotoSans-Bold
     outlineSettings {
      bApplyOutlineToDropShadows=true
      bMiteredCorners=false
      bSeparateFillAlpha=true
      outlineMaterial=&M-FontShadow
      outlineSize=3
      outlineColor=<rgba 0 0 0 0.5>
     }
    }
    shadowOffset=<0 0>
   }
   slot {class=VerticalBoxSlot, properties {horizontalAlignment=HAlign_Center, verticalAlignment=VAlign_Center}}
  }
  HorizontalBoxRoot {widgetClassPath=&HorizontalBox, bIsVariable=true, parent=VerticalBox_0, widget=HorizontalBoxRoot, slot {class=VerticalBoxSlot}}
  SizeBox_3 {
   widgetClassPath=&SizeBox
   parent=HorizontalBoxRoot
   widget=SizeBox_3
   properties {bOverride_HeightOverride=true, bOverride_WidthOverride=true, heightOverride=40, widthOverride=40}
   slot {class=HorizontalBoxSlot, properties {horizontalAlignment=HAlign_Center, verticalAlignment=VAlign_Center}}
  }
  Icon {
   widgetClassPath=&Image
   bIsVariable=true
   parent=SizeBox_3
   widget=Icon
   properties {brush {
     bIsDynamicallyLoaded=false
     drawAs=Image
     imageType=NoImage
     mirroring=NoMirror
     resourceName=None
     resourceObject=/Game/Widgets/SMyTownMapWidget/Common/InfoIcon
     tiling=NoTile
     imageSize=<32 32>
     margin=<ltrb 0 0 0 0>
     outlineSettings {
      bUseBrushTransparency=false
      roundingType=HalfHeightRadius
      width=0
      color {colorUseRule=UseColor_Specified, specifiedColor=<rgba 0 0 0 0>}
      cornerRadii=<0 0 0 0>
     }
     tintColor {colorUseRule=UseColor_Specified, specifiedColor=<rgba 1 1 1 1>}
     uVRegion {bIsValid=false, max=<0 0>, min=<0 0>}
    }}
   slot {class=SizeBoxSlot}
  }
  SizeBox_1 {
   widgetClassPath=&SizeBox
   parent=HorizontalBoxRoot
   widget=SizeBox_1
   properties {bOverride_WidthOverride=true, widthOverride=432}
   slot {class=HorizontalBoxSlot, properties {padding=<ltrb 8 8 8 8>}}
  }
  TextDetailText {
   widgetClassPath=/Script/FortniteUI.FortRichTextBlockLegacy
   bIsVariable=true
   parent=SizeBox_1
   widget=TextDetailText
   properties {
    autoWrapText=true
    styleSet=/FortUILibrary/Styles/Typography/Foundation/AnnouncementRichTextStyles
    text="Press <keybind>HotbarKeyF3</> to select stairs, then press <keybind>Fire</> to build and then go find yourself a <keybind>Use</> to go ahead and read this very long, but very informative, sentence."
    wrapTextAt=400
   }
   slot {class=SizeBoxSlot}
  }
  Overlay_1 {widgetClassPath=&Overlay, parent=VerticalBox_0, widget=Overlay_1, slot {class=VerticalBoxSlot}}
  SizeBox_0 {
   widgetClassPath=&SizeBox
   parent=Overlay_0
   widget=SizeBox_0
   properties {bOverride_HeightOverride=true, heightOverride=4}
   slot {class=OverlaySlot, properties {horizontalAlignment=HAlign_Fill}}
  }
  CommonBorder_1 {
   widgetClassPath=&CommonBorder
   parent=SizeBox_0
   widget=CommonBorder_1
   properties {style=/FortUILibrary/Styles/Border/Border_HUD_Line_Horz.Border_HUD_Line_Horz_C}
   slot {class=SizeBoxSlot}
  }
  SizeBox_2 {
   widgetClassPath=&SizeBox
   parent=Overlay_0
   widget=SizeBox_2
   properties {bOverride_HeightOverride=true, heightOverride=4}
   slot {class=OverlaySlot, properties {horizontalAlignment=HAlign_Fill, verticalAlignment=VAlign_Bottom}}
  }
  CommonBorder_2 {
   widgetClassPath=&CommonBorder
   parent=SizeBox_2
   widget=CommonBorder_2
   properties {style=&Border_HUD_Line_Horz_C}
   slot {class=SizeBoxSlot}
  }
 ]
}
