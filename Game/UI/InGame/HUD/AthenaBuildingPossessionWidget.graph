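# AthenaBuildingPossessionWidget is an in-game HUD Widget Blueprint based on the native FortHUDBuildingPossessionWidget, apparently an indicator for building possession.
# Its tree holds a root DotImage, drawn as a small white circle using a Social UI texture, alongside a TutorialHighlight widget.
# The event graph and variable list are empty, so the Blueprint supplies only visuals and leaves any behavior to the parent class.
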WidgetBlueprint /Game/UI/InGame/HUD/AthenaBuildingPossessionWidget
graphs [EventGraph {links [], nodes []}]
variables []
widgets {
 info {inheritedWidgetCount=2, parentClass=/Script/FortniteUI.FortHUDBuildingPossessionWidget, rootWidgetClass=&Image, widgetCount=2}
 widgets [
  DotImage {
   widgetClassPath=&Image
   bInherited=true
   bIsVariable=true
   widget=DotImage
   properties {brush {
     bIsDynamicallyLoaded=false
     drawAs=Image
     imageType=NoImage
     mirroring=NoMirror
     resourceName=None
     resourceObject=/Game/UI/Friends_UI/Social/circle
     tiling=NoTile
     imageSize=<8 8>
     margin=<ltrb 0 0 0 0>
     outlineSettings {
      bUseBrushTransparency=false
      roundingType=HalfHeightRadius
      width=0
      color {colorUseRule=UseColor_Specified, specifiedColor=<rgba 0 0 0 0>}
      cornerRadii=<0 0 0 0>
     }
     tintColor {colorUseRule=UseColor_Specified, specifiedColor=<rgba 1 1 1 1>}
     uVRegion {bIsValid=false, max=<0 0>, min=<0 0>}
    }}
  }
  TutorialHighlight {widgetClassPath=&TutorialHighlight}
 ]
}
