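# Turret HUD widget, based on FortAthenaVehicleDashboardWidget, that shows a centered warning while the vehicle is disabled.
# OnVehicleUpdated starts repeating async waits for the Vehicle.Disabled gameplay tag being added to or removed from the new vehicle.
# Both feed OnVehicleEnabled, which shows or removes one warning type on WBP_Vehicle_Warning and toggles its visibility.
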WidgetBlueprint /Game/UI/InGame/HUD/AthenaTurretReticle
graphs [
 EventGraph {nodes [
   n1 = Event OnVehicleUpdated!() -> (OutputDelegate, then, NewVehicle:Fort Athena Vehicle Object Reference) {is=title K2Node_Event}
   n2 = Get DisabledTag() -> (DisabledTag:Gameplay Tag Structure)
   > n3 = WaitGameplayTagAddToActor!(execute:Exec, TargetActor:Actor Object Reference=@n1.NewVehicle, Tag:Gameplay Tag Structure=@n2, OnlyTriggerOnce:bool=false) -> (then:Exec, AsyncTaskProxy:Ability Async Wait Gameplay Tag Added Object Reference, Added:Exec) {is=title K2Node_AsyncAction}
   > n4 = WaitGameplayTagRemoveFromActor!(execute:Exec, TargetActor:Actor Object Reference=@n1.NewVehicle, Tag:Gameplay Tag Structure=@n2, OnlyTriggerOnce:bool=false) -> (then:Exec, AsyncTaskProxy:Ability Async Wait Gameplay Tag Removed Object Reference, Removed:Exec) {is=title K2Node_AsyncAction}
   OnVehicleEnabled!(execute=@n4.Removed, self, Enabled:bool=true) -> (then)
   OnVehicleEnabled(execute=@n3.Added, Enabled=false)
  ]}
 OnVehicleEnabled {nodes [
   n1 = OnVehicleEnabled!() -> (then, Enabled:bool=~) {is=title K2Node_FunctionEntry}
   > n2 = Branch(Condition=@n1)
   n3 = Get WBP_Vehicle_Warning() -> (WBP_Vehicle_Warning:WBP Vehicle Warning Object Reference)
   > RemoveWarning!(execute, self:WBP Vehicle Warning Object Reference=@n3, Type:ENUM_WarningType Enum=NewEnumerator2) -> (then)
   > SetVisibility(self=@n3, InVisibility=Collapsed)
   ShowWarning!(execute=@n2.else, self:WBP Vehicle Warning Object Reference=@n3, Type:ENUM_WarningType Enum=NewEnumerator2) -> (then)
   > SetVisibility(self=@n3, InVisibility=HitTestInvisible)
  ]}
]
variables [DisabledTag {type=GameplayTag, default {tagName=Vehicle.Disabled}}]
widgets {
 info {
  inheritedWidgetCount=10
  parentClass=/Script/FortniteUI.FortAthenaVehicleDashboardWidget
  rootWidgetClass=&Overlay
  widgetCount=12
 }
 widgets [
  Overlay_25 {widgetClassPath=&Overlay, widget=Overlay_25}
  WBP_Vehicle_Warning {
   widgetClassPath=/Game/Athena/HUD/Vehicles/Dashboard/WBP_Vehicle_Warning
   bIsVariable=true
   parent=Overlay_25
   widget=WBP_Vehicle_Warning
   slot {class=OverlaySlot, properties {horizontalAlignment=HAlign_Center, verticalAlignment=VAlign_Center}}
  }
  Widget_FuelGauge {widgetClassPath=/Script/FortniteUI.FortVehicleFuelGauge, bInherited=true, widget=null}
  HorizontalBox_FuelCluster {widgetClassPath=&HorizontalBox, bInherited=true, widget=null}
  Image_Speedometer {widgetClassPath=&Image, bInherited=true, widget=null}
  Image_BoostGlow {widgetClassPath=&Image, bInherited=true, widget=null}
  Image_BoostMeter {widgetClassPath=&Image, bInherited=true, widget=null}
  Overlay_BoostSparks {widgetClassPath=&Overlay, bInherited=true, widget=null}
  NumericText_Speed {widgetClassPath=/Script/CommonUI.CommonNumericTextBlock, bInherited=true, widget=null}
  Overlay_Gauge {widgetClassPath=&Overlay, bInherited=true, widget=null}
  Text_ItemName {widgetClassPath=&CommonTextBlock, bInherited=true, widget=null}
  TutorialHighlight {widgetClassPath=&TutorialHighlight}
 ]
}
