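# Translucent UI material for a HUD bar that produces a two-color horizontal gradient.
# LinearGradient's UGradient output, driven by texture coordinates, is the blend factor between the ColorA and ColorB vector parameters, red and green by default.
# RGB and alpha are interpolated separately, so each end of the gradient can have its own color and transparency.
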Material /Game/UI/HUD/Materials/HUD-Bar_Material
graph {
 properties {
  bUsedWithStaticMesh=false
  blendMode=BLEND_Translucent
  materialDomain=MD_UI
  pixelDepthOffsetMode=PDOM_Legacy
  usedShadingModels=Default Lit
 }
 nodes [
  n1 = VectorParameter() {parameterName=ColorA, defaultValue=<rgba 1 0 0 1>}
  n2 = VectorParameter() {parameterName=ColorB, defaultValue=<rgba 0 1 0 1>}
  n3 = MaterialFunctionCall(UV Channel=TextureCoordinate()) {materialFunction=&LinearGradient}
  LinearInterpolate(A=@n1.RGB, B=@n2.RGB, Alpha=@n3.UGradient)
  LinearInterpolate(A=@n1.A, B=@n2.A, Alpha=@n3.UGradient)
 ]
}
