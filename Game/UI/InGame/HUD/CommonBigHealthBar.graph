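# CommonBigHealthBar is a HUD health bar widget with a switcher between current HP text and build-mode max HP text.
# InitializeBar, run at PreConstruct, sets green colors and the fill texture through the progress bar's dynamic material.
# Current and max HP setters refresh their text and the fill ratio.
# The fill update also resizes a leading-edge bevel to follow the fill, fading its alpha when it gets narrow.
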WidgetBlueprint /Game/UI/InGame/HUD/CommonBigHealthBar
graphs [
 EventGraph {nodes [
   Event PreConstruct()
   > InitializeBar()
  ]}
 InitializeBar {nodes [
   InitializeBar() {is=title K2Node_FunctionEntry}
   > SetColors!(
    execute
    self
    ColorA:Linear Color Structure=(R=0.0745,G=0.502,B=0.0353,A=1)
    ColorB:Linear Color Structure=(R=0.1765,G=0.7294,B=0.0784,A=1)
    BevelColor:Linear Color Structure=(R=0.4784,G=0.8824,B=0.2314,A=1)
    BaseplateColor:Linear Color Structure=(R=0,G=0,B=0.0078,A=0.6)
   ) -> (then)
   n1 = Get WidgetStyle!(self:Progress Bar Object Reference="Get [ProgressBar] HP Fill"() -> ("[ProgressBar] HP Fill":Progress Bar Object Reference)) -> (WidgetStyle:Progress Bar Style Structure)
   > SetTextureParameterValue(
    self=GetDynamicMaterial!(Brush:"Slate Brush Structure (by ref)"=Break Progress Bar Style!(ProgressBarStyle:"Progress Bar Style Structure (by ref)"=@n1) -> (FillImage:Slate Brush Structure, EnableFillAnimation:bool=~) {is=title K2Node_BreakStruct}) -> (ReturnValue)
    ParameterName=Fill
    Value=/Game/UI/Foundation/Textures/Border/T-Border-HUD-HitPointBar-Bg
   )
   > Update HP Fill Bar()
  ]}
 SetColors {nodes [
   n1 = SetColors!() -> (then, ColorA:Linear Color Structure, ColorB:Linear Color Structure, BevelColor:Linear Color Structure, BaseplateColor:Linear Color Structure) {is=title K2Node_FunctionEntry}
   n2 = "Get [ProgressBar] HP Fill"()
   n3 = GetDynamicMaterial(Brush=Break Progress Bar Style(ProgressBarStyle=Get WidgetStyle(self=@n2)) -> (EnableFillAnimation=~))
   > SetVectorParameterValue(self=@n3, ParameterName=ColorA, Value=@n1)
   > SetVectorParameterValue(self=@n3, ParameterName=ColorB, Value=@n1.ColorB)
   > SetColorAndOpacity(
    self:Image Object Reference=Get ImageHPBevel() -> (ImageHPBevel:Image Object Reference)
    InColorAndOpacity:Linear Color Structure=@n1.BevelColor
   )
   > SetColorAndOpacity(self="Get [Image] HP Baseplate"() -> ("[Image] HP Baseplate":Image Object Reference), InColorAndOpacity=@n1.BaseplateColor)
   SetVectorParameterValue(ParameterName=ColorA)
  ]}
 SetSwitcherTextIndex {nodes [
   n1 = SetSwitcherTextIndex!() -> (then, Index:int) {is=title K2Node_FunctionEntry}
   > SetActiveWidgetIndex(self=Get SwitcherHPText() -> (SwitcherHPText:Common Widget Switcher Legacy Object Reference), Index=@n1)
  ]}
 Update {nodes [
   Update() {is=title K2Node_FunctionEntry}
   n1 = Get Last HP Value() -> (Last HP Value:double)
   n2 = Get Current HP Value() -> (Current HP Value:double)
   > Branch(Condition="Not Equal (Float)"!(A:double=@n1, B:double=@n2) -> (ReturnValue:bool) {is=title K2Node_CallFunction})
   > Set Last HP Value(Last HP Value:double=Lerp(A=@n1, B=@n2, Alpha=Get HP Fade Update Rate() -> (HP Fade Update Rate:double))) -> (Output_Get:double)
  ]}
 Update Current HP Value {nodes [
   n1 = Update Current HP Value!() -> (then, Current:double=~) {is=title K2Node_FunctionEntry}
   > n2 = Branch(Condition="Equal (Float)"!(A:double=@n1, B:double=Get Current HP Value()) -> (ReturnValue:bool) {is=title K2Node_CallFunction})
   Set Last HP Value(execute=@n2.else, Last HP Value=@n1)
   > Set Current HP Value(Current HP Value:double=@n1) -> (Output_Get:double)
   > Update HP Fill Bar()
   > "SetText (Text)"(
    self=Get TextCurrentHealthStat() -> (TextCurrentHealthStat:Common Text Object Reference)
    InText="To Text (Integer)"(Value=Ceil(A=@n1) {is=title K2Node_CallFunction})
   )
  ]}
 Update HP Fill {nodes [
   n1 = Update HP Fill!() -> (then, Percent:double=~) {is=title K2Node_FunctionEntry}
   > SetPercent!(execute, self:Progress Bar Object Reference="Get [ProgressBar] HP Fill"(), InPercent:float=@n1) -> (then)
  ]}
 Update HP Fill Bar {nodes [
   Update HP Fill Bar() {is=title K2Node_FunctionEntry}
   n1 = Get Max HP Value() -> (Max HP Value:double)
   > Branch(Condition="float > float"(A=@n1, B=0) {is=title K2Node_CallFunction})
   n2 = float / float(A=Get Current HP Value(), B=@n1) {is=title K2Node_CallFunction}
   > SetPercent(self="Get [ProgressBar] HP Fill"(), InPercent=@n2)
   n3 = Get ImageHPBevel()
   > n4 = Cast To SizeBoxSlot!(execute, Object=Get Slot!(self:Widget Object Reference=@n3) -> (Slot:Panel Slot Object Reference)) -> (then, CastFailed, AsSize Box Slot:Size Box Slot Object Reference) {is=title K2Node_DynamicCast}
   n5 = float + float(
    A=Break Margin!(Margin:"Margin Structure (by ref)"=Get Padding!(self:Size Box Slot Object Reference=@n4) -> (Padding:Margin Structure)) -> (Right:float=~) {is=title K2Node_BreakStruct}
    B=float * float(A=@n2, B="To Float (Integer)"(InInt=288))
   ) {is=title K2Node_CommutativeAssociativeBinaryOperator}
   > SetWidthOverride(self="Get [Size Box] HP Bevel"() -> ("[Size Box] HP Bevel":Size Box Object Reference), InWidthOverride=@n5)
   n6 = Break Linear Color!(LinearColor:"Linear Color Structure (by ref)"=Get ColorAndOpacity!(self:Image Object Reference=@n3) -> (ColorAndOpacity:Linear Color Structure)) -> (R:float=~, G:float=~, B:float=~, A:float=~) {is=title K2Node_BreakStruct}
   n7 = Get Brush!(self:Image Object Reference=@n3) -> (Brush:Slate Brush Structure)
   n8 = float * float(
    A=BreakVector2D(InVec=Break Slate Brush!(SlateBrush:"Slate Brush Structure (by ref)"=@n7) -> (ImageSize:Vector 2D Structure, OutlineSettings:Slate Brush Outline Settings Structure) {is=title K2Node_BreakStruct})
    B=1.2
   )
   n9 = "float <= float"!(A:double=@n5, B:double=@n8) -> (ReturnValue:bool) {is=title K2Node_CallFunction}
   n10 = float / float(A=@n5, B=@n8)
   > SetColorAndOpacity(
    self=@n3
    InColorAndOpacity=Make Linear Color!(R:float=@n6, G:float=@n6.G, B:float=@n6.B, A:float=Select(Option 0:double=1, Option 1:double=@n10, Index=@n9) -> (ReturnValue:double)) -> (LinearColor:Linear Color Structure) {is=title K2Node_MakeStruct}
   )
  ]}
 Update HP Text {nodes [
   n1 = Update HP Text!() -> (then, HP Text:Text) {is=title K2Node_FunctionEntry}
   > "SetText (Text)"(self=Get TextCurrentHealthStat(), InText=@n1)
  ]}
 Update Max HP Value {nodes [
   n1 = Update Max HP Value!() -> (then, Max:double=~) {is=title K2Node_FunctionEntry}
   > n2 = Branch(Condition="Equal (Float)"(A=@n1, B=Get Max HP Value()))
   n3 = Set Max HP Value(execute=@n2.else, Max HP Value:double=@n1) -> (Output_Get:double)
   n4 = Ceil(A=@n3)
   > "SetText (Text)"(
    self=Get TextBuildModeHPText() -> (TextBuildModeHPText:Common Text Object Reference)
    InText=Format Text!(Format="LOCTEXT(\"{x} HP\")", x:Text="To Text (Integer)"(Value=@n4)) -> (Result)
   )
   > Update HP Fill Bar() {is=title K2Node_CallFunction}
  ]}
]
variables [
 Current HP Value {type=number}
 HP Fade Update Rate {type=number, default=0.05}
 Last HP Value {type=number, default=1000}
 Max HP Value {type=number, default=1}
]
widgets {
 info {inheritedWidgetCount=0, parentClass=&CommonUserWidget, rootWidgetClass=&SizeBox, widgetCount=11}
 widgets [
  "[Size Box] HP Bar" {
   widgetClassPath=&SizeBox
   bIsVariable=true
   widget="[Size Box] HP Bar"
   properties {bOverride_HeightOverride=true, bOverride_WidthOverride=true, heightOverride=24, widthOverride=288}
  }
  "[Overlay]" {widgetClassPath=&Overlay, parent="[Size Box] HP Bar", widget="[Overlay]", slot {class=SizeBoxSlot}}
  "[Overlay] HP Bar Artwork" {
   widgetClassPath=&Overlay
   bIsVariable=true
   parent="[Overlay]"
   widget="[Overlay] HP Bar Artwork"
   slot {class=OverlaySlot, properties {horizontalAlignment=HAlign_Fill, verticalAlignment=VAlign_Fill}}
  }
  "[Image] HP Baseplate" {
   widgetClassPath=&Image
   bIsVariable=true
   parent="[Overlay] HP Bar Artwork"
   widget="[Image] HP Baseplate"
   properties {
    brush {
     bIsDynamicallyLoaded=false
     drawAs=Image
     imageType=NoImage
     mirroring=NoMirror
     resourceName=None
     resourceObject=&T-Border-HUD-HitPointBar-Bg
     tiling=NoTile
     imageSize=<288 24>
     margin=<ltrb 0.4 0 0.4 0>
     outlineSettings {
      bUseBrushTransparency=false
      roundingType=HalfHeightRadius
      width=0
      color {colorUseRule=UseColor_Specified, specifiedColor=<rgba 0 0 0 0>}
      cornerRadii=<0 0 0 0>
     }
     tintColor {colorUseRule=UseColor_Specified, specifiedColor=<rgba 1 1 1 1>}
     uVRegion {bIsValid=false, max=<0 0>, min=<0 0>}
    }
    colorAndOpacity=<rgba 0 0.0012 0.0103 0.6>
   }
   slot {class=OverlaySlot, properties {horizontalAlignment=HAlign_Fill, verticalAlignment=VAlign_Fill}}
  }
  "[ProgressBar] HP Fill" {
   widgetClassPath=/Script/UMG.ProgressBar
   bIsVariable=true
   parent="[Overlay] HP Bar Artwork"
   widget="[ProgressBar] HP Fill"
   properties {
    percent=1
    widgetStyle {
     enableFillAnimation=false
     backgroundImage {
      bIsDynamicallyLoaded=false
      drawAs=NoDrawType
      imageType=FullColor
      mirroring=NoMirror
      resourceName=../../../Engine/Content/Slate/Common/ProgressBar_Background.png
      resourceObject=null
      tiling=NoTile
      imageSize=<0 0>
      margin=<ltrb 0 0 0 0>
      outlineSettings {
       bUseBrushTransparency=false
       roundingType=HalfHeightRadius
       width=0
       color {colorUseRule=UseColor_Specified, specifiedColor=<rgba 0 0 0 0>}
       cornerRadii=<0 0 0 0>
      }
      tintColor {colorUseRule=UseColor_Specified, specifiedColor=<rgba 1 1 1 1>}
      uVRegion {bIsValid=false, max=<0 0>, min=<0 0>}
     }
     fillImage {
      bIsDynamicallyLoaded=false
      drawAs=Image
      imageType=FullColor
      mirroring=NoMirror
      resourceName=../../../Engine/Content/Slate/Common/ProgressBar_Fill.png
      resourceObject=/Game/UI/HUD/Materials/HUD-Bar_Material
      tiling=NoTile
      imageSize=<0 0>
      margin=<ltrb 0.4 0 0.4 0>
      outlineSettings {
       bUseBrushTransparency=false
       roundingType=HalfHeightRadius
       width=0
       color {colorUseRule=UseColor_Specified, specifiedColor=<rgba 0 0 0 0>}
       cornerRadii=<0 0 0 0>
      }
      tintColor {colorUseRule=UseColor_Specified, specifiedColor=<rgba 1 1 1 1>}
      uVRegion {bIsValid=false, max=<0 0>, min=<0 0>}
     }
     marqueeImage {
      bIsDynamicallyLoaded=false
      drawAs=NoDrawType
      imageType=FullColor
      mirroring=NoMirror
      resourceName=../../../Engine/Content/Slate/Common/ProgressBar_Marquee.png
      resourceObject=null
      tiling=Both
      imageSize=<4 4>
      margin=<ltrb 0 0 0 0>
      outlineSettings {
       bUseBrushTransparency=false
       roundingType=HalfHeightRadius
       width=0
       color {colorUseRule=UseColor_Specified, specifiedColor=<rgba 0 0 0 0>}
       cornerRadii=<0 0 0 0>
      }
      tintColor {colorUseRule=UseColor_Specified, specifiedColor=<rgba 1 1 1 1>}
      uVRegion {bIsValid=false, max=<0 0>, min=<0 0>}
     }
    }
   }
   slot {class=OverlaySlot, properties {horizontalAlignment=HAlign_Fill, verticalAlignment=VAlign_Fill}}
  }
  "[Size Box] HP Bevel" {
   widgetClassPath=&SizeBox
   bIsVariable=true
   parent="[Overlay] HP Bar Artwork"
   widget="[Size Box] HP Bevel"
   properties {bOverride_WidthOverride=true, widthOverride=288}
   slot {class=OverlaySlot, properties {verticalAlignment=VAlign_Center}}
  }
  ImageHPBevel {
   widgetClassPath=&Image
   bIsVariable=true
   parent="[Size Box] HP Bevel"
   widget=ImageHPBevel
   properties {
    brush {
     bIsDynamicallyLoaded=false
     drawAs=Box
     imageType=NoImage
     mirroring=NoMirror
     resourceName=None
     resourceObject=/Game/UI/Foundation/Textures/Border/T-Border-HUD-HitPointBar-LeadEdge
     tiling=NoTile
     imageSize=<24 24>
     margin=<ltrb 0 0 0.9 0>
     outlineSettings {
      bUseBrushTransparency=false
      roundingType=HalfHeightRadius
      width=0
      color {colorUseRule=UseColor_Specified, specifiedColor=<rgba 0 0 0 0>}
      cornerRadii=<0 0 0 0>
     }
     tintColor {colorUseRule=UseColor_Specified, specifiedColor=<rgba 1 1 1 1>}
     uVRegion {bIsValid=false, max=<0 0>, min=<0 0>}
    }
    colorAndOpacity=<rgba 0.3278 0.807 0.0185 1>
   }
   slot {class=SizeBoxSlot, properties {verticalAlignment=VAlign_Center}}
  }
  SwitcherHPText {
   widgetClassPath=&CommonWidgetSwitcherLegacy
   bIsVariable=true
   parent="[Overlay]"
   widget=SwitcherHPText
   properties {bOutroPanelBelow=false, bWidgetActivationEnabled=false, transitionDuration=0}
   slot {class=OverlaySlot, properties {verticalAlignment=VAlign_Center, padding=<ltrb 16 0 0 0>}}
  }
  "[Horizontal Box] Health Stat Text" {
   widgetClassPath=&HorizontalBox
   bIsVariable=true
   parent=SwitcherHPText
   widget="[Horizontal Box] Health Stat Text"
   slot {class=WidgetSwitcherSlot}
  }
  TextCurrentHealthStat {
   widgetClassPath=&CommonTextBlock
   bIsVariable=true
   parent="[Horizontal Box] Health Stat Text"
   widget=TextCurrentHealthStat
   properties {
    justification=Right
    style=/FortUILibrary/Styles/Typography/Foundation/TextStyle-Base-XS-B.TextStyle-Base-XS-B_C
    text=BIG HEALTH LEFT
    font {
     bForceMonospaced=false
     bMaterialIsStencil=false
     fontMaterial=null
     fontObject=&NotoSans
     letterSpacing=0
     monospacedWidth=1
     size=16
     skewAmount=0
     typefaceFontName=NotoSans-Bold
     outlineSettings {
      bApplyOutlineToDropShadows=false
      bMiteredCorners=false
      bSeparateFillAlpha=false
      outlineMaterial=null
      outlineSize=0
      outlineColor=<rgba 0 0 0 1>
     }
    }
    shadowOffset=<0 0>
   }
   slot {class=HorizontalBoxSlot, properties {verticalAlignment=VAlign_Center, padding=<ltrb 0 -2.5 0 0>}}
  }
  TextBuildModeHPText {
   widgetClassPath=&CommonTextBlock
   bIsVariable=true
   parent=SwitcherHPText
   widget=TextBuildModeHPText
   properties {
    style=&TextStyle-Base-XS-B_C
    text=??
    font {
     bForceMonospaced=false
     bMaterialIsStencil=false
     fontMaterial=null
     fontObject=&NotoSans
     letterSpacing=0
     monospacedWidth=1
     size=16
     skewAmount=0
     typefaceFontName=NotoSans-Bold
     outlineSettings {
      bApplyOutlineToDropShadows=false
      bMiteredCorners=false
      bSeparateFillAlpha=false
      outlineMaterial=null
      outlineSize=0
      outlineColor=<rgba 0 0 0 1>
     }
    }
    shadowOffset=<0 0>
   }
   slot {class=WidgetSwitcherSlot, properties {horizontalAlignment=HAlign_Left, verticalAlignment=VAlign_Center, padding=<ltrb 0 -2.5 0 0>}}
  }
 ]
}
