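# Common HUD health bar widget that overlays a fill bar on a delta bar, with switchable current/max or build-mode "HP" text.
# Setter functions update both bars as fractions of max HP and refresh rounded-up labels; others control size, fill colors and visibility.
# A timer-based lerp fade for the delta bar exists, but current-HP updates snap it to the new value and clear the timer.
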WidgetBlueprint /Game/UI/InGame/HUD/CommonHealthBar
graphs [
 EventGraph {nodes [
   Event PreConstruct()
   > InitializeBar()
  ]}
 InitializeBar {nodes [
   InitializeBar() {is=title K2Node_FunctionEntry}
   > Update HP Delta Bar()
   > Update HP Fill Bar()
   > SetDeltaVisibility!(execute, self, InVisibility:ESlateVisibility Enum=Visible) -> (then)
  ]}
 Set Timer {nodes [
   n1 = Set Timer!() -> (then, InEnabled:bool=~) {is=title K2Node_FunctionEntry}
   n2 = Get HP Fade Timer() -> (HP Fade Timer:Timer Handle Structure)
   > n3 = Branch(Condition=Is Valid Timer Handle!(Handle:Timer Handle Structure=@n2) -> (ReturnValue:bool))
   > n4 = Clear and Invalidate Timer by Handle(Handle=@n2)
   Branch(execute=[@n4.then, @n3.else], Condition=@n1)
   > n5 = Set Timer by Event(Delegate=Create Event(self=Self-Reference()), Time=Get HP Fade Update Rate() -> (HP Fade Update Rate:double), bLooping=true)
   > Set HP Fade Timer(HP Fade Timer:Timer Handle Structure=@n5) -> (Output_Get:Timer Handle Structure)
  ]}
 SetColors {nodes [
   n1 = SetColors!() -> (then, Color A:Linear Color Structure, Color B:Linear Color Structure, Baseplate Color:Linear Color Structure) {is=title K2Node_FunctionEntry}
   n2 = Get ProgressBarFill() -> (ProgressBarFill:Simple Material Progress Bar Object Reference)
   > SetColorA!(execute, self:Fort Simple Material Progress Bar Object Reference=@n2, ColorA:Linear Color Structure=@n1) -> (then)
   > SetColorB!(execute, self:Fort Simple Material Progress Bar Object Reference=@n2, ColorB:Linear Color Structure=@n1.Color B) -> (then)
  ]}
 SetDeltaVisibility {nodes [
   n1 = SetDeltaVisibility!() -> (then, InVisibility:ESlateVisibility Enum=Visible) {is=title K2Node_FunctionEntry}
   > SetVisibility(self=Get ProgressBarDelta() -> (ProgressBarDelta:Simple Material Progress Bar Object Reference), InVisibility=@n1)
  ]}
 SetHealthSize {nodes [
   n1 = SetHealthSize!() -> (then, NewSize:Vector 2D Structure) {is=title K2Node_FunctionEntry}
   > n2 = Branch(Condition=IsNearlyZero2D!(A:"Vector 2D Structure (by ref)"=@n1, Tolerance:float=0.0001) -> (ReturnValue:bool))
   n3 = "Get [Size Box] HP Bar"() -> ("[Size Box] HP Bar":Size Box Object Reference)
   > ClearWidthOverride(self:Size Box Object Reference=@n3)
   > ClearHeightOverride(self:Size Box Object Reference=@n3)
   n4 = BreakVector2D(InVec=Get NewSize() -> (NewSize:Vector 2D Structure))
   n5 = "Get [Size Box] HP Bar"()
   SetWidthOverride(execute=@n2.else, self=@n5, InWidthOverride=@n4)
   > SetHeightOverride(self=@n5, InHeightOverride=@n4.Y)
  ]}
 SetShowNumbers {nodes [
   n1 = SetShowNumbers!() -> (then, DoShow:bool=~) {is=title K2Node_FunctionEntry}
   > n2 = Delta Gate Wildcard!(execute:Exec, Variable:"Boolean (by ref)"=Get IsNumberShowing() -> (IsNumberShowing:bool), New Value:bool=@n1) -> (Success:Exec, Updated Variable:"Boolean (by ref)", Previous Variable:bool=~) {is=title K2Node_MacroInstance}
   > SetVisibility(
    self=Get SwitcherHPText() -> (SwitcherHPText:Common Widget Switcher Legacy Object Reference)
    InVisibility=Select(Option 0=Collapsed, Option 1=SelfHitTestInvisible, Index=@n2)
   )
  ]}
 SetSwitcherTextIndex {nodes [
   n1 = SetSwitcherTextIndex!() -> (then, Index:int) {is=title K2Node_FunctionEntry}
   > SetActiveWidgetIndex(self=Get SwitcherHPText(), Index=@n1)
  ]}
 Update {nodes [
   Update() {is=title K2Node_FunctionEntry}
   n1 = Get Last HP Value() -> (Last HP Value:double)
   n2 = Get Current HP Value() -> (Current HP Value:double)
   > n3 = Branch(Condition="Not Equal (Float)"!(A:double=@n1, B:double=@n2) -> (ReturnValue:bool) {is=title K2Node_CallFunction})
   > Set Last HP Value(Last HP Value:double=Lerp(A=@n1, B=@n2, Alpha=Get HP Fade Update Rate())) -> (Output_Get:double)
   > Update HP Delta Bar()
   Set Timer!(execute=@n3.else, self, InEnabled:bool=false) -> (then) {is=title K2Node_CallFunction}
  ]}
 Update Current HP Value {nodes [
   n1 = Update Current HP Value!() -> (then, Current:double=~) {is=title K2Node_FunctionEntry}
   > n2 = Branch(Condition="Equal (Float)"!(A:double=@n1, B:double=Get Current HP Value()) -> (ReturnValue:bool) {is=title K2Node_CallFunction})
   Set Last HP Value(execute=@n2.else, Last HP Value=@n1)
   > Update HP Delta Bar()
   > Set Current HP Value(Current HP Value:double=@n1) -> (Output_Get:double)
   > Set Timer(InEnabled=false)
   > Update HP Fill Bar()
   n3 = Ceil(A=@n1) {is=title K2Node_CallFunction}
   > "SetText (Text)"(self=Get CurrentHealth() -> (CurrentHealth:Common Text Object Reference), InText="To Text (Integer)"(Value="Max (Integer)"(A=@n3, B=0)))
  ]}
 Update HP Delta Bar {nodes [
   Update HP Delta Bar() {is=title K2Node_FunctionEntry}
   n1 = Get Max HP Value() -> (Max HP Value:double)
   > Branch(Condition="float > float"(A=@n1, B=0) {is=title K2Node_CallFunction})
   > SetProgress!(
    execute
    self:Fort Simple Material Progress Bar Object Reference=Get ProgressBarDelta()
    Progress:float=float / float(A=Get Last HP Value(), B=@n1) {is=title K2Node_CallFunction}
   ) -> (then)
  ]}
 Update HP Fill Bar {nodes [
   Update HP Fill Bar() {is=title K2Node_FunctionEntry}
   > Branch(Condition="float > float"(A=Get Max HP Value(), B=0))
   > SetProgress(self=Get ProgressBarFill(), Progress=float / float(A=Get Current HP Value(), B=Get Max HP Value()))
  ]}
 Update Max HP Value {nodes [
   n1 = Update Max HP Value!() -> (then, Max:double=~) {is=title K2Node_FunctionEntry}
   > n2 = Branch(Condition="Equal (Float)"(A=@n1, B=Get Max HP Value()))
   n3 = Set Max HP Value(execute=@n2.else, Max HP Value:double=@n1) -> (Output_Get:double)
   n4 = "To Text (Integer)"(Value=Ceil(A=@n3))
   > "SetText (Text)"(self=Get MaxHealth() -> (MaxHealth:Common Text Object Reference), InText=@n4)
   > "SetText (Text)"(
    self=Get BuildModeHP() -> (BuildModeHP:Common Text Object Reference)
    InText=Format Text!(Format="LOCTEXT(\"{x} HP\")", x:Text=@n4) -> (Result)
   )
   > Update HP Fill Bar() {is=title K2Node_CallFunction}
   > Update HP Delta Bar() {is=title K2Node_CallFunction}
  ]}
]
variables [
 Current HP Value {type=number}
 HP Fade Timer {type=TimerHandle}
 HP Fade Update Rate {type=number, default=0.05}
 IsNumberShowing {type=boolean}
 Last HP Value {type=number, default=1000}
 Max HP Value {type=number, default=1}
]
widgets {
 info {inheritedWidgetCount=0, parentClass=&CommonUserWidget, rootWidgetClass=&SizeBox, widgetCount=10}
 widgets [
  "[Size Box] HP Bar" {
   widgetClassPath=&SizeBox
   bIsVariable=true
   widget="[Size Box] HP Bar"
   properties {bOverride_MinDesiredWidth=true, minDesiredWidth=300}
  }
  Overlay_0 {widgetClassPath=&Overlay, parent="[Size Box] HP Bar", widget=Overlay_0, slot {class=SizeBoxSlot}}
  ProgressBarDelta {
   widgetClassPath=/Game/Athena/HUD/SimpleMaterialProgressBar
   bIsVariable=true
   parent=Overlay_0
   widget=ProgressBarDelta
   properties {
    bOverrideDefaultColorA=true
    bOverrideDefaultColorB=true
    bOverrideDefaultColorBackground=true
    designTime_Progress=0.75
    cachedColorA=<rgba 0.0742 0.5029 0.0343 1>
    cachedColorB=<rgba 0.1779 0.7305 0.0782 1>
    cachedColorBackground=<rgba 0 0.0012 0.0103 0.6>
   }
   slot {class=OverlaySlot, properties {horizontalAlignment=HAlign_Fill, verticalAlignment=VAlign_Fill}}
  }
  ProgressBarFill {
   widgetClassPath=&SimpleMaterialProgressBar
   bIsVariable=true
   parent=Overlay_0
   widget=ProgressBarFill
   properties {
    bOverrideDefaultColorA=true
    bOverrideDefaultColorB=true
    bOverrideDefaultColorBackground=true
    designTime_Progress=0.5
    cachedColorA=<rgba 0.0742 0.5029 0.0343 1>
    cachedColorB=<rgba 0.1779 0.7305 0.0782 1>
   }
   slot {class=OverlaySlot, properties {horizontalAlignment=HAlign_Fill, verticalAlignment=VAlign_Fill}}
  }
  SwitcherHPText {
   widgetClassPath=&CommonWidgetSwitcherLegacy
   bIsVariable=true
   parent=Overlay_0
   widget=SwitcherHPText
   properties {bOutroPanelBelow=false, bWidgetActivationEnabled=false, transitionDuration=0}
   slot {class=OverlaySlot, properties {verticalAlignment=VAlign_Center, padding=<ltrb 16 0 0 0>}}
  }
  "[Horizontal Box] Health Stat Text" {
   widgetClassPath=&HorizontalBox
   bIsVariable=true
   parent=SwitcherHPText
   widget="[Horizontal Box] Health Stat Text"
   slot {class=WidgetSwitcherSlot}
  }
  CurrentHealth {
   widgetClassPath=&CommonTextBlock
   bIsVariable=true
   parent="[Horizontal Box] Health Stat Text"
   widget=CurrentHealth
   properties {
    justification=Right
    style=/FortUILibrary/Styles/Typography/Foundation/TextStyle-Base-XS-B.TextStyle-Base-XS-B_C
    text=999
    font {
     bForceMonospaced=false
     bMaterialIsStencil=false
     fontMaterial=null
     fontObject=&NotoSans
     letterSpacing=0
     monospacedWidth=1
     size=16
     skewAmount=0
     typefaceFontName=NotoSans-Bold
     outlineSettings {
      bApplyOutlineToDropShadows=false
      bMiteredCorners=false
      bSeparateFillAlpha=false
      outlineMaterial=null
      outlineSize=0
      outlineColor=<rgba 0 0 0 1>
     }
    }
    shadowOffset=<0 0>
   }
   slot {class=HorizontalBoxSlot, properties {verticalAlignment=VAlign_Center}}
  }
  Split {
   widgetClassPath=&Image
   parent="[Horizontal Box] Health Stat Text"
   widget=Split
   properties {
    brush {
     bIsDynamicallyLoaded=false
     drawAs=Image
     imageType=NoImage
     mirroring=NoMirror
     resourceName=None
     resourceObject=null
     tiling=NoTile
     imageSize=<4 4>
     margin=<ltrb 0 0 0 0>
     outlineSettings {
      bUseBrushTransparency=false
      roundingType=HalfHeightRadius
      width=0
      color {colorUseRule=UseColor_Specified, specifiedColor=<rgba 0 0 0 0>}
      cornerRadii=<0 0 0 0>
     }
     tintColor {colorUseRule=UseColor_Specified, specifiedColor=<rgba 1 1 1 1>}
     uVRegion {bIsValid=false, max=<0 0>, min=<0 0>}
    }
    colorAndOpacity=<rgba 1 1 1 0.2>
   }
   slot {class=HorizontalBoxSlot, properties {padding=<ltrb 4 2 4 2>}}
  }
  MaxHealth {
   widgetClassPath=&CommonTextBlock
   bIsVariable=true
   parent="[Horizontal Box] Health Stat Text"
   widget=MaxHealth
   properties {
    style=&TextStyle-Base-XS-B_C
    text=999?
    font {
     bForceMonospaced=false
     bMaterialIsStencil=false
     fontMaterial=null
     fontObject=&NotoSans
     letterSpacing=0
     monospacedWidth=1
     size=16
     skewAmount=0
     typefaceFontName=NotoSans-Bold
     outlineSettings {
      bApplyOutlineToDropShadows=false
      bMiteredCorners=false
      bSeparateFillAlpha=false
      outlineMaterial=null
      outlineSize=0
      outlineColor=<rgba 0 0 0 1>
     }
    }
    shadowOffset=<0 0>
   }
   slot {class=HorizontalBoxSlot, properties {horizontalAlignment=HAlign_Left, verticalAlignment=VAlign_Center}}
  }
  BuildModeHP {
   widgetClassPath=&CommonTextBlock
   bIsVariable=true
   parent=SwitcherHPText
   widget=BuildModeHP
   properties {
    style=&TextStyle-Base-XS-B_C
    text="{x} HP"
    font {
     bForceMonospaced=false
     bMaterialIsStencil=false
     fontMaterial=null
     fontObject=&NotoSans
     letterSpacing=0
     monospacedWidth=1
     size=16
     skewAmount=0
     typefaceFontName=NotoSans-Bold
     outlineSettings {
      bApplyOutlineToDropShadows=false
      bMiteredCorners=false
      bSeparateFillAlpha=false
      outlineMaterial=null
      outlineSize=0
      outlineColor=<rgba 0 0 0 1>
     }
    }
    shadowOffset=<0 0>
   }
   slot {class=WidgetSwitcherSlot, properties {horizontalAlignment=HAlign_Left, verticalAlignment=VAlign_Center}}
  }
 ]
}
