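# BuildingBar is a HUD widget showing a building's health and, for traps, an item card with name and uses.
# A switcher toggles normal and big health bars, and display mode selects the normal bar's colors, text and delta visibility.
# UpdateBuildingTrap shows the trap box when the player controller allows, caching a temporary item instance until the trap definition changes.
# Durability setters round up and skip unchanged values, with -1 maximum shown as infinity.
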WidgetBlueprint /Game/UI/InGame/HUD/BuildingBar
graphs [
 EventGraph {links [], nodes []}
 Set Display Mode {nodes [
   n1 = Set Display Mode!() -> (then, InDisplayMode:EFortBuildingInteraction Enum=None) {is=title K2Node_FunctionEntry}
   > n2 = Branch(Condition="Equal (Enum)"(A:EFortBuildingInteraction Enum=@n1, B:EFortBuildingInteraction Enum=Get DisplayMode() -> (DisplayMode:EFortBuildingInteraction Enum=None)))
   n3 = Set DisplayMode(execute=@n2.else, DisplayMode:EFortBuildingInteraction Enum=@n1) -> (Output_Get:EFortBuildingInteraction Enum=None)
   n4 = Get Default Baseplate Color() -> (Default Baseplate Color:Linear Color Structure)
   n5 = Get Default HP Color A() -> (Default HP Color A:Linear Color Structure)
   n6 = Get Alternate HP Color A() -> (Alternate HP Color A:Linear Color Structure)
   n7 = Get Default HP Color B() -> (Default HP Color B:Linear Color Structure)
   n8 = Get Alternate HP Color B() -> (Alternate HP Color B:Linear Color Structure)
   > SetColors!(
    execute
    self:Common Health Bar Object Reference=Get CommonHealthBar() -> (CommonHealthBar:Common Health Bar Object Reference)
    Color A:Linear Color Structure=Select!(
     None:Linear Color Structure=@n5
     Build:Linear Color Structure
     Repair:Linear Color Structure=@n6
     Upgrade:Linear Color Structure=@n5
     Edit:Linear Color Structure=@n5
     BeingModified:Linear Color Structure=@n6
     ConfirmEdit:Linear Color Structure
     Creative:Linear Color Structure
     Index:EFortBuildingInteraction Enum=@n3
    ) -> (ReturnValue:Linear Color Structure)
    Color B:Linear Color Structure=Select(None=@n7, Repair=@n8, Upgrade=@n7, Edit=@n7, BeingModified=@n8, Index=@n3)
    Baseplate Color:Linear Color Structure=Select(None=@n4, Build=@n4, Repair=@n4, Upgrade=@n4, Edit=@n4, BeingModified=@n4, ConfirmEdit=@n4, Index=@n3)
   ) -> (then)
   > SetSwitcherTextIndex!(
    execute
    self:Common Health Bar Object Reference=Get CommonHealthBar()
    Index:int=Select(None:int=0, Build:int=1, Repair:int=0, Upgrade:int=0, Edit:int=0, BeingModified:int=0, ConfirmEdit:int=1, Creative:int=0, Index=@n3) -> (ReturnValue:int)
   ) -> (then)
   > SetDeltaVisibility!(
    execute
    self:Common Health Bar Object Reference=Get CommonHealthBar()
    InVisibility:ESlateVisibility Enum=Select(
     None:ESlateVisibility Enum=Visible
     Build:ESlateVisibility Enum=Hidden
     Repair:ESlateVisibility Enum=Visible
     Upgrade:ESlateVisibility Enum=Visible
     Edit:ESlateVisibility Enum=Visible
     BeingModified:ESlateVisibility Enum=Visible
     ConfirmEdit:ESlateVisibility Enum=Hidden
     Creative:ESlateVisibility Enum=Visible
     Index=@n3
    ) -> (ReturnValue:ESlateVisibility Enum)
   ) -> (then)
  ]}
 SetBigHealthMode {nodes [
   n1 = SetBigHealthMode!() -> (then, BigHealth:bool=~) {is=title K2Node_FunctionEntry}
   > n2 = Branch(Condition=@n1)
   n3 = Get BigHealthSwitcher() -> (BigHealthSwitcher:Widget Switcher Object Reference)
   > SetActiveWidgetIndex(self=@n3, Index=1)
   SetActiveWidgetIndex(execute=@n2.else, self=@n3, Index=0)
  ]}
 SetBigHealthPercentFill {nodes [
   n1 = SetBigHealthPercentFill!() -> (then, Percent:double=~) {is=title K2Node_FunctionEntry}
   > Update HP Fill!(
    execute
    self:Common Big Health Bar Object Reference=Get CommonBigHealthBar() -> (CommonBigHealthBar:Common Big Health Bar Object Reference)
    Percent:double=@n1
   ) -> (then)
  ]}
 SetBigHealthValue {nodes [
   n1 = SetBigHealthValue!() -> (then, HP Text:Text) {is=title K2Node_FunctionEntry}
   > Update HP Text!(execute, self:Common Big Health Bar Object Reference=Get CommonBigHealthBar(), HP Text:Text=@n1) -> (then)
  ]}
 SetHealthBarScale {nodes [
   n1 = SetHealthBarScale!() -> (then, Scale:Vector 2D Structure) {is=title K2Node_FunctionEntry}
   > SetRenderScale!(execute, self:Widget Object Reference=Get CommonHealthBar(), Scale:Vector 2D Structure=@n1) -> (then)
  ]}
 SetHealthBarSize {nodes [
   n1 = SetHealthBarSize!() -> (then, Size:Vector 2D Structure) {is=title K2Node_FunctionEntry}
   > SetHealthSize!(execute, self:Common Health Bar Object Reference=Get CommonHealthBar(), NewSize:Vector 2D Structure=@n1) -> (then)
  ]}
 SetShowNumbersOnHealthBar {nodes [
   n1 = SetShowNumbersOnHealthBar!() -> (then, DoShow:bool=~) {is=title K2Node_FunctionEntry}
   > SetShowNumbers!(execute, self:Common Health Bar Object Reference=Get CommonHealthBar(), DoShow:bool=@n1) -> (then)
  ]}
 Update Current Durability Value {nodes [
   n1 = Update Current Durability Value!() -> (then, CurrentDurability:double=~) {is=title K2Node_FunctionEntry}
   n2 = Ceil(A=@n1) {is=title K2Node_CallFunction}
   > n3 = Branch(Condition="Equal (Integer)"(A=@n2, B=Get Current Durability Value() -> (Current Durability Value:int=0)) {is=title K2Node_CallFunction})
   Set Current Durability Value(execute=@n3.else, Current Durability Value:int=@n2) -> (Output_Get:int=0)
   > Update Durability Text()
  ]}
 Update Current HP Value {nodes [
   n1 = Update Current HP Value!() -> (then, Current:double=~) {is=title K2Node_FunctionEntry}
   > Update Current HP Value!(execute, self:Common Health Bar Object Reference=Get CommonHealthBar(), Current:double=@n1) -> (then) {is=title K2Node_CallFunction}
  ]}
 Update Durability Text {nodes [
   Update Durability Text() {is=title K2Node_FunctionEntry}
   n1 = Get Max Durability Value() -> (Max Durability Value:int=0)
   n2 = Get Current Durability Value() -> (Current Durability Value=0)
   > "SetText (Text)"(
    self=Get TrapUsesText() -> (TrapUsesText:Common Text Object Reference)
    InText=Select!(
     Option 0:Text=Format Text!(Format="LOCTEXT(\"{0} / {1}\")", 0:int=@n2, 1:int=@n1) -> (Result)
     Option 1:Text=MakeLiteralText(Value="LOCTEXT(\"∞\")")
     Index:bool="Equal (Integer)"(A=@n1, B=-1)
    ) -> (ReturnValue:Text)
   )
  ]}
 Update Max Durability Value {nodes [
   n1 = Update Max Durability Value!() -> (then, Max:double=~) {is=title K2Node_FunctionEntry}
   n2 = Ceil(A=@n1)
   > n3 = Branch(Condition="Equal (Integer)"(A=@n2, B=Get Max Durability Value() -> (Max Durability Value=0)))
   n4 = Set Max Durability Value(execute=@n3.else, Max Durability Value:int=@n2) -> (Output_Get:int=0)
   > n5 = Branch(Condition="Equal (Integer)"(A=@n4, B=-1))
   > n6 = Set Current Durability Value(Current Durability Value=-1) -> (Output_Get=0)
   Update Durability Text(execute=[@n5.else, @n6.then]) {is=title K2Node_CallFunction}
  ]}
 Update Max HP Value {nodes [
   n1 = Update Max HP Value!() -> (then, Max:double=~) {is=title K2Node_FunctionEntry}
   > Update Max HP Value!(execute, self:Common Health Bar Object Reference=Get CommonHealthBar(), Max:double=@n1) -> (then) {is=title K2Node_CallFunction}
  ]}
 UpdateBuildingTrap {nodes [
   n1 = UpdateBuildingTrap!() -> (then, BuildingTrap:Building Trap Object Reference) {is=title K2Node_FunctionEntry}
   > n2 = Sequence()
   > Is Valid(InputObject=@n1)
   > Is Valid(InputObject=Get TrapData!(self:Building Trap Object Reference=@n1) -> (TrapData:Fort Trap Item Definition Object Reference))
   > n3 = Cast To FortPlayerController(Object=GetOwningPlayer())
   > Set bShowTrapInfo(bShowTrapInfo:bool=ShouldShowInfoIndicator!(self:Fort Player Controller Object Reference=@n3, TestActor:Actor Object Reference=@n1) -> (ReturnValue:bool)) -> (Output_Get:bool)
   n4 = Branch(execute=@n2.then_1, Condition=Get bShowTrapInfo() -> (bShowTrapInfo:bool))
   n5 = Get TrapBox() -> (TrapBox:Size Box Object Reference)
   > SetVisibility(self=@n5, InVisibility=HitTestInvisible)
   > n6 = Is Valid(InputObject=Get TemporaryTrapInstance() -> (TemporaryTrapInstance:Fort Item Object Reference))
   n7 = Get TrapData(self=@n1)
   "SetText (Text)"(
    execute=[@n11.else, @n6.Is Not Valid]
    self=Get TrapNameTextBlock() -> (TrapNameTextBlock:Common Text Object Reference)
    InText=GetShortDisplayName!(self:Fort Weapon Item Definition Object Reference=@n7) -> (ReturnValue:Text)
   )
   n8 = CreateTemporaryItemInstance!(
    self:Fort Item Definition Object Reference=@n7
    Count:int=1
    Level:int=GetTrapLevel!(self:Building Trap Object Reference=@n1) -> (ReturnValue:int=0)
   ) -> (ReturnValue:Fort Item Object Reference)
   > n9 = Set TemporaryTrapInstance(TemporaryTrapInstance:Fort Item Object Reference=@n8) -> (Output_Get:Fort Item Object Reference)
   > SetItemToRepresent!(
    execute
    self:Fort Item Widget NUI Object Reference=Get TrapItemCard() -> (TrapItemCard:Fort Multi Size Item Card Object Reference)
    ItemToRepresent:Fort Item Object Reference=@n9
   ) -> (then)
   > SetColorAndOpacity(
    self=Get TrapNameTextBlock()
    InColorAndOpacity=Make Slate Color(SpecifiedColor=GetRarityColor!(self:Fort Item Object Reference=@n8) -> (ReturnValue:Linear Color Structure))
   )
   n10 = Get TemporaryTrapInstance()
   n11 = Branch(
    execute=@n6.Is Valid
    Condition="Equal (Object)"(A=GetItemDefinition!(self:Fort Item Object Reference=@n10) -> (ReturnValue:Fort Item Definition Object Reference), B=@n7) {is=title K2Node_CallFunction}
   )
   SetVisibility(execute=@n4.else, self=@n5, InVisibility=Collapsed)
   > Set TemporaryTrapInstance()
  ]}
]
variables [
 Alternate HP Color A {type=LinearColor, default=<rgba 0.6654 0.2462 0.0296 1>}
 Alternate HP Color B {type=LinearColor, default=<rgba 0.8228 0.552 0.0648 1>}
 Current Durability Value {type=integer}
 Default Baseplate Color {type=LinearColor, default=<rgba 0 0.0012 0.0103 0.6>}
 Default HP Color A {type=LinearColor, default=<rgba 0.0762 0.5029 0.0356 1>}
 Default HP Color B {type=LinearColor, default=<rgba 0.1779 0.7305 0.0802 1>}
 DisplayMode {type=EFortBuildingInteraction, default=None}
 Max Durability Value {type=integer}
 TemporaryTrapInstance {type=FortItem, default=null}
]
widgets {
 info {inheritedWidgetCount=0, parentClass=&CommonUserWidget, rootWidgetClass=&SizeBox, widgetCount=16}
 widgets [
  SizeBox_1 {widgetClassPath=&SizeBox, widget=SizeBox_1, properties {bOverride_MinDesiredHeight=true, minDesiredHeight=155}}
  VerticalBox_0 {
   widgetClassPath=&VerticalBox
   parent=SizeBox_1
   widget=VerticalBox_0
   slot {class=SizeBoxSlot, properties {verticalAlignment=VAlign_Bottom}}
  }
  TrapBox {
   widgetClassPath=&SizeBox
   bIsVariable=true
   parent=VerticalBox_0
   widget=TrapBox
   properties {bOverride_MinDesiredWidth=true, minDesiredWidth=288, visibility=Collapsed, widthOverride=288}
   slot {class=VerticalBoxSlot, properties {horizontalAlignment=HAlign_Center, padding=<ltrb 0 0 0 24>}}
  }
  TrapVerticalBox {widgetClassPath=&VerticalBox, parent=TrapBox, widget=TrapVerticalBox, slot {class=SizeBoxSlot}}
  CommonBorder_3 {
   widgetClassPath=&CommonBorder
   parent=TrapVerticalBox
   widget=CommonBorder_3
   properties {style=/FortUILibrary/Styles/Border/Border-MainL-Black.Border-MainL-Black_C, padding=<ltrb 6 6 6 6>}
   slot {class=VerticalBoxSlot}
  }
  HorizontalBox_0 {
   widgetClassPath=&HorizontalBox
   parent=CommonBorder_3
   widget=HorizontalBox_0
   slot {class=BorderSlot, properties {padding=<ltrb 6 6 6 6>}}
  }
  SizeBox_0 {
   widgetClassPath=&SizeBox
   parent=HorizontalBox_0
   widget=SizeBox_0
   properties {bOverride_HeightOverride=true, bOverride_WidthOverride=true, heightOverride=90, widthOverride=90}
   slot {class=HorizontalBoxSlot}
  }
  TrapItemCard {
   widgetClassPath=/Script/FortniteUI.FortMultiSizeItemCard
   bIsVariable=true
   parent=SizeBox_0
   widget=TrapItemCard
   properties {bPItemCardSize=S, emptyItemType=AccountItem}
   slot {class=SizeBoxSlot}
  }
  VerticalBox_1 {
   widgetClassPath=&VerticalBox
   parent=HorizontalBox_0
   widget=VerticalBox_1
   slot {class=HorizontalBoxSlot, properties {verticalAlignment=VAlign_Center, padding=<ltrb 6 0 0 0>}}
  }
  HorizontalBox_2 {widgetClassPath=&HorizontalBox, parent=VerticalBox_1, widget=HorizontalBox_2, slot {class=VerticalBoxSlot}}
  CommonTextBlock_75 {
   widgetClassPath=&CommonTextBlock
   parent=HorizontalBox_2
   widget=CommonTextBlock_75
   properties {
    style=/FortUILibrary/Styles/Typography/Foundation/TextStyle-Base-S-B-S.TextStyle-Base-S-B-S_C
    text="Uses: "
    font {
     bForceMonospaced=false
     bMaterialIsStencil=false
     fontMaterial=null
     fontObject=&NotoSans
     letterSpacing=0
     monospacedWidth=1
     size=18
     skewAmount=0
     typefaceFontName=NotoSans-Bold
     outlineSettings {
      bApplyOutlineToDropShadows=true
      bMiteredCorners=false
      bSeparateFillAlpha=true
      outlineMaterial=&M-FontShadow
      outlineSize=3
      outlineColor=<rgba 0 0 0 0.5>
     }
    }
    shadowOffset=<0 0>
   }
   slot {class=HorizontalBoxSlot}
  }
  TrapUsesText {
   widgetClassPath=&CommonTextBlock
   bIsVariable=true
   parent=HorizontalBox_2
   widget=TrapUsesText
   properties {
    style=&TextStyle-Base-S-B-S_C
    text=X/X
    font {
     bForceMonospaced=false
     bMaterialIsStencil=false
     fontMaterial=null
     fontObject=&NotoSans
     letterSpacing=0
     monospacedWidth=1
     size=18
     skewAmount=0
     typefaceFontName=NotoSans-Bold
     outlineSettings {
      bApplyOutlineToDropShadows=true
      bMiteredCorners=false
      bSeparateFillAlpha=true
      outlineMaterial=&M-FontShadow
      outlineSize=3
      outlineColor=<rgba 0 0 0 0.5>
     }
    }
    shadowOffset=<0 0>
   }
   slot {class=HorizontalBoxSlot}
  }
  TrapNameTextBlock {
   widgetClassPath=&CommonTextBlock
   bIsVariable=true
   parent=VerticalBox_1
   widget=TrapNameTextBlock
   properties {
    autoWrapText=true
    style=/FortUILibrary/Styles/Typography/Foundation/TextStyle-Base-XS-B-S.TextStyle-Base-XS-B-S_C
    text=Trap Name
    wrapTextAt=150
    font {
     bForceMonospaced=false
     bMaterialIsStencil=false
     fontMaterial=null
     fontObject=&NotoSans
     letterSpacing=0
     monospacedWidth=1
     size=16
     skewAmount=0
     typefaceFontName=NotoSans-Bold
     outlineSettings {
      bApplyOutlineToDropShadows=true
      bMiteredCorners=false
      bSeparateFillAlpha=true
      outlineMaterial=&M-FontShadow
      outlineSize=2
      outlineColor=<rgba 0 0 0 0.5>
     }
    }
    shadowOffset=<0 0>
   }
   slot {class=VerticalBoxSlot}
  }
  BigHealthSwitcher {
   widgetClassPath=&WidgetSwitcher
   bIsVariable=true
   parent=VerticalBox_0
   widget=BigHealthSwitcher
   slot {class=VerticalBoxSlot, properties {horizontalAlignment=HAlign_Center}}
  }
  CommonHealthBar {
   widgetClassPath=&/CommonHealthBar
   bIsVariable=true
   parent=BigHealthSwitcher
   widget=CommonHealthBar
   slot {class=WidgetSwitcherSlot}
  }
  CommonBigHealthBar {
   widgetClassPath=&/CommonBigHealthBar
   bIsVariable=true
   parent=BigHealthSwitcher
   widget=CommonBigHealthBar
   slot {class=WidgetSwitcherSlot}
  }
 ]
}
