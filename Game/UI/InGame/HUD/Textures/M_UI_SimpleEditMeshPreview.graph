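# Translucent UI material for the in-game HUD, apparently for showing mesh preview images of building pieces in simple edit.
# Its only node is a texture parameter named Mesh, sampled as linear color, so material instances can swap the preview image.
# The default texture is a roof-piece image from the SimpleEditPreviews folder.
# The graph lists no math nodes or output connections.
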Material /Game/UI/InGame/HUD/Textures/M_UI_SimpleEditMeshPreview
graph {
 properties {
  bAutomaticallySetUsageInEditor=false
  bUsedWithStaticMesh=false
  blendMode=BLEND_Translucent
  materialDomain=MD_UI
  usedShadingModels=Default Lit
 }
 nodes [TextureSampleParameter2D() {parameterName=Mesh, samplerType=SAMPLERTYPE_LinearColor, texture=/Game/UI/InGame/HUD/SimpleEditPreviews/_0000_PBWA_W1_Roof0_C_2}]
}
