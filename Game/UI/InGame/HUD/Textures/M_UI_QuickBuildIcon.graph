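# Translucent UI material for the HUD quick-build icon, drawing a pip that changes between idle and highlighted looks.
# The red, green and blue channels of T_UI_QuickEdit_Pip feed channel-packed distance field functions, producing Idle, Inner, Outline and Outer masks.
# An isHighlighted scalar drives two lerps: one switches between mask shapes, the other blends white toward a mask-driven mix of white and two blues.
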Material /Game/UI/InGame/HUD/Textures/M_UI_QuickBuildIcon
graph {
 properties {
  bAutomaticallySetUsageInEditor=false
  bUsedWithStaticMesh=false
  blendMode=BLEND_Translucent
  materialDomain=MD_UI
  usedShadingModels=Default Lit
 }
 nodes [
  n1 = NamedRerouteUsage()
  n2 = NamedRerouteUsage()
  LinearInterpolate(
   A=NamedRerouteUsage()
   B=Add(A=NamedRerouteUsage(), B=LinearInterpolate(A=@n1, B=@n2, Alpha=@n2))
   Alpha=ScalarParameter() {parameterName=isHighlighted}
  )
  n3 = Constant3Vector() {constant=<rgba 1 1 1 0>}
  n4 = Constant3Vector() {constant=<rgba 0.0018 0.2502 1 1>}
  n5 = Constant3Vector() {constant=<rgba 0.2384 0.6038 1 1>}
  n6 = NamedRerouteUsage()
  LinearInterpolate(
   A=@n3
   B=LinearInterpolate(A=@n3, B=LinearInterpolate(A=@n4, B=@n5, Alpha=@n6), Alpha=NamedRerouteUsage())
   Alpha=ScalarParameter() {parameterName=isHighlighted}
  )
  n7 = TextureSample() {texture=&/T_UI_QuickEdit_Pip}
  NamedRerouteDeclaration(None=MaterialFunctionCall!(RGBA Channels=@n7.R, Glow Min=Constant() {r=0.08}, Glow Max=Constant() {r=0.336}, Outline Thickness) {materialFunction=/FortUILibrary/MaterialFunctions/Basics/MF_UI_ChannelPackedDistanceField}.Fill) {name=Idle, nodeColor=<rgba 0 0 0 1>}
  n8 = MaterialFunctionCall(RGBA Channels=@n7.G, Glow Min=Constant() {r=0.044}, Glow Max=Constant() {r=0.376}, Outline Thickness=Constant() {r=0.7322}) {materialFunction=&MF_UI_ChannelPackedDistanceField}
  NamedRerouteDeclaration(None=@n8.Fill) {name=Inner, nodeColor=<rgba 0 0 0 1>}
  NamedRerouteDeclaration(None=@n8.Outline) {name=Outline, nodeColor=<rgba 0 0 0 1>}
  NamedRerouteDeclaration(None=MaterialFunctionCall(RGBA Channels=@n7.B, Glow Min=Constant() {r=0.08}, Glow Max=Constant() {r=0.336}) {materialFunction=&MF_UI_ChannelPackedDistanceField}.Fill) {name=Outer, nodeColor=<rgba 0 0 0 1>}
 ]
}
