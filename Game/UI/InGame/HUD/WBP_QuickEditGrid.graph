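# HUD quick-edit grid showing the selectable variants of a building piece as icons.
# SetBuildingType checks for wall, stairs, floor or roof, maps variant names to Images and shows that layout in a visibility switcher; floor and roof share one.
# Switch State builds a lookup name from item name plus mirror suffix, or rotation on the shared layout.
# It then highlights the matching icon through the isHighlighted dynamic material parameter.
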WidgetBlueprint /Game/UI/InGame/HUD/WBP_QuickEditGrid
graphs [
 EventGraph {nodes [Event Construct()]}
 SetBuildingType {nodes [
   n1 = Get RotationsToPipMap() -> (RotationsToPipMap:Map of Integers to Image Object References)
   Clear!(execute, TargetMap:Map of Integers to Image Object References=@n1) -> (then) {is=title K2Node_CallFunction}
   > Add!(
    execute
    TargetMap:Map of Integers to Image Object References=@n1
    Key:"Integer (by ref)"
    Value:"Image Object Reference (by ref)"=Get RoofTop_L() -> (RoofTop_L:Image Object Reference)
   ) -> (then) {is=title K2Node_CallFunction}
   > Add(TargetMap=@n1, Key=1, Value=Get Brace_L() -> (Brace_L:Image Object Reference))
   > Add(TargetMap=@n1, Key=2, Value=Get Brace_R() -> (Brace_R:Image Object Reference))
   > Add(TargetMap=@n1, Key=3, Value=Get Window() -> (Window:Image Object Reference))
   > Add(TargetMap=@n1, Key=4, Value=Get Door() -> (Door:Image Object Reference))
   n2 = Get RotationsToPipMap()
   Add(TargetMap=@n2, Key=5)
   > Add(TargetMap=@n2, Key=6)
   > Add(TargetMap=@n2, Key=7)
   > Add(TargetMap=@n2, Key=8)
   > Add(TargetMap=@n2, Key=9)
   n3 = SetBuildingType!() -> (then, BuildingActor:Building Actor Object Reference) {is=title K2Node_FunctionEntry}
   > n4 = Branch(Condition=IsA!(Object:Object Reference=@n3, ObjectClass:Object Class Reference=&BuildingWall) -> (ReturnValue:bool))
   n5 = Append(A="To String (Name)"(InName=Get RoofTop_Name() -> (RoofTop_Name:Name=None)), B=_mirrored)
   n6 = Get Brace_Name() -> (Brace_Name:Name=None)
   n7 = Append(A="To String (Name)"(InName=@n6), B=_mirrored)
   > Set WidgetMap(WidgetMap:Map of Names to Image Object References=Make Map!(
     Key 0:Name=String To Name(InString=@n5)
     Value 0:Image Object Reference=Get RoofTop_L()
     Key 1:Name=Get RoofTop_Name() -> (RoofTop_Name=None)
     Value 1:Image Object Reference=Get RoofTop_R() -> (RoofTop_R:Image Object Reference)
     Key 2:Name=@n6
     Value 2:Image Object Reference=Get Brace_L()
     Key 3:Name=String To Name(InString=@n7)
     Value 3:Image Object Reference=Get Brace_R()
     Key 4:Name=Get Window_Name() -> (Window_Name:Name=None)
     Value 4:Image Object Reference=Get Window()
     Key 5:Name=Get Door_Name() -> (Door_Name:Name=None)
     Value 5:Image Object Reference=Get Door()
    ) -> (Map:Map of Names to Image Object References) {is=title K2Node_MakeMap}) -> (Output_Get:Map of Names to Image Object References)
   > SetActiveWidgetIndex(self:Common Visibility Switcher Object Reference=Get TypeSwitcher() -> (TypeSwitcher:Common Visibility Switcher Object Reference), Index=0)
   n8 = Branch(execute=@n4.else, Condition=IsA(Object=@n3, ObjectClass=&BuildingStairs))
   n9 = Append(A="To String (Name)"(InName=Get Stair_Name() -> (Stair_Name:Name=None)), B=_mirrored)
   > Set WidgetMap(WidgetMap=Make Map!(
     Key 0=String To Name(InString=@n9)
     Value 0=Get Stair_l() -> (Stair_l:Image Object Reference)
     Key 1=Get Stair_Name() -> (Stair_Name=None)
     Value 1=Get Stair_R() -> (Stair_R:Image Object Reference)
    ) -> (Map))
   > SetActiveWidgetIndex(self=Get TypeSwitcher(), Index=1)
   n10 = Branch(execute=@n8.else, Condition=IsA(Object=@n3, ObjectClass=&BuildingFloor))
   n11 = Append(A="To String (Name)"(InName=Get Balcony() -> (Balcony:Name=None)), B=0)
   n12 = Append(A="To String (Name)"(InName=Get Balcony() -> (Balcony=None)), B=1)
   n13 = Append(A="To String (Name)"(InName=Get Balcony() -> (Balcony=None)), B=2)
   n14 = Append(A="To String (Name)"(InName=Get Balcony() -> (Balcony=None)), B=3)
   > Set WidgetMap(WidgetMap=Make Map!(
     Key 0=String To Name(InString=@n11)
     Value 0=Get Balcony_0() -> (Balcony_0:Image Object Reference)
     Key 1=String To Name(InString=@n12)
     Value 1=Get Balcony_1() -> (Balcony_1:Image Object Reference)
     Key 2:Name=String To Name(InString=@n13)
     Value 2:Image Object Reference=Get Balcony_2() -> (Balcony_2:Image Object Reference)
     Key 3:Name=String To Name(InString=@n14)
     Value 3:Image Object Reference=Get Balcony_3() -> (Balcony_3:Image Object Reference)
    ) -> (Map))
   > SetActiveWidgetIndex(self=Get TypeSwitcher(), Index=2)
   Branch(execute=@n10.else, Condition=IsA(Object=@n3, ObjectClass=&BuildingRoof))
   n15 = Append(A="To String (Name)"(InName=Get Roof_Name() -> (Roof_Name:Name=None)), B=3)
   n16 = Append(A="To String (Name)"(InName=Get Roof_Name() -> (Roof_Name=None)), B=0)
   n17 = Append(A="To String (Name)"(InName=Get Roof_Name() -> (Roof_Name=None)), B=1)
   n18 = Append(A="To String (Name)"(InName=Get Roof_Name() -> (Roof_Name=None)), B=2)
   > Set WidgetMap(WidgetMap=Make Map(
     Key 0=String To Name(InString=@n16)
     Value 0=Get Balcony_0()
     Key 1=String To Name(InString=@n17)
     Value 1=Get Balcony_1()
     Key 2=String To Name(InString=@n18)
     Value 2=Get Balcony_2()
     Key 3=String To Name(InString=@n15)
     Value 3=Get Balcony_3()
    ))
   > SetActiveWidgetIndex(self=Get TypeSwitcher(), Index=2)
  ]}
 Switch State {nodes [
   n1 = Switch State!() -> (then, SelectedItemName:Name, IsMirrored:bool=~, Rotation:Byte) {is=title K2Node_FunctionEntry}
   n2 = SelectString(A=_mirrored, bPickA=@n1.IsMirrored)
   n3 = "To String (Name)"(InName=@n1)
   > n4 = Set Selected Item Name(Selected Item Name:Name=String To Name(InString=Append(A=@n3, B=@n2))) -> (Output_Get:Name=None)
   n5 = "To String (Name)"(InName=@n4)
   n6 = Get TypeSwitcher()
   n7 = SelectString(
    A="To String (Byte)"!(InByte:Byte=@n1.Rotation) -> (ReturnValue:String)
    bPickA="Equal (Integer)"(A=GetActiveWidgetIndex!(self:Common Visibility Switcher Object Reference=@n6) -> (ReturnValue:int=0), B=2)
   )
   > n8 = Set Selected Item Name(Selected Item Name=String To Name(InString=Append(A=@n5, B=@n7))) -> (Output_Get=None)
   > Set SelectedItem(SelectedItem:Image Object Reference=Find(
     TargetMap:Map of Names to Image Object References=Get WidgetMap() -> (WidgetMap:Map of Names to Image Object References)
     Key:"Name (by ref)"=@n8
    ) -> (Value:Image Object Reference)) -> (Output_Get:Image Object Reference)
   n9 = Cast To PanelWidget!(Object=GetActiveWidget!(self:Common Visibility Switcher Object Reference=Get TypeSwitcher()) -> (ReturnValue:Widget Object Reference)) -> (AsPanel Widget:Panel Widget Object Reference, bSuccess:bool=~) {is=title K2Node_DynamicCast}
   > n10 = For Loop(LastIndex=int - int(A=GetChildrenCount!(self:Panel Widget Object Reference=@n9) -> (ReturnValue:int=0)))
   n11 = GetActiveWidget(self=Get TypeSwitcher())
   n12 = Cast To Image!(Object=GetChildAt!(self:Panel Widget Object Reference=Cast To PanelWidget(Object=@n11) -> (bSuccess=~), Index:int=@n10) -> (ReturnValue:Widget Object Reference)) -> (AsImage:Image Object Reference, bSuccess:bool=~) {is=title K2Node_DynamicCast}
   > n13 = GetDynamicMaterial(self=@n12)
   n14 = Get SelectedItem() -> (SelectedItem:Image Object Reference)
   > SetScalarParameterValue(self=@n13, ParameterName=isHighlighted, Value=SelectFloat(A=1, B=0, bPickA="Equal (Object)"(A=@n12, B=@n14)))
  ]}
]
variables [
 Balcony {type=string, default=pbwa_w1_balconyi_c}
 Brace_Name {type=string, default=pbwa_w1_brace_c}
 Door_Name {type=string, default=pbwa_w1_doorC_C}
 isBalcony {type=boolean}
 Roof_Name {type=string, default=pbwa_w1_RoofO_C}
 RoofTop_Name {type=string, default=pbwa_w1_RoofWall_c}
 RotationsToPipMap {default {}}
 Selected Item Name {type=string, default=None}
 SelectedItem {type=Image, default=null}
 Stair_Name {type=string, default=pbwa_w1_stairF_C}
 WidgetMap {type="map<Image>", default {}}
 Window_Name {type=string, default=pbwa_w1_windowc_c}
]
widgets {
 info {inheritedWidgetCount=0, parentClass=&UserWidget, rootWidgetClass=&CommonVisibilitySwitcher, widgetCount=16}
 widgets [
  TypeSwitcher {widgetClassPath=&CommonVisibilitySwitcher, bIsVariable=true, widget=TypeSwitcher}
  Wall {widgetClassPath=/Script/UMG.UniformGridPanel, bIsVariable=true, parent=TypeSwitcher, widget=Wall, slot {class=CommonVisibilitySwitcherSlot}}
  RoofTop_L {
   widgetClassPath=&Image
   bIsVariable=true
   parent=Wall
   widget=RoofTop_L
   properties {brush {
     bIsDynamicallyLoaded=false
     drawAs=Image
     imageType=FullColor
     mirroring=NoMirror
     resourceName=None
     resourceObject=/Game/UI/InGame/HUD/Textures/M_UI_QuickBuildIcon
     tiling=NoTile
     imageSize=<32 32>
     margin=<ltrb 0 0 0 0>
     outlineSettings {
      bUseBrushTransparency=false
      roundingType=HalfHeightRadius
      width=0
      color {colorUseRule=UseColor_Specified, specifiedColor=<rgba 0 0 0 0>}
      cornerRadii=<0 0 0 0>
     }
     tintColor {colorUseRule=UseColor_Specified, specifiedColor=<rgba 1 1 1 1>}
     uVRegion {bIsValid=false, max=<0 0>, min=<0 0>}
    }}
   slot {class=UniformGridSlot}
  }
  RoofTop_R {
   widgetClassPath=&Image
   bIsVariable=true
   parent=Wall
   widget=RoofTop_R
   properties {brush {
     bIsDynamicallyLoaded=false
     drawAs=Image
     imageType=FullColor
     mirroring=NoMirror
     resourceName=None
     resourceObject=&M_UI_QuickBuildIcon
     tiling=NoTile
     imageSize=<32 32>
     margin=<ltrb 0 0 0 0>
     outlineSettings {
      bUseBrushTransparency=false
      roundingType=HalfHeightRadius
      width=0
      color {colorUseRule=UseColor_Specified, specifiedColor=<rgba 0 0 0 0>}
      cornerRadii=<0 0 0 0>
     }
     tintColor {colorUseRule=UseColor_Specified, specifiedColor=<rgba 1 1 1 1>}
     uVRegion {bIsValid=false, max=<0 0>, min=<0 0>}
    }}
   slot {class=UniformGridSlot, properties {column=2, horizontalAlignment=HAlign_Right}}
  }
  Brace_L {
   widgetClassPath=&Image
   bIsVariable=true
   parent=Wall
   widget=Brace_L
   properties {brush {
     bIsDynamicallyLoaded=false
     drawAs=Image
     imageType=FullColor
     mirroring=NoMirror
     resourceName=None
     resourceObject=&M_UI_QuickBuildIcon
     tiling=NoTile
     imageSize=<32 32>
     margin=<ltrb 0 0 0 0>
     outlineSettings {
      bUseBrushTransparency=false
      roundingType=HalfHeightRadius
      width=0
      color {colorUseRule=UseColor_Specified, specifiedColor=<rgba 0 0 0 0>}
      cornerRadii=<0 0 0 0>
     }
     tintColor {colorUseRule=UseColor_Specified, specifiedColor=<rgba 1 1 1 1>}
     uVRegion {bIsValid=false, max=<0 0>, min=<0 0>}
    }}
   slot {class=UniformGridSlot, properties {row=2, verticalAlignment=VAlign_Bottom}}
  }
  Brace_R {
   widgetClassPath=&Image
   bIsVariable=true
   parent=Wall
   widget=Brace_R
   properties {brush {
     bIsDynamicallyLoaded=false
     drawAs=Image
     imageType=FullColor
     mirroring=NoMirror
     resourceName=None
     resourceObject=&M_UI_QuickBuildIcon
     tiling=NoTile
     imageSize=<32 32>
     margin=<ltrb 0 0 0 0>
     outlineSettings {
      bUseBrushTransparency=false
      roundingType=HalfHeightRadius
      width=0
      color {colorUseRule=UseColor_Specified, specifiedColor=<rgba 0 0 0 0>}
      cornerRadii=<0 0 0 0>
     }
     tintColor {colorUseRule=UseColor_Specified, specifiedColor=<rgba 1 1 1 1>}
     uVRegion {bIsValid=false, max=<0 0>, min=<0 0>}
    }}
   slot {class=UniformGridSlot, properties {column=2, horizontalAlignment=HAlign_Right, row=2, verticalAlignment=VAlign_Bottom}}
  }
  Window {
   widgetClassPath=&Image
   bIsVariable=true
   parent=Wall
   widget=Window
   properties {brush {
     bIsDynamicallyLoaded=false
     drawAs=Image
     imageType=FullColor
     mirroring=NoMirror
     resourceName=None
     resourceObject=&M_UI_QuickBuildIcon
     tiling=NoTile
     imageSize=<32 32>
     margin=<ltrb 0 0 0 0>
     outlineSettings {
      bUseBrushTransparency=false
      roundingType=HalfHeightRadius
      width=0
      color {colorUseRule=UseColor_Specified, specifiedColor=<rgba 0 0 0 0>}
      cornerRadii=<0 0 0 0>
     }
     tintColor {colorUseRule=UseColor_Specified, specifiedColor=<rgba 1 1 1 1>}
     uVRegion {bIsValid=false, max=<0 0>, min=<0 0>}
    }}
   slot {class=UniformGridSlot, properties {column=1, horizontalAlignment=HAlign_Center, row=1, verticalAlignment=VAlign_Center}}
  }
  Door {
   widgetClassPath=&Image
   bIsVariable=true
   parent=Wall
   widget=Door
   properties {brush {
     bIsDynamicallyLoaded=false
     drawAs=Image
     imageType=FullColor
     mirroring=NoMirror
     resourceName=None
     resourceObject=&M_UI_QuickBuildIcon
     tiling=NoTile
     imageSize=<32 32>
     margin=<ltrb 0 0 0 0>
     outlineSettings {
      bUseBrushTransparency=false
      roundingType=HalfHeightRadius
      width=0
      color {colorUseRule=UseColor_Specified, specifiedColor=<rgba 0 0 0 0>}
      cornerRadii=<0 0 0 0>
     }
     tintColor {colorUseRule=UseColor_Specified, specifiedColor=<rgba 1 1 1 1>}
     uVRegion {bIsValid=false, max=<0 0>, min=<0 0>}
    }}
   slot {class=UniformGridSlot, properties {column=1, horizontalAlignment=HAlign_Center, row=2, verticalAlignment=VAlign_Bottom}}
  }
  Stair {widgetClassPath=&UniformGridPanel, parent=TypeSwitcher, widget=Stair, slot {class=CommonVisibilitySwitcherSlot}}
  Stair_l {
   widgetClassPath=&Image
   bIsVariable=true
   parent=Stair
   widget=Stair_l
   properties {brush {
     bIsDynamicallyLoaded=false
     drawAs=Image
     imageType=FullColor
     mirroring=NoMirror
     resourceName=None
     resourceObject=&M_UI_QuickBuildIcon
     tiling=NoTile
     imageSize=<32 32>
     margin=<ltrb 0 0 0 0>
     outlineSettings {
      bUseBrushTransparency=false
      roundingType=HalfHeightRadius
      width=0
      color {colorUseRule=UseColor_Specified, specifiedColor=<rgba 0 0 0 0>}
      cornerRadii=<0 0 0 0>
     }
     tintColor {colorUseRule=UseColor_Specified, specifiedColor=<rgba 1 1 1 1>}
     uVRegion {bIsValid=false, max=<0 0>, min=<0 0>}
    }}
   slot {class=UniformGridSlot, properties {verticalAlignment=VAlign_Center}}
  }
  Stair_R {
   widgetClassPath=&Image
   bIsVariable=true
   parent=Stair
   widget=Stair_R
   properties {brush {
     bIsDynamicallyLoaded=false
     drawAs=Image
     imageType=FullColor
     mirroring=NoMirror
     resourceName=None
     resourceObject=&M_UI_QuickBuildIcon
     tiling=NoTile
     imageSize=<32 32>
     margin=<ltrb 0 0 0 0>
     outlineSettings {
      bUseBrushTransparency=false
      roundingType=HalfHeightRadius
      width=0
      color {colorUseRule=UseColor_Specified, specifiedColor=<rgba 0 0 0 0>}
      cornerRadii=<0 0 0 0>
     }
     tintColor {colorUseRule=UseColor_Specified, specifiedColor=<rgba 1 1 1 1>}
     uVRegion {bIsValid=false, max=<0 0>, min=<0 0>}
    }}
   slot {class=UniformGridSlot, properties {column=1, horizontalAlignment=HAlign_Right, verticalAlignment=VAlign_Center}}
  }
  Balcony {widgetClassPath=&UniformGridPanel, parent=TypeSwitcher, widget=Balcony, slot {class=CommonVisibilitySwitcherSlot}}
  Balcony_0 {
   widgetClassPath=&Image
   bIsVariable=true
   parent=Balcony
   widget=Balcony_0
   properties {brush {
     bIsDynamicallyLoaded=false
     drawAs=Image
     imageType=FullColor
     mirroring=NoMirror
     resourceName=None
     resourceObject=&M_UI_QuickBuildIcon
     tiling=NoTile
     imageSize=<32 32>
     margin=<ltrb 0 0 0 0>
     outlineSettings {
      bUseBrushTransparency=false
      roundingType=HalfHeightRadius
      width=0
      color {colorUseRule=UseColor_Specified, specifiedColor=<rgba 0 0 0 0>}
      cornerRadii=<0 0 0 0>
     }
     tintColor {colorUseRule=UseColor_Specified, specifiedColor=<rgba 1 1 1 1>}
     uVRegion {bIsValid=false, max=<0 0>, min=<0 0>}
    }}
   slot {class=UniformGridSlot, properties {column=1, horizontalAlignment=HAlign_Right}}
  }
  Balcony_1 {
   widgetClassPath=&Image
   bIsVariable=true
   parent=Balcony
   widget=Balcony_1
   properties {brush {
     bIsDynamicallyLoaded=false
     drawAs=Image
     imageType=FullColor
     mirroring=NoMirror
     resourceName=None
     resourceObject=&M_UI_QuickBuildIcon
     tiling=NoTile
     imageSize=<32 32>
     margin=<ltrb 0 0 0 0>
     outlineSettings {
      bUseBrushTransparency=false
      roundingType=HalfHeightRadius
      width=0
      color {colorUseRule=UseColor_Specified, specifiedColor=<rgba 0 0 0 0>}
      cornerRadii=<0 0 0 0>
     }
     tintColor {colorUseRule=UseColor_Specified, specifiedColor=<rgba 1 1 1 1>}
     uVRegion {bIsValid=false, max=<0 0>, min=<0 0>}
    }}
   slot {class=UniformGridSlot, properties {column=1, horizontalAlignment=HAlign_Right, row=1, verticalAlignment=VAlign_Bottom}}
  }
  Balcony_2 {
   widgetClassPath=&Image
   bIsVariable=true
   parent=Balcony
   widget=Balcony_2
   properties {brush {
     bIsDynamicallyLoaded=false
     drawAs=Image
     imageType=FullColor
     mirroring=NoMirror
     resourceName=None
     resourceObject=&M_UI_QuickBuildIcon
     tiling=NoTile
     imageSize=<32 32>
     margin=<ltrb 0 0 0 0>
     outlineSettings {
      bUseBrushTransparency=false
      roundingType=HalfHeightRadius
      width=0
      color {colorUseRule=UseColor_Specified, specifiedColor=<rgba 0 0 0 0>}
      cornerRadii=<0 0 0 0>
     }
     tintColor {colorUseRule=UseColor_Specified, specifiedColor=<rgba 1 1 1 1>}
     uVRegion {bIsValid=false, max=<0 0>, min=<0 0>}
    }}
   slot {class=UniformGridSlot, properties {row=1, verticalAlignment=VAlign_Bottom}}
  }
  Balcony_3 {
   widgetClassPath=&Image
   bIsVariable=true
   parent=Balcony
   widget=Balcony_3
   properties {brush {
     bIsDynamicallyLoaded=false
     drawAs=Image
     imageType=FullColor
     mirroring=NoMirror
     resourceName=None
     resourceObject=&M_UI_QuickBuildIcon
     tiling=NoTile
     imageSize=<32 32>
     margin=<ltrb 0 0 0 0>
     outlineSettings {
      bUseBrushTransparency=false
      roundingType=HalfHeightRadius
      width=0
      color {colorUseRule=UseColor_Specified, specifiedColor=<rgba 0 0 0 0>}
      cornerRadii=<0 0 0 0>
     }
     tintColor {colorUseRule=UseColor_Specified, specifiedColor=<rgba 1 1 1 1>}
     uVRegion {bIsValid=false, max=<0 0>, min=<0 0>}
    }}
   slot {class=UniformGridSlot}
  }
 ]
}
