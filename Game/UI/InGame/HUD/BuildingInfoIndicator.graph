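# HUD indicator widget that follows the focused building, showing health, trap durability and interaction prompts for edit, repair, upgrade and trap placement.
# Construct binds HUD, inventory, input, quick-edit, smart-build and game-phase events.
# Focus and info changes refresh the HP/trap bars and switch prompt widgets, with timed health fading and big-health prop support.
# Quick-edit previews map option names, mirroring and rotation to materials and mesh textures.
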WidgetBlueprint /Game/UI/InGame/HUD/BuildingInfoIndicator
graphs [
 EventGraph {nodes [
   n1 = CustomEvent_0!() -> (OutputDelegate, then, FocusedBuildingInfo:"Fort Focused Building Info Structure (by ref)") {is=title K2Node_CustomEvent}
   > UpdateBuildingInfo!(execute, self, NewBuildingInfo:"Fort Focused Building Info Structure (by ref)"=@n1.FocusedBuildingInfo) -> (then)
   > InvalidateLayoutAndVolatility(self:Widget Object Reference)
   n2 = OnQuickEditActiveChanged!() -> (OutputDelegate, then, bIsActive:bool=~) {is=title K2Node_CustomEvent}
   > n3 = Set bIsQuickEditEnabled(bIsQuickEditEnabled:bool=@n2.bIsActive) -> (Output_Get:bool)
   > n4 = Branch(Condition=@n3)
   UpdateQuickEditPreviewVisibility(execute=@n4.else)
   n5 = On Simple Build Toggled!() -> (OutputDelegate, then, bIsEnabled:bool=~) {is=title K2Node_CustomEvent}
   > HandleSimpleBuild!(execute, self, Enabled?:bool=@n5.bIsEnabled) -> (then)
   n6 = OnPreviewMethodChanged!() -> (OutputDelegate, then, PreviewMethod:EQuickEditPreviewMethod Enum=MeshPreview) {is=title K2Node_CustomEvent}
   > Set QuickEditPreviewMethod(QuickEditPreviewMethod:EQuickEditPreviewMethod Enum=@n6.PreviewMethod) -> (Output_Get:EQuickEditPreviewMethod Enum=MeshPreview)
   Event Destruct()
   > n7 = Get!(execute:Exec) -> (then:Exec, else:Exec, GamePhaseUpdatedAsyncTask:Async Action Start Listening To Stateful Event Object Reference)
   > StopListeningToStatefulEvent(self:Async Action Start Listening To Stateful Event Object Reference=@n7)
   Event Construct()
   n8 = Get FortHUDContext!() -> (ReturnValue:Fort HUDContext Object Reference) {is=title K2Node_GetBlueprintContextEx}
   > Bind Event to On Building Focused(self:Fort HUDContext Object Reference=@n8, Delegate=Create Event()) {is=title K2Node_AddDelegate}
   > Bind Event to On Focused Building State Changed(self:Fort HUDContext Object Reference=@n8, Delegate=@n1) {is=title K2Node_AddDelegate}
   > Bind Event to On Focused Building Health Changed(self:Fort HUDContext Object Reference=@n8, Delegate=Create Event()) {is=title K2Node_AddDelegate}
   > Bind Event to On Focused Building Big Health Changed(self:Fort HUDContext Object Reference=@n8, Delegate=Create Event()) {is=title K2Node_AddDelegate}
   > Bind Event to On Focused Building Attached Trap Durability Changed(self:Fort HUDContext Object Reference=@n8, Delegate=Create Event()) {is=title K2Node_AddDelegate}
   > Bind Event to On Focused Building Attached Trap Changed(self:Fort HUDContext Object Reference=@n8, Delegate=Create Event()) {is=title K2Node_AddDelegate}
   > Bind Event to On Player Targeting Changed(self:Fort HUDContext Object Reference=@n8, Delegate=Create Event()) {is=title K2Node_AddDelegate}
   > Bind Event to On Focused Building Repair Cost Changed(self:Fort HUDContext Object Reference=@n8, Delegate=Create Event()) {is=title K2Node_AddDelegate}
   > Assign On Focused Building Dynamic Data Received!(execute:Exec, self:Fort HUDContext Object Reference=@n8, Delegate:"Delegate (by ref)"=Create Event()) -> (then:Exec) {is=title K2Node_AssignDelegate}
   > Bind Event to On Keybinds Changed(
    self:Fort Global UIContext Object Reference=Get FortGlobalUIContext!() -> (ReturnValue:Fort Global UIContext Object Reference) {is=title K2Node_GetBlueprintContextEx}
    Delegate=Create Event()
   ) {is=title K2Node_AddDelegate}
   > Bind Event to On Quickbar Slot Focus Changed(
    self:Fort Inventory Context Object Reference=Get FortInventoryContext!() -> (ReturnValue:Fort Inventory Context Object Reference) {is=title K2Node_GetBlueprintContextEx}
    Delegate=Create Event()
   ) {is=title K2Node_AddDelegate}
   n9 = Get CommonInputSubsystem()
   > Assign On Input Method Changed!(execute:Exec, self:Common Input Object Reference=@n9, Delegate:"Delegate (by ref)"=Create Event()) -> (then:Exec) {is=title K2Node_AssignDelegate}
   > OnInputTypeChanged!(execute, self, Input Type:ECommonInputType Enum=GetCurrentInputType(self=@n9)) -> (then)
   n10 = GetOwningPlayer()
   > n11 = Cast To FortPlayerController(Object=@n10)
   n12 = GetQuickEditComponent!(self:Fort Player Controller Object Reference=@n11) -> (ReturnValue:Fort Controller Component Quick Edit Object Reference)
   > n13 = Is Valid(InputObject=@n12)
   > Bind Event to On Building Quick Edit Option Changed(self:Fort Controller Component Quick Edit Object Reference=@n12, Delegate=Create Event()) {is=title K2Node_AddDelegate}
   > Bind Event to On Quick Edit Active Changed(self:Fort Controller Component Quick Edit Object Reference=@n12, Delegate=@n2) {is=title K2Node_AddDelegate}
   > Set bIsQuickEditEnabled(bIsQuickEditEnabled=IsEnabled!(self:Fort Controller Component Quick Edit Object Reference=@n12) -> (ReturnValue:bool))
   > Bind Event to On Quick Edit Preview Method Changed(self:Fort Controller Component Quick Edit Object Reference=@n12, Delegate=@n6) {is=title K2Node_AddDelegate}
   > n14 = Set QuickEditPreviewMethod(QuickEditPreviewMethod=Get QuickEditPreviewMethod!(self:Fort Controller Component Quick Edit Object Reference=@n12) -> (QuickEditPreviewMethod:EQuickEditPreviewMethod Enum=MeshPreview)) -> (Output_Get=MeshPreview)
   n15 = Cast To FortPlayerControllerAthena(execute=[@n14.then, @n13.Is Not Valid], Object=@n10)
   > Bind Event to On Build Preview Marker Visibility Changed(self:Fort Player Controller Object Reference=@n15, Delegate=Create Event()) {is=title K2Node_AddDelegate}
   > n16 = StartListeningToStatefulEvent!(
    execute:Exec
    Target:Gameplay Event Router Component Object Reference=GetGlobalEventRouter()
    EventType:Script Struct Object Reference=/Script/FortniteGame.FortGamePhaseStepUpdatedEvent
    Context:Object Reference
    EventBubblingRule:EEventBubblingRule Enum
   ) -> (
    then:Exec
    AsyncTaskProxy:Async Action Start Listening To Stateful Event Object Reference
    OnEventReceived:Exec
    OnSavedEventStateExists:Exec
    OnEventStateCleared:Exec
    Payload:Fort Game Phase Step Updated Event Structure
   ) {is=title K2Node_StartListeningToStatefulEventAsyncAction}
   > Set GamePhaseUpdatedAsyncTask(GamePhaseUpdatedAsyncTask:Async Action Start Listening To Stateful Event Object Reference=@n16) -> (Output_Get:Async Action Start Listening To Stateful Event Object Reference)
   > n17 = Cast To FortPlayerControllerAthena(Object=GetOwningPlayer())
   n18 = GetSmartBuildComponent!(self:Fort Player Controller Object Reference=@n17) -> (ReturnValue:Fort Controller Component Smart Build Object Reference)
   > Is Valid(InputObject=@n18)
   > HandleSimpleBuild(Enabled?=IsSmartBuildEnabled!(self:Fort Controller Component Smart Build Object Reference=@n18) -> (ReturnValue:bool))
   > Bind Event to On Simple Build Toggled(self:Fort Controller Component Smart Build Object Reference=@n18, Delegate=@n5) {is=title K2Node_AddDelegate}
   HandleGamePhaseChanged!(
    execute=[@n16.OnEventReceived, @n16.OnSavedEventStateExists]
    self
    NewPhase:EAthenaGamePhaseStep Enum=Break Fort Game Phase Step Updated Event!(FortGamePhaseStepUpdatedEvent:"Fort Game Phase Step Updated Event Structure (by ref)"=@n16.Payload) -> (GamePhaseStep:EAthenaGamePhaseStep Enum) {is=title K2Node_BreakStruct}
   ) -> (then)
   n19 = GetCurrentFocusedBuilding!(self:Fort HUDContext Object Reference=Get FortHUDContext()) -> (ReturnValue:Building Actor Object Reference)
   Is Valid(InputObject=@n19)
   n20 = GetOwningPlayerPawn!(self:User Widget Object Reference) -> (ReturnValue:Pawn Object Reference)
   n21 = BreakRotator(InRot=FindRelativeLookAtRotation!(StartTransform:"Transform (by ref)"=Get Actor Transform(self=@n19), TargetLocation:"Vector (by ref)"=Get Actor Location(self=@n20)) -> (ReturnValue:Rotator))
   > PrintText!(
    execute
    InText="To Text (Float)"(Value=@n21.Yaw, MinimumIntegralDigits=1, MaximumIntegralDigits=324, MinimumFractionalDigits=0, MaximumFractionalDigits=3)
    bPrintToScreen
    bPrintToLog
    TextColor
    Duration
    Key
   ) -> (then)
   n22 = "float > float"(A=@n21.Yaw, B=130)
   n23 = "float < float"(A=@n21.Yaw, B=179)
   n24 = "float > float"(A=@n21.Yaw, B=-10)
   n25 = "float < float"(A=@n21.Yaw, B=55)
   > n26 = Set bPlayerFacingFromSide(bPlayerFacingFromSide:bool=OR Boolean(A=AND Boolean(A=@n22, B=@n23), B=AND Boolean(A=@n24, B=@n25))) -> (Output_Get:bool)
   > PrintString!(execute, InString="To String (Boolean)"(InBool=@n26), bPrintToScreen, bPrintToLog, TextColor, Duration, Key) -> (then)
   PrintText(InText=Format Text())
   Event Tick()
  ]}
 HandleAttachedTrapChanged {nodes [
   n1 = HandleAttachedTrapChanged!() -> (then, BuildingTrap:Building Trap Object Reference) {is=title K2Node_FunctionEntry}
   > UpdateBuildingTrap!(
    execute
    self:Building Bar Object Reference="Get [Widget] Building HP && Trap Bars"() -> ("[Widget] Building HP && Trap Bars":Building Bar Object Reference)
    BuildingTrap:Building Trap Object Reference=@n1
   ) -> (then)
  ]}
 HandleBuildPreviewMarkerVisibilityChanged {nodes [
   n1 = HandleBuildPreviewMarkerVisibilityChanged!() -> (then, Hidden:bool=~) {is=title K2Node_FunctionEntry}
   > SetVisibility(InVisibility=Select(Option 0=SelfHitTestInvisible, Option 1=Collapsed, Index=@n1))
   > n2 = Branch(Condition=@n1)
   RefreshBuildingInfo(execute=@n2.else)
  ]}
 HandleBuildingBigHealthChanged {nodes [
   n1 = HandleBuildingBigHealthChanged!() -> (then, HP Value:Text, HP Percent:float=~) {is=title K2Node_FunctionEntry}
   n2 = "Get [Widget] Building HP && Trap Bars"()
   > SetBigHealthValue!(execute, self:Building Bar Object Reference=@n2, HP Text:Text=@n1) -> (then)
   > SetBigHealthPercentFill!(execute, self:Building Bar Object Reference=@n2, Percent:double=@n1.HP Percent) -> (then)
  ]}
 HandleBuildingFocused {nodes [
   n1 = HandleBuildingFocused!() -> (then, FocusedBuilding:Building Actor Object Reference) {is=title K2Node_FunctionEntry}
   > Set IsNewFocusedBuilding(IsNewFocusedBuilding:bool="Not Equal (Object)"!(A:Object Reference=Get PreviouslyFocusedBuilding() -> (PreviouslyFocusedBuilding:Building Actor Object Reference), B:Object Reference=@n1) -> (ReturnValue:bool)) -> (Output_Get:bool)
   > Set PreviouslyFocusedBuilding(PreviouslyFocusedBuilding:Building Actor Object Reference=@n1) -> (Output_Get:Building Actor Object Reference)
   > SetBuildingType!(
    execute
    self:WBP Quick Edit Grid Object Reference=Get WBP_QuickEditGrid() -> (WBP_QuickEditGrid:WBP Quick Edit Grid Object Reference)
    BuildingActor:Building Actor Object Reference=@n1
   ) -> (then)
   > SetIndicatedActor!(execute, self:Fort Actor Indicator Widget Object Reference, NewIndicatorActor:Actor Object Reference=@n1) -> (then)
   > n2 = Sequence()
   n3 = GetIndicatedActor!(self:Fort Actor Indicator Interface Interface) -> (ReturnValue:Actor Object Reference)
   > n4 = Cast To BuildingPropBigHealth!(execute, Object=@n3) -> (then, CastFailed, AsBuilding Prop Big Health:Building Prop Big Health Object Reference) {is=title K2Node_DynamicCast}
   n5 = "Get [Widget] Building HP && Trap Bars"()
   > SetBigHealthMode!(execute, self:Building Bar Object Reference=@n5, BigHealth:bool=true) -> (then)
   > SetBigHealthValue(self=@n5, HP Text=GetHealthValueText!(self:Building Prop Big Health Object Reference=@n4) -> (ReturnValue:Text))
   > SetBigHealthPercentFill(self=@n5, Percent=GetHealthValuePercent!(self:Building Prop Big Health Object Reference=@n4) -> (ReturnValue:float))
   SetBigHealthMode(execute=@n4.CastFailed, self=@n5, BigHealth=false)
   UpdateFromTargetingDataOverride(execute=@n2.then_1)
   > n6 = IsCurrentWeaponTheMoveTool!(execute, self) -> (then, Result:bool)
   n7 = IsValid(Object=@n3)
   n8 = NOT Boolean(A=@n6)
   > SetVisibility(InVisibility=Select(Option 0=Hidden, Option 1=SelfHitTestInvisible, Index=AND Boolean(A=@n8, B=@n7)))
   > n9 = Branch(Condition=Get IsNewFocusedBuilding() -> (IsNewFocusedBuilding:bool))
   Set Current Health(execute=[@n11.then, @n9.then], Current Health:double=0) -> (Output_Get:double)
   > Set Current Durability(Current Durability:double=0) -> (Output_Get:double)
   > Set BuildingInfo(BuildingInfo:Fort Focused Building Info Structure) -> (Output_Get:Fort Focused Building Info Structure)
   > Clear and Invalidate Timer by Handle(Handle=Get HealthFadeTimerHandle() -> (HealthFadeTimerHandle:Timer Handle Structure))
   n10 = GetFocusedBuildingInfo!(self:Fort HUDContext Object Reference=Get FortHUDContext()) -> (OutBuildingInfo:Fort Focused Building Info Structure, ReturnValue:bool)
   > Branch(Condition=@n10.ReturnValue)
   > UpdateBuildingInfo(NewBuildingInfo=@n10)
   n11 = HideHealth(execute=@n9.else)
  ]}
 HandleBuildingHealthChanged {nodes [
   n1 = HandleBuildingHealthChanged!() -> (then, Health:float=~) {is=title K2Node_FunctionEntry}
   > Branch(Condition="Not Equal (Float)"!(A:double=@n1, B:double=Get Current Health() -> (Current Health:double)) -> (ReturnValue:bool) {is=title K2Node_CallFunction})
   > n2 = Set Current Health(Current Health=@n1)
   > Update Current HP Value!(execute, self:Building Bar Object Reference="Get [Widget] Building HP && Trap Bars"(), Current:double=@n2) -> (then)
   n3 = Get BuildingInfo() -> (BuildingInfo:Fort Focused Building Info Structure)
   > Branch(Condition="Equal (Enum)"(
     A:EFortBuildingHealthDisplayRule Enum=Break Fort Focused Building Info!(FortFocusedBuildingInfo:"Fort Focused Building Info Structure (by ref)"=@n3) -> (
      bCanBePlayerEdited:bool=~
      HealthDisplayRule:EFortBuildingHealthDisplayRule Enum
      MaxHealth:float=~
      bIsAnyTrapAttached:bool=~
      bIsTrapAttachedFacingPlayer:bool=~
      bIsPreviewTrapAttached:bool=~
      InteractionType:EFortBuildingInteraction Enum
      QuickEditClass:Name
      bIsPlayerBuilt:bool=~
     ) {is=title K2Node_BreakStruct}.HealthDisplayRule
     B:EFortBuildingHealthDisplayRule Enum=Allowed
    ))
   > n4 = Cast To BuildingActor!(execute, Object=GetIndicatedActor()) -> (then, CastFailed, AsBuilding Actor:Building Actor Object Reference) {is=title K2Node_DynamicCast}
   > n5 = Set TargetingData(TargetingData:Building Targeting Data Object Reference=Get BuildingTargetingDataOverride!(self:Building Actor Object Reference=@n4) -> (BuildingTargetingDataOverride:Building Targeting Data Object Reference)) -> (Output_Get:Building Targeting Data Object Reference)
   > n6 = Is Valid(InputObject=@n5)
   > n7 = Branch(Condition=Get bShowHealthWhenFull!(self:Building Targeting Data Object Reference=@n5) -> (bShowHealthWhenFull:bool))
   SetVisibility(execute=[@n15.then, @n7.then], self="Get [Widget] Building HP && Trap Bars"(), InVisibility=SelfHitTestInvisible)
   n8 = Get BuildingInfo()
   > Branch(Condition="Equal (Enum)"(
     A:EFortBuildingInteraction Enum=Break Fort Focused Building Info(FortFocusedBuildingInfo=@n8) -> (bCanBePlayerEdited=~, MaxHealth=~, bIsAnyTrapAttached=~, bIsTrapAttachedFacingPlayer=~, bIsPreviewTrapAttached=~, bIsPlayerBuilt=~).InteractionType
     B:EFortBuildingInteraction Enum=None
    ))
   > Clear and Invalidate Timer by Handle(Handle=Get HealthFadeTimerHandle())
   > n9 = Get!(execute) -> (then, else, TargetingData:Building Targeting Data Object Reference)
   n10 = Get HideBuildingInfoAfterSeconds!(self:Building Targeting Data Object Reference=@n9) -> (HideBuildingInfoAfterSeconds:float)
   > Branch(Condition="float > float"(A=@n10))
   n11 = Create Event()
   > n12 = Set Timer by Event(Delegate=@n11, Time=@n10, bLooping=false)
   > Set HealthFadeTimerHandle(HealthFadeTimerHandle:Timer Handle Structure=@n12) -> (Output_Get:Timer Handle Structure)
   n13 = Set Timer by Event(execute=@n9.else, Delegate=@n11, Time=3, bLooping=false)
   > Set HealthFadeTimerHandle(HealthFadeTimerHandle=@n13)
   n14 = Get BuildingInfo()
   n15 = Branch(
    execute=[@n7.else, @n6.Is Not Valid, @n4.CastFailed]
    Condition="float < float"!(
     A:double=Get Current Health()
     B:double=Break Fort Focused Building Info(FortFocusedBuildingInfo=@n14) -> (bCanBePlayerEdited=~, MaxHealth=~, bIsAnyTrapAttached=~, bIsTrapAttachedFacingPlayer=~, bIsPreviewTrapAttached=~, bIsPlayerBuilt=~).MaxHealth
    ) -> (ReturnValue:bool) {is=title K2Node_CallFunction}
   )
  ]}
 HandleGamePhaseChanged {nodes [
   HandleGamePhaseChanged!() -> (then, NewPhase:EAthenaGamePhaseStep Enum=None) {is=title K2Node_FunctionEntry}
   > HideHealth()
   > Return Node()
  ]}
 HandlePlayerTargetingChanged {nodes [
   n1 = Branch(Condition=HandlePlayerTargetingChanged!() -> (then, bIsTargeting:bool=~) {is=title K2Node_FunctionEntry})
   > Cast To FortWeaponRanged!(
    execute
    Object=GetEquippedWeapon!(self:Fort Inventory Context Object Reference=Get FortInventoryContext()) -> (ReturnValue:Fort Weapon Object Reference)
   ) -> (then, CastFailed, AsFort Weapon Ranged:Fort Weapon Ranged Object Reference) {is=title K2Node_DynamicCast}
   > SetVisibility(InVisibility=Hidden)
   n2 = Get IndicatedActor() -> (IndicatedActor:Actor Weak object)
   SetVisibility(execute=@n1.else, InVisibility=Select(Option 0=Hidden, Option 1=SelfHitTestInvisible, Index=IsValid(Object=@n2)))
  ]}
 HandlePromptChange {nodes [
   n1 = HandlePromptChange!() -> (then, ShouldReset:bool=~) {is=title K2Node_FunctionEntry}
   > n2 = Set Promt(Promt:Text=SelectText(A="LOCTEXT(\"RESET\")", B="LOCTEXT(\"EDIT\")", bPickA=@n1)) -> (Output_Get:Text)
   > Set Text(Text=@n2, self=Get TextEditLabel() -> (TextEditLabel:Common Text Object Reference))
   > Set Text(Text=@n2, self=Get TextEdit() -> (TextEdit:Common Text Object Reference))
   > Set Text(Text=@n2, self=Get TextBuild-Edit() -> (TextBuild-Edit:Common Text Object Reference))
  ]}
 HandleQuickEditMeshPreview {nodes [
   n1 = HandleQuickEditMeshPreview!() -> (then, bIsMirrored:bool=~, Rotation:Byte, SelectedItemName:Name) {is=title K2Node_FunctionEntry}
   n2 = SelectString(A=_mirrored, bPickA=Get bIsMirrored() -> (bIsMirrored:bool))
   n3 = Append(
    A=Append(A="To String (Name)"(InName=@n1.SelectedItemName), B=_)
    B="To String (Byte)"!(InByte:Byte=Get Rotation() -> (Rotation:Byte=0)) -> (ReturnValue:String)
   )
   > Set ItemName(ItemName:Name=String To Name(InString=Append(A=@n3, B=@n2))) -> (Output_Get:Name=None)
   n4 = Find(
    TargetMap:Map of Names to Texture Object References=Get MeshPreviewMap_Wood() -> (MeshPreviewMap_Wood:Map of Names to Texture Object References)
    Key:"Name (by ref)"=Get ItemName() -> (ItemName:Name=None)
   ) -> (Value:Texture Object Reference)
   > n5 = Branch(Condition=@n4.ReturnValue)
   > n6 = Set FoundPreviewImage(FoundPreviewImage:Texture Object Reference=@n4) -> (Output_Get:Texture Object Reference)
   n7 = GetDynamicMaterial(execute=[@n13.then, @n6.then, @n11.then], self=Get Mesh_Preview() -> (Mesh_Preview:Image Object Reference))
   n8 = Get FoundPreviewImage() -> (FoundPreviewImage:Texture Object Reference)
   > SetTextureParameterValue(self=@n7, ParameterName=Mesh, Value=@n8)
   n9 = Find(
    TargetMap=Get MeshPreviewMap_Stone() -> (MeshPreviewMap_Stone:Map of Names to Texture Object References)
    Key=Get ItemName() -> (ItemName=None)
   )
   n10 = Branch(execute=@n5.else, Condition=@n9.ReturnValue)
   > n11 = Set FoundPreviewImage(FoundPreviewImage=@n9)
   n12 = Find(
    TargetMap=Get MeshPreviewMap_Metal() -> (MeshPreviewMap_Metal:Map of Names to Texture Object References)
    Key=Get ItemName() -> (ItemName=None)
   )
   Branch(execute=@n10.else, Condition=@n12.ReturnValue)
   > n13 = Set FoundPreviewImage(FoundPreviewImage=@n12)
   n14 = GetCurrentFocusedBuilding(self=Get FortHUDContext())
   Select(
    Option 0:Texture Object Reference=@n8
    Option 1:Texture Object Reference=/Game/UI/InGame/HUD/SimpleEditPreviews/PBWA_W1_StairF_D
    Index=AND Boolean(
     A=Get bPlayerFacingFromSide() -> (bPlayerFacingFromSide:bool)
     B=IsA!(Object:Object Reference=@n14, ObjectClass:Object Class Reference=&BuildingStairs) -> (ReturnValue:bool)
    )
   ) -> (ReturnValue:Texture Object Reference)
  ]}
 HandleQuickEditOptionChanged {nodes [
   n1 = HandleQuickEditOptionChanged!() -> (then, Building Class Name:Name, Building Edit Class Name:Name, Is Mirrored:bool=~, Rotations:Byte) {is=title K2Node_FunctionEntry}
   > Set CurrentQuickEditClass(CurrentQuickEditClass:Name=@n1) -> (Output_Get:Name=None)
   n2 = Find(
    TargetMap:Map of Names to Material Instance Object References=Get QuickEditMap() -> (QuickEditMap:Map of Names to Material Instance Object References)
    Key=@n1
   ) -> (Value:Material Instance Object Reference)
   > Set bHasValidQuickEditOption(bHasValidQuickEditOption:bool=@n2.ReturnValue) -> (Output_Get:bool)
   n3 = Get Image_QuickEditPreview() -> (Image_QuickEditPreview:Image Object Reference)
   n4 = Get Image_QuickEditPreview_2() -> (Image_QuickEditPreview_2:Image Object Reference)
   n5 = Get Image_QuickEditPreview_Touch() -> (Image_QuickEditPreview_Touch:Image Object Reference)
   n6 = Get Image_QuickEditPreview_1() -> (Image_QuickEditPreview_1:Image Object Reference)
   > SetBrushFromMaterial(
    self=[@n3, @n4, @n5, @n6]
    Material=Select(
     Option 0:Material Interface Object Reference=/Game/UI/InGame/HUD/Textures/M_UI_QuickEdit_0
     Option 1:Material Interface Object Reference=@n2
     Index=@n2.ReturnValue
    ) -> (ReturnValue:Material Interface Object Reference)
   )
   > SetRenderScale!(
    execute
    self:Widget Object Reference=[@n5, @n6, @n4, @n3]
    Scale:Vector 2D Structure=Select(Option 0:Vector 2D Structure=(X=1,Y=1), Option 1:Vector 2D Structure=(X=-1,Y=1), Index=@n1.Is Mirrored) -> (ReturnValue:Vector 2D Structure)
   ) -> (then)
   > SetRenderTransformAngle!(
    execute
    self:Widget Object Reference=[@n5, @n6, @n4, @n3]
    Angle:float=float * float(A=90, B="To Float (Byte)"!(InByte:Byte=@n1.Rotations) -> (ReturnValue:double))
   ) -> (then)
   > UpdateQuickEditPreviewVisibility()
   > HandleQuickEditMeshPreview!(
    execute
    self
    bIsMirrored:bool=Get Is Mirrored() -> (Is Mirrored:bool)
    Rotation:Byte=Get Rotations() -> (Rotations:Byte=0)
    SelectedItemName:Name=Get CurrentQuickEditClass() -> (CurrentQuickEditClass:Name=None)
   ) -> (then) {is=title K2Node_CallFunction}
   > Switch State!(
    execute
    self:WBP Quick Edit Grid Object Reference=Get WBP_QuickEditGrid()
    SelectedItemName:Name=Get CurrentQuickEditClass() -> (CurrentQuickEditClass=None)
    IsMirrored:bool=Get Is Mirrored()
    Rotation:Byte=Get Rotations() -> (Rotations=0)
   ) -> (then)
   > HandlePromptChange!(
    execute
    self
    ShouldReset:bool=Contains Item(
     TargetArray=Get MashPreviewResetNames() -> (MashPreviewResetNames:Array of Names)
     ItemToFind=Get CurrentQuickEditClass() -> (CurrentQuickEditClass=None)
    )
   ) -> (then) {is=title K2Node_CallFunction}
  ]}
 HandleQuickbarFocusChanged {nodes [
   HandleQuickbarFocusChanged!() -> (then, Quickbar:EFortQuickBars Enum=Primary, QuickbarSlot:int) {is=title K2Node_FunctionEntry}
   > HandleBuildingFocused!(execute, self, FocusedBuilding:Building Actor Object Reference=GetCurrentFocusedBuilding(self=Get FortHUDContext())) -> (then) {is=title K2Node_CallFunction}
   > RefreshBuildingInfo()
   n1 = Get FortInventoryContext()
   > Branch(Condition="Equal (Enum)"(
     A:EFortQuickBars Enum=GetQuickbarFocus!(self:Fort Inventory Context Object Reference=@n1) -> (OutQuickBar:EFortQuickBars Enum=Primary, OutSlot:int=0, OutSecondarySlot:int=0, OutPreviousFocusedSlot:int=0)
     B:EFortQuickBars Enum=Secondary
    ))
   > n2 = Cast To FortPlayerController(Object=GetOwningPlayer())
   > RefreshTrapUIControls(self:Fort Player Controller Object Reference=@n2)
  ]}
 HandleSimpleBuild {nodes [
   n1 = HandleSimpleBuild!() -> (then, Enabled?:bool=~) {is=title K2Node_FunctionEntry}
   > SetVisibility(
    self=Get RotateBox() -> (RotateBox:Horizontal Box Object Reference)
    InVisibility=Select(Option 0:ESlateVisibility Enum=SelfHitTestInvisible, Option 1:ESlateVisibility Enum=Collapsed, Index=@n1) -> (ReturnValue:ESlateVisibility Enum)
   )
  ]}
 HandleTrapDurabilityChanged {nodes [
   n1 = HandleTrapDurabilityChanged!() -> (then, CurrentDurability:float=~, MaxDurability:float=~) {is=title K2Node_FunctionEntry}
   n2 = "Get [Widget] Building HP && Trap Bars"()
   > Update Current Durability Value!(execute, self:Building Bar Object Reference=@n2, CurrentDurability:double=@n1) -> (then)
   > Update Max Durability Value!(execute, self:Building Bar Object Reference=@n2, Max:double=@n1.MaxDurability) -> (then)
  ]}
 HideHealth {nodes [
   HideHealth() {is=title K2Node_FunctionEntry}
   > SetVisibility(self="Get [Widget] Building HP && Trap Bars"(), InVisibility=Hidden)
  ]}
 IsCurrentBuildingBeingRepaired {nodes [
   IsCurrentBuildingBeingRepaired() {is=title K2Node_FunctionEntry}
   n1 = Break Fort Focused Building Info!(FortFocusedBuildingInfo=Get BuildingInfo()) -> (
    bIsInteractable:bool=~
    bCanBePlayerEdited=~
    IndicatorRelativeLocation:Vector=~
    HealthDisplayRule
    MaxHealth=~
    bIsAnyTrapAttached=~
    bIsTrapAttachedFacingPlayer=~
    bIsPreviewTrapAttached=~
    AttachedTrapMaxDurability:float=~
    InteractionType
    InteractionCost:int
    BuildingMaterial:EFortResourceType Enum
    QuickEditClass
    bIsPlayerBuilt=~
   )
   > Return Node!(
    execute
    ReturnValue:bool=AND Boolean(A=@n1, B="Equal (Enum)"(A:EFortBuildingInteraction Enum=@n1.InteractionType, B:EFortBuildingInteraction Enum=Repair))
   )
  ]}
 IsCurrentWeaponTheMoveTool {nodes [
   IsCurrentWeaponTheMoveTool() {is=title K2Node_FunctionEntry}
   > n1 = Cast To FortPawn!(
    execute
    Object=Get Controlled Pawn!(self:Controller Object Reference=GetOwningPlayer()) -> (ReturnValue:Pawn Object Reference) {is=title K2Node_CallFunction}
   ) -> (then, CastFailed, AsFort Pawn:Fort Pawn Object Reference) {is=title K2Node_DynamicCast}
   > n2 = Cast To FortCreativeMoveTool!(execute, Object=Get CurrentWeapon!(self:Fort Pawn Object Reference=@n1) -> (CurrentWeapon:Fort Weapon Object Reference)) -> (then, CastFailed, AsFort Creative Move Tool:Fort Creative Move Tool Object Reference) {is=title K2Node_DynamicCast}
   > Return Node!(execute, Result:bool=true)
   Return Node(execute=@n2.CastFailed, Result=false)
  ]}
 OnFocusedBuildingDynamicDataReceived {nodes [
   n1 = OnFocusedBuildingDynamicDataReceived!() -> (then, DynamicData:Focused Building Dynamic Data Object Reference) {is=title K2Node_FunctionEntry}
   > n2 = Is Valid(InputObject=@n1)
   > n3 = Branch(Condition=IsVisible(self="Get [Widget] Building HP && Trap Bars"()))
   > n4 = Sequence()
   > n5 = Cast To BuildingInteractionPrompts!(execute, Object=@n1) -> (then, CastFailed, AsBuilding Interaction Prompts:Building Interaction Prompts Object Reference) {is=title K2Node_DynamicCast}
   > "SetText (Text)"(
    self=Get DynamicInteractionPrompt() -> (DynamicInteractionPrompt:Common Text Object Reference)
    InText=Get Prompt!(self:Building Interaction Prompts Object Reference=@n5) -> (Prompt:Text)
   )
   > SetBrushFromTexture!(
    execute
    self:Image Object Reference=Get DynamicInteractionTexture() -> (DynamicInteractionTexture:Image Object Reference)
    Texture:Texture 2D Object Reference=Get Texture!(self:Building Interaction Prompts Object Reference=@n5) -> (Texture:Texture 2D Object Reference)
    bMatchSize:bool=false
   ) -> (then)
   n6 = Get SwitcherInteractions() -> (SwitcherInteractions:Common Widget Switcher Legacy Object Reference)
   SetVisibility(execute=@n4.then_1, self=@n6, InVisibility=HitTestInvisible)
   > SetActiveWidget(self:Widget Switcher Object Reference=@n6, Widget=Get DynamicInteraction() -> (DynamicInteraction:Size Box Object Reference))
   SetVisibility(execute=[@n3.else, @n2.Is Not Valid], self=Get SwitcherInteractions(), InVisibility=Hidden)
  ]}
 OnInputTypeChanged {nodes [
   n1 = OnInputTypeChanged!() -> (then, Input Type:ECommonInputType Enum) {is=title K2Node_FunctionEntry}
   n2 = Get Image_QuickEditPreview_Touch()
   n3 = Get SwitcherInteractions()
   > SetActiveWidget(
    self=Get Input() -> (Input:Widget Switcher Object Reference)
    Widget=Select!(
     MouseAndKeyboard:Widget Object Reference=@n3
     Gamepad:Widget Object Reference=@n3
     Touch:Widget Object Reference=@n2
     Count:Widget Object Reference=@n2
     Index:ECommonInputType Enum=@n1
    ) -> (ReturnValue:Widget Object Reference)
   )
   n4 = GetSubGame!() -> (ReturnValue:ESubGame Enum=Campaign)
   n5 = Get Input Type() -> (Input Type:ECommonInputType Enum)
   > Branch(Condition=AND Boolean(
     A="Equal (Enum)"(A:ECommonInputType Enum=@n5, B:ECommonInputType Enum=Gamepad)
     B="Equal (Enum)"(A:ESubGame Enum=@n4, B:ESubGame Enum=Campaign)
    ))
   > SetForcedHoldKeybindStatus!(
    execute
    self:Fort Keybind Widget Object Reference=Get KeybindUpgradeRepair() -> (KeybindUpgradeRepair:Keybind Widget Dark Object Reference)
    InForcedHoldKeybindStatus:EFortKeybindForcedHoldStatus Enum=ForcedHold
   ) -> (then)
  ]}
 RefreshBuildingInfo {nodes [
   RefreshBuildingInfo() {is=title K2Node_FunctionEntry}
   n1 = Break Fort Focused Building Info(FortFocusedBuildingInfo=Get BuildingInfo()) -> (
    bIsInteractable=~
    bCanBePlayerEdited=~
    IndicatorRelativeLocation=~
    MaxHealth=~
    bIsAnyTrapAttached=~
    bIsTrapAttachedFacingPlayer=~
    bIsPreviewTrapAttached=~
    AttachedTrapMaxDurability=~
    bIsPlayerBuilt=~
   )
   > UpdateBuildingInfo_InteractionWidgets!(
    execute
    self
    Interactable:bool=@n1
    CanBePlayerEdited:bool=@n1.bCanBePlayerEdited
    IsAnyTrapAttached:bool=@n1.bIsAnyTrapAttached
    IsTrapAttachedFacingPlayer:bool=@n1.bIsTrapAttachedFacingPlayer
    IsPreviewTrapAttached:bool=@n1.bIsPreviewTrapAttached
    InteractionType:EFortBuildingInteraction Enum=@n1.InteractionType
    InteractionCost:int=@n1.InteractionCost
    BuildingMaterial:EFortResourceType Enum=@n1.BuildingMaterial
   ) -> (then)
  ]}
 UpdateBuildingInfo {nodes [
   n1 = UpdateBuildingInfo!() -> (then, NewBuildingInfo:"Fort Focused Building Info Structure (by ref)") {is=title K2Node_FunctionEntry}
   > n2 = Set BuildingInfo(BuildingInfo=@n1)
   n3 = Break Fort Focused Building Info(FortFocusedBuildingInfo=@n2) -> (
    bIsInteractable=~
    bCanBePlayerEdited=~
    IndicatorRelativeLocation=~
    MaxHealth=~
    bIsAnyTrapAttached=~
    bIsTrapAttachedFacingPlayer=~
    bIsPreviewTrapAttached=~
    AttachedTrapMaxDurability=~
    bIsPlayerBuilt=~
   )
   n4 = "Get [Widget] Building HP && Trap Bars"()
   > Update Max HP Value!(execute, self:Building Bar Object Reference=@n4, Max:double=@n3.MaxHealth) -> (then)
   > Update Max Durability Value(self=@n4, Max=@n3.AttachedTrapMaxDurability)
   > Set Display Mode!(execute, self:Building Bar Object Reference=@n4, InDisplayMode:EFortBuildingInteraction Enum=@n3.InteractionType) -> (then) {is=title K2Node_CallFunction}
   > Set RelativeLocation(RelativeLocation:Vector=@n3.IndicatorRelativeLocation) -> (Output_Get:Vector)
   > UpdateBuildingInfo_InteractionWidgets(
    Interactable=@n3
    CanBePlayerEdited=@n3.bCanBePlayerEdited
    IsAnyTrapAttached=@n3.bIsAnyTrapAttached
    IsTrapAttachedFacingPlayer=@n3.bIsTrapAttachedFacingPlayer
    IsPreviewTrapAttached=@n3.bIsPreviewTrapAttached
    InteractionType=@n3.InteractionType
    InteractionCost=@n3.InteractionCost
    BuildingMaterial=@n3.BuildingMaterial
   )
   > SetVisibility(
    self="Get [Widget] Building HP && Trap Bars"()
    InVisibility=Select!(
     Never:ESlateVisibility Enum=Hidden
     Allowed:ESlateVisibility Enum=Select!(
      None:ESlateVisibility Enum=Hidden
      Build:ESlateVisibility Enum=Hidden
      Repair:ESlateVisibility Enum=SelfHitTestInvisible
      Upgrade:ESlateVisibility Enum=SelfHitTestInvisible
      Edit:ESlateVisibility Enum=SelfHitTestInvisible
      BeingModified:ESlateVisibility Enum=SelfHitTestInvisible
      ConfirmEdit:ESlateVisibility Enum=Hidden
      Creative:ESlateVisibility Enum=Hidden
      Index:EFortBuildingInteraction Enum=@n3.InteractionType
     ) -> (ReturnValue)
     Always:ESlateVisibility Enum=SelfHitTestInvisible
     Index:EFortBuildingHealthDisplayRule Enum=@n3.HealthDisplayRule
    ) -> (ReturnValue:ESlateVisibility Enum)
   )
   > Clear and Invalidate Timer by Handle(Handle=Get HealthFadeTimerHandle())
  ]}
 UpdateBuildingInfo_InteractionWidgets {nodes [
   n1 = Get BuildingMaterial() -> (BuildingMaterial:EFortResourceType Enum=Wood)
   n2 = PrintText(InText=Format Text!(Format="LOCTEXT(\"Hello {0}\")", 0:String=Enum to String!(Enumerator:EFortResourceType Enum=@n1) -> (ReturnValue:String)) -> (Result))
   n3 = Get TextInteractionCostWithRotate() -> (TextInteractionCostWithRotate:Common Text Object Reference)
   n4 = Get TextInteractionCost() -> (TextInteractionCost:Common Text Object Reference)
   "SetText (Text)"(execute=[@n2.then, @n58.then], self=[@n3, @n4], InText="To Text (Integer)"(Value=Get InteractionCost() -> (InteractionCost:int=0)))
   > Set TopTrapActive(TopTrapActive:bool=false) -> (Output_Get:bool)
   UpdateBuildingInfo_InteractionWidgets!() -> (
    then
    Interactable:bool=~
    CanBePlayerEdited:bool=~
    IsAnyTrapAttached:bool=~
    IsTrapAttachedFacingPlayer:bool=~
    IsPreviewTrapAttached:bool=~
    InteractionType:EFortBuildingInteraction Enum=None
    InteractionCost:int
    BuildingMaterial:EFortResourceType Enum=Wood
   ) {is=title K2Node_FunctionEntry}
   n5 = Get FortInventoryContext()
   > Is Valid(InputObject=@n5)
   n6 = GetQuickbarFocus(self=@n5) -> (OutQuickBar=Primary, OutSlot=0, OutSecondarySlot=0, OutPreviousFocusedSlot=0)
   > n7 = Set PlaceTrapAvailable(PlaceTrapAvailable:bool=AND Boolean(A=Get IsPreviewTrapAttached() -> (IsPreviewTrapAttached:bool), B="Equal (Enum)"(A:EFortQuickBars Enum=@n6, B:EFortQuickBars Enum=Secondary))) -> (Output_Get:bool)
   n8 = Get Interactable() -> (Interactable:bool)
   n9 = Select(
    None=Select!(Option 0:ESlateVisibility Enum=Hidden, Option 1:ESlateVisibility Enum=SelfHitTestInvisible, Index:bool=OR Boolean(A=@n7, B=@n8)) -> (ReturnValue)
    Build=SelfHitTestInvisible
    Repair=SelfHitTestInvisible
    Upgrade=SelfHitTestInvisible
    Edit=SelfHitTestInvisible
    BeingModified=SelfHitTestInvisible
    ConfirmEdit=SelfHitTestInvisible
    Creative=Visible
    Index=Get InteractionType() -> (InteractionType:EFortBuildingInteraction Enum=None)
   )
   > SetVisibility(self=Get SwitcherInteractions(), InVisibility=@n9)
   > n10 = Branch(Condition="Not Equal (Enum)"!(A:ESlateVisibility Enum=@n9, B:ESlateVisibility Enum=Hidden) -> (ReturnValue:bool))
   n11 = Get IsAnyTrapAttached() -> (IsAnyTrapAttached:bool)
   > Set bIsEditAvailable(bIsEditAvailable:bool=AND Boolean(A=Get CanBePlayerEdited() -> (CanBePlayerEdited:bool), B=NOT Boolean(A=@n11))) -> (Output_Get:bool)
   > UpdateQuickEditPreviewVisibility()
   > n12 = Cast To BuildingSMActor!(execute, Object=GetIndicatedActor()) -> (then, CastFailed, AsBuilding Static Mesh:Building Static Mesh Object Reference) {is=title K2Node_DynamicCast}
   n13 = Get IsTrapAttachedFacingPlayer() -> (IsTrapAttachedFacingPlayer:bool)
   n14 = Get FortInventoryContext()
   n15 = NOT Boolean(A=Get Interactable())
   n16 = Get bIsEditAvailable() -> (bIsEditAvailable:bool)
   n17 = Get CommonInputSubsystem!() -> (ReturnValue:Common Input Object Reference) {is=title K2Node_GetBlueprintContextEx}
   n18 = AND Boolean(
    A=IsPlayerBuilt!(self:Building Actor Object Reference=@n12) -> (ReturnValue:bool)
    B="Not Equal (Enum)"(A:ECommonInputType Enum=GetCurrentInputType(self=@n17), B:ECommonInputType Enum=Gamepad)
   )
   n19 = "Equal (Enum)"(
    A=GetQuickbarFocus(self=Get FortInventoryContext()) -> (OutQuickBar=Primary, OutSlot=0, OutSecondarySlot=0, OutPreviousFocusedSlot=0)
    B=Secondary
   )
   > n20 = Set TrapAvailable(TrapAvailable:bool=AND Boolean!(
     A=IsTrapAvailableForBuilding!(self:Fort Inventory Context Object Reference=@n14, Actor:Building Static Mesh Object Reference=@n12) -> (ReturnValue:bool)
     B=OR Boolean(A=@n16, B=@n15)
     C:bool=NOT Boolean(A=@n13)
     D:bool=OR Boolean(A=@n19, B=@n18)
    ) -> (ReturnValue)) -> (Output_Get:bool)
   n21 = Get TrapAvailable() -> (TrapAvailable:bool)
   n22 = Get InteractionType() -> (InteractionType=None)
   n23 = "Equal (Enum)"(A:EFortBuildingInteraction Enum=@n22, B:EFortBuildingInteraction Enum=Edit)
   n24 = "Equal (Enum)"(A=@n22, B=BeingModified)
   n25 = Branch(execute=[@n20.then, @n71.then], Condition=AND Boolean!(A=OR Boolean(A=@n23, B=@n24), B=NOT Boolean(A=@n21)) -> (ReturnValue))
   n26 = Get SwitcherInteractions()
   > SetActiveWidget(self=@n26, Widget="Get [Size Box] Edit (standalone)"() -> ("[Size Box] Edit (standalone)":Size Box Object Reference))
   > SetVisibility(self="Get [Size Box] Edit (standalone)"(), InVisibility=Select(Option 0=Hidden, Option 1=HitTestInvisible, Index=Get bIsEditAvailable()))
   > Branch(Condition=Get bIsEditAvailable())
   > SetKeyBindWidgetBoundAction!(
    execute
    self:Fort Building Info Indicator Widget Object Reference
    FortKeyBindWidget:Fort Keybind Widget Object Reference=Get KeybindEditstandalone() -> (KeybindEditstandalone:Keybind Widget Dark Object Reference)
    ActionName:Name=GetPerformBuildingEditInteractionActionName!(self:Fort Global UIContext Object Reference=Get FortGlobalUIContext(), OverrideInputType:ECommonInputType Enum=Count) -> (ReturnValue:Name=None)
    InteractionType:EFortBuildingInteraction Enum=Get InteractionType() -> (InteractionType=None)
   ) -> (then)
   n27 = Get SizeBoxTriOption() -> (SizeBoxTriOption:Size Box Object Reference)
   n28 = Select!(
    None:Widget Object Reference=Get SizeBoxTrapstandalone() -> (SizeBoxTrapstandalone:Size Box Object Reference)
    Build:Widget Object Reference=@n27
    Repair:Widget Object Reference=@n27
    Upgrade:Widget Object Reference=@n27
    Edit:Widget Object Reference=@n27
    BeingModified:Widget Object Reference=@n27
    ConfirmEdit:Widget Object Reference="Get [Size Box] Confirm"() -> ("[Size Box] Confirm":Size Box Object Reference)
    Creative:Widget Object Reference
    Index:EFortBuildingInteraction Enum=@n22
   ) -> (ReturnValue:Widget Object Reference)
   SetActiveWidget(execute=@n25.else, self=@n26, Widget=@n28)
   > n29 = Branch(Condition="Equal (Object)"(A=@n28, B=Get SizeBoxTrapstandalone()) {is=title K2Node_CallFunction})
   > SetKeyBindWidgetBoundAction(
    FortKeyBindWidget=Get KeybindPlaceTrapstandalone() -> (KeybindPlaceTrapstandalone:Keybind Widget Dark Object Reference)
    ActionName=GetTrapConfirmActionName!(self:Fort Global UIContext Object Reference=Get FortGlobalUIContext(), OverrideInputType:ECommonInputType Enum=Count) -> (ReturnValue:Name=None)
    InteractionType=Get InteractionType() -> (InteractionType=None)
   )
   > UpdateTrapWidgets!(
    execute
    self
    Place Trap Available:bool=Get PlaceTrapAvailable() -> (PlaceTrapAvailable:bool)
    Interaction Type:EFortBuildingInteraction Enum=Get InteractionType() -> (InteractionType=None)
    Keybind:Keybind Widget Dark Object Reference=Get KeybindPickTrapstandalone() -> (KeybindPickTrapstandalone:Keybind Widget Dark Object Reference)
    Text:Common Text Object Reference=Get TextPickTrapLabel() -> (TextPickTrapLabel:Common Text Object Reference)
   ) -> (then)
   n30 = Switch on EFortResourceType!(execute=@n29.else, Selection:EFortResourceType Enum=Get BuildingMaterial() -> (BuildingMaterial=Wood)) -> (Wood:Exec, Stone:Exec, Metal:Exec, Permanite:Exec, GoldCurrency:Exec, Ingredient:Exec, None:Exec) {is=title K2Node_SwitchEnum}
   n31 = Switch on EFortBuildingInteraction!(
    execute=[@n30.Metal, @n30.Permanite, @n30.Stone, @n30.Wood]
    Selection:EFortBuildingInteraction Enum=Get InteractionType() -> (InteractionType=None)
   ) -> (None:Exec, Build:Exec, Repair:Exec, Upgrade:Exec, Edit:Exec, BeingModified:Exec, ConfirmEdit:Exec, Creative:Exec) {is=title K2Node_SwitchEnum}
   n32 = "Get [Horizontal Box] Edit"() -> ("[Horizontal Box] Edit":Horizontal Box Object Reference)
   n33 = Get SwitcherSideOption2() -> (SwitcherSideOption2:Common Widget Switcher Legacy Object Reference)
   SetActiveWidget(execute=[@n31.BeingModified, @n31.Edit], self=@n33, Widget=@n32)
   n34 = Select!(Option 0:ESlateVisibility Enum=Hidden, Option 1:ESlateVisibility Enum=SelfHitTestInvisible, Index:bool=Get bIsEditAvailable()) -> (ReturnValue:ESlateVisibility Enum)
   > SetVisibility(self=@n32, InVisibility=@n34)
   n35 = Get SwitcherSideOption1() -> (SwitcherSideOption1:Common Widget Switcher Legacy Object Reference)
   > SetActiveWidget(self=@n35, Widget=Get TrapOptions() -> (TrapOptions:Vertical Box Object Reference))
   > n36 = Sequence!(execute) -> (then_0, then_1, then_2:Exec)
   n37 = Get TopTrapActive() -> (TopTrapActive:bool)
   n38 = Select(Option 0=Hidden, Option 1=SelfHitTestInvisible, Index=AND Boolean(A=Get TrapAvailable(), B=NOT Boolean(A=@n37)))
   SetVisibility(
    execute=[@n36.then_2, @n62.then_2, @n44.then_2, @n61.then_2]
    self=Get VerticalBoxLowerOption() -> (VerticalBoxLowerOption:Vertical Box Object Reference)
    InVisibility=@n38
   )
   > Branch(Condition="Not Equal (Enum)"(A:ESlateVisibility Enum=@n38, B:ESlateVisibility Enum=Hidden))
   > UpdateTrapWidgets(
    Place Trap Available=Get PlaceTrapAvailable()
    Interaction Type=Get InteractionType() -> (InteractionType=None)
    Keybind=Get KeybindTrapLower() -> (KeybindTrapLower:Keybind Widget Dark Object Reference)
    Text=Get TextPickTrapBottom() -> (TextPickTrapBottom:Common Text Object Reference)
   )
   n39 = Select(Option 0=Collapsed, Option 1=SelfHitTestInvisible, Index=Get PlaceTrapAvailable())
   SetVisibility(execute=@n36.then_1, self=Get HorizontalBoxPlaceTrap() -> (HorizontalBoxPlaceTrap:Horizontal Box Object Reference), InVisibility=@n39)
   > n40 = Switch on ESlateVisibility!(execute, Selection:ESlateVisibility Enum=@n39) -> (Visible:Exec, Collapsed:Exec, Hidden:Exec, HitTestInvisible:Exec, SelfHitTestInvisible:Exec) {is=title K2Node_SwitchEnum}
   n41 = Branch(
    execute=[@n40.HitTestInvisible, @n40.SelfHitTestInvisible, @n40.Visible]
    Condition="Equal (Enum)"(A:ESlateVisibility Enum=@n39, B:ESlateVisibility Enum=Collapsed)
   )
   Branch(execute=[@n42.then, @n41.then, @n40.Collapsed, @n40.Hidden], Condition=Get TrapAvailable())
   > UpdateTrapWidgets(
    Place Trap Available=Get PlaceTrapAvailable()
    Interaction Type=Get InteractionType() -> (InteractionType=None)
    Keybind=Get KeybindCycleTrapLeft() -> (KeybindCycleTrapLeft:Keybind Widget Dark Object Reference)
    Text=Get TextPickTrap() -> (TextPickTrap:Common Text Object Reference)
   )
   > Set TopTrapActive(TopTrapActive=true)
   n42 = SetKeyBindWidgetBoundAction(
    execute=@n41.else
    FortKeyBindWidget=Get KeybindPlaceTrapLeft() -> (KeybindPlaceTrapLeft:Keybind Widget Dark Object Reference)
    ActionName=GetTrapConfirmActionName(self=Get FortGlobalUIContext(), OverrideInputType=Count) -> (ReturnValue=None)
    InteractionType=Get InteractionType() -> (InteractionType=None)
   )
   Branch(execute=[@n36.then_0, @n44.then_1, @n61.then_1], Condition=Get bIsEditAvailable())
   > SetKeyBindWidgetBoundAction(
    FortKeyBindWidget=Get KeybindEdit() -> (KeybindEdit:Keybind Widget Dark Object Reference)
    ActionName=GetPerformBuildingEditInteractionActionName(self=Get FortGlobalUIContext(), OverrideInputType=Count) -> (ReturnValue=None)
    InteractionType=Get InteractionType() -> (InteractionType=None)
   )
   > Set TopTrapActive(TopTrapActive=false)
   SetActiveWidget(execute=@n31.Repair, self=@n33, Widget=@n32)
   > SetVisibility(self=@n32, InVisibility=@n34)
   n43 = Get HorizontalBoxUpgradeRepair() -> (HorizontalBoxUpgradeRepair:Horizontal Box Object Reference)
   > SetActiveWidget(self=@n35, Widget=@n43)
   > n44 = Sequence()
   n45 = Get FortGlobalUIContext()
   n46 = Get InteractionType() -> (InteractionType=None)
   n47 = SetKeyBindWidgetBoundAction(
    execute=[@n44.then_0, @n61.then_0]
    FortKeyBindWidget=Get KeybindUpgradeRepair()
    ActionName=Select(
     Option 0:Name=GetFireActionName!(self:Fort Global UIContext Object Reference=@n45) -> (ReturnValue:Name=None)
     Option 1:Name=GetPerformBuildingImprovementInteractionActionName!(self:Fort Global UIContext Object Reference=@n45, OverrideInputType:ECommonInputType Enum=Count) -> (ReturnValue:Name=None)
     Index="Not Equal (Enum)"(A:EFortBuildingInteraction Enum=@n46, B:EFortBuildingInteraction Enum=Build)
    ) -> (ReturnValue:Name)
    InteractionType=Get InteractionType() -> (InteractionType=None)
   )
   n48 = Switch on EFortBuildingInteraction(execute=[@n63.then, @n47.then], Selection=Get InteractionType() -> (InteractionType=None))
   n49 = Get TextInteractionName() -> (TextInteractionName:Common Text Object Reference)
   n50 = "SetText (Text)"(execute=@n48.Build, self=@n49, InText="LOCTEXT(\"BUILD\")")
   n51 = Get FortInventoryContext()
   n52 = Cast To FortPlayerController!(Object=GetOwningPlayer()) -> (AsFort Player Controller, bSuccess:bool=~)
   n53 = AND Boolean(A=Get bBuildFree!(self:Fort Player Controller Object Reference=@n52) -> (bBuildFree:bool), B=@n52.bSuccess)
   n54 = Get InteractionCost() -> (InteractionCost=0)
   n55 = Select(
    Option 0:Linear Color Structure=(R=1,G=1,B=1,A=1)
    Option 1:Linear Color Structure=(R=0.8118,G=0.1451,B=0.102,A=1)
    Index="integer > integer"(
     A=Select(Option 0:int=@n54, Option 1:int=0, Index=@n53) -> (ReturnValue:int)
     B=GetResourceCount!(self:Fort Inventory Context Object Reference=@n51, ResourceType:EFortResourceType Enum=@n1) -> (ReturnValue:int=0)
    ) {is=title K2Node_CallFunction}
   ) -> (ReturnValue:Linear Color Structure)
   SetColorAndOpacity(execute=[@n50.then, @n60.then, @n59.then], self=[@n4, @n3, @n49], InColorAndOpacity=Make Slate Color(SpecifiedColor=@n55))
   n56 = "Get [Image] Resource Type Icon"() -> ("[Image] Resource Type Icon":Fort Lazy Image Object Reference)
   n57 = Get ImageResourceTypeIconwithRotate() -> (ImageResourceTypeIconwithRotate:Fort Lazy Image Object Reference)
   > SetColorAndOpacity(self:Image Object Reference=[@n56, @n57], InColorAndOpacity:Linear Color Structure=@n55)
   > n58 = SetBrushFromItemDefinition!(
    execute
    self:Fort Lazy Image Object Reference=[@n56, @n57]
    ItemDefinition:Item Definition Base Object Reference=GetResourceItemDefinition!(self:Fort Inventory Context Object Reference=@n51, ResourceType:EFortResourceType Enum=@n1) -> (ReturnValue:Fort Resource Item Definition Object Reference)
   ) -> (then)
   n59 = "SetText (Text)"(execute=@n48.Repair, self=@n49, InText="LOCTEXT(\"REPAIR\")")
   n60 = "SetText (Text)"(execute=@n48.Upgrade, self=@n49, InText="LOCTEXT(\"UPGRADE\")")
   SetActiveWidget(execute=@n31.Upgrade, self=@n33, Widget=@n32)
   > SetVisibility(self=@n32, InVisibility=@n34)
   > SetActiveWidget(self=@n35, Widget=@n43)
   > n61 = Sequence()
   SetActiveWidget(execute=@n31.Build, self=@n33, Widget="Get [Vertical Box] Build Subset"() -> ("[Vertical Box] Build Subset":Vertical Box Object Reference))
   > SetActiveWidget(self=@n35, Widget=Get VerticalBoxBuildWithRotate() -> (VerticalBoxBuildWithRotate:Vertical Box Object Reference))
   > n62 = Sequence()
   > HandleQuickEditOptionChanged!(execute, self, Building Class Name:Name=None, Building Edit Class Name:Name=None, Is Mirrored:bool=false, Rotations:Byte=0) -> (then) {is=title K2Node_CallFunction}
   > SetKeyBindWidgetBoundAction(
    FortKeyBindWidget=Get KeybindBuild() -> (KeybindBuild:Keybind Widget Dark Object Reference)
    ActionName=GetBuildConfirmActionName!(self:Fort Global UIContext Object Reference=Get FortGlobalUIContext(), OverrideInputType:ECommonInputType Enum=Count) -> (ReturnValue:Name=None)
    InteractionType=Get InteractionType() -> (InteractionType=None)
   )
   > n63 = SetKeyBindWidgetBoundAction(
    FortKeyBindWidget=Get KeybindRotate() -> (KeybindRotate:Keybind Widget Dark Object Reference)
    ActionName=GetRotatePrimitiveClockwiseActionName!(self:Fort Global UIContext Object Reference=Get FortGlobalUIContext(), OverrideInputType:ECommonInputType Enum=Count) -> (ReturnValue:Name=None)
    InteractionType=Get InteractionType() -> (InteractionType=None)
   )
   n64 = NOT Boolean(A=Get bIsEditAvailable())
   n65 = GetDisablePreEditsWhenPlacingBuildingState!(self:Fort Settings Context Object Reference=Get FortSettingsContext!() -> (ReturnValue:Fort Settings Context Object Reference) {is=title K2Node_GetBlueprintContextEx}) -> (ReturnValue:bool)
   n66 = GetSubGame() -> (ReturnValue=Campaign)
   n67 = Branch(
    execute=@n62.then_1
    Condition=AND Boolean(A="Not Equal (Enum)"(A:ESubGame Enum=@n66, B:ESubGame Enum=Campaign), B=OR Boolean(A=@n65, B=@n64))
   )
   n68 = "Get [Horizontal Box] Build-Edit"() -> ("[Horizontal Box] Build-Edit":Horizontal Box Object Reference)
   > n69 = SetVisibility(self=@n68, InVisibility=Hidden)
   SetKeyBindWidgetBoundAction(
    execute=[@n70.then, @n69.then]
    FortKeyBindWidget=Get KeybindMaterial() -> (KeybindMaterial:Keybind Widget Dark Object Reference)
    ActionName=GetChangeMaterialActionName!(self:Fort Global UIContext Object Reference=Get FortGlobalUIContext()) -> (ReturnValue:Name=None)
    InteractionType=Get InteractionType() -> (InteractionType=None)
   )
   > Set TopTrapActive(TopTrapActive=false)
   SetVisibility(execute=@n67.else, self=@n68, InVisibility=Visible)
   > n70 = SetKeyBindWidgetBoundAction(
    FortKeyBindWidget=Get KeyBindBuild-Edit() -> (KeyBindBuild-Edit:Keybind Widget Dark Object Reference)
    ActionName=GetPerformBuildingEditInteractionActionName(self=Get FortGlobalUIContext(), OverrideInputType=Count) -> (ReturnValue=None)
    InteractionType=Get InteractionType() -> (InteractionType=None)
   )
   n71 = Set TrapAvailable(execute=@n12.CastFailed, TrapAvailable=false)
   Set bIsEditAvailable(execute=@n10.else, bIsEditAvailable=false)
   > UpdateQuickEditPreviewVisibility()
  ]}
 UpdateBuildingRepairCost {nodes [
   n1 = UpdateBuildingRepairCost!() -> (then, NewInteractionCost:int) {is=title K2Node_FunctionEntry}
   > "SetText (Text)"(self=Get TextInteractionCost(), InText="To Text (Integer)"(Value=@n1))
  ]}
 UpdateFromTargetingDataOverride {nodes [
   UpdateFromTargetingDataOverride() {is=title K2Node_FunctionEntry}
   > n1 = Cast To BuildingActor(Object=GetIndicatedActor())
   n2 = Get BuildingTargetingDataOverride(self=@n1)
   > n3 = Is Valid(InputObject=@n2)
   n4 = "Get [Widget] Building HP && Trap Bars"()
   SetShowNumbersOnHealthBar!(execute=@n3.Is Not Valid, self:Building Bar Object Reference=@n4, DoShow:bool=true) -> (then)
   > SetHealthBarSize!(execute, self:Building Bar Object Reference=@n4, Size:Vector 2D Structure) -> (then)
   n5 = "Get [Widget] Building HP && Trap Bars"()
   SetShowNumbersOnHealthBar(
    execute=@n3.Is Valid
    self=@n5
    DoShow=Get bShowNumbersOnHealthBar!(self:Building Targeting Data Object Reference=@n2) -> (bShowNumbersOnHealthBar:bool)
   )
   > SetHealthBarSize(self=@n5, Size=Get HealthBarSizeOverride!(self:Building Targeting Data Object Reference=@n2) -> (HealthBarSizeOverride:Vector 2f Structure))
  ]}
 UpdateQuickEditPreviewVisibility {nodes [
   UpdateQuickEditPreviewVisibility() {is=title K2Node_FunctionEntry}
   n1 = "Equal (Enum)"(
    A:EQuickEditPreviewMethod Enum=Get QuickEditPreviewMethod!() -> (QuickEditPreviewMethod=MeshPreview)
    B:EQuickEditPreviewMethod Enum=MeshPreview
   )
   n2 = AND Boolean!(
    A=Get bIsQuickEditEnabled() -> (bIsQuickEditEnabled:bool)
    B=Get bIsEditAvailable()
    C:bool=Get bHasValidQuickEditOption() -> (bHasValidQuickEditOption:bool)
   ) -> (ReturnValue)
   > SetVisibility(
    self=Get Mesh_Preview()
    InVisibility=Select(Option 0:ESlateVisibility Enum=Collapsed, Option 1:ESlateVisibility Enum=HitTestInvisible, Index=AND Boolean!(A=@n2, B=@n1) -> (ReturnValue)) -> (ReturnValue:ESlateVisibility Enum)
   )
   n3 = "Equal (Enum)"(A=Get QuickEditPreviewMethod() -> (QuickEditPreviewMethod=MeshPreview), B=PatternGrid)
   > SetVisibility(
    self=[Get Image_QuickEditPreview(), Get Image_QuickEditPreview_2(), Get Image_QuickEditPreview_Touch(), Get Image_QuickEditPreview_1()]
    InVisibility=Select(Option 0=Collapsed, Option 1=HitTestInvisible, Index=AND Boolean(A=@n2, B=@n3))
   )
   n4 = "Equal (Enum)"(A=Get QuickEditPreviewMethod() -> (QuickEditPreviewMethod=MeshPreview), B=Pips)
   > SetVisibility(self=Get WBP_QuickEditGrid(), InVisibility=Select(Option 0=Hidden, Option 1=HitTestInvisible, Index=AND Boolean(A=@n2, B=@n4)))
  ]}
 UpdateTrapWidgets {nodes [
   UpdateTrapWidgets!() -> (
    then
    Place Trap Available:bool=~
    Interaction Type:EFortBuildingInteraction Enum=None
    Keybind:Keybind Widget Dark Object Reference
    Text:Common Text Object Reference
   ) {is=title K2Node_FunctionEntry}
   > n1 = Cast To BuildingSMActor(Object=GetIndicatedActor())
   n2 = GetSubGame() -> (ReturnValue=Campaign)
   > n3 = Branch(Condition=AND Boolean(A="Equal (Enum)"(A:ESubGame Enum=@n2, B:ESubGame Enum=Athena), B=Get Place Trap Available() -> (Place Trap Available:bool)))
   > SetKeyBindWidgetBoundAction(
    FortKeyBindWidget=Get Keybind() -> (Keybind:Keybind Widget Dark Object Reference)
    ActionName=GetChangeMaterialActionName(self=Get FortGlobalUIContext()) -> (ReturnValue=None)
    InteractionType=Get Interaction Type() -> (Interaction Type:EFortBuildingInteraction Enum=None)
   )
   > "SetText (Text)"(self=Get Text!() -> (Text:Common Text Object Reference), InText="LOCTEXT(\"SWITCH TRAP\")")
   n4 = Get FortGlobalUIContext()
   n5 = GetTrapPickerActionName!(self:Fort Global UIContext Object Reference=@n4, OverrideInputType:ECommonInputType Enum=Count) -> (ReturnValue:Name=None)
   n6 = GetKeyForAction!(self:Fort Global UIContext Object Reference=@n4, Action:Name=@n5, bForceGamepadKey:bool=false, PresetNameOverride:Name=None) -> (ReturnValue:Key Structure)
   SetKeyBindWidgetBoundAction(
    execute=@n3.else
    FortKeyBindWidget=Get Keybind()
    ActionName=Select(
     Option 0:Name=GetTrapHotbarActionName!(self:Fort Global UIContext Object Reference=@n4, OverrideInputType:ECommonInputType Enum=Count) -> (ReturnValue:Name=None)
     Option 1:Name=@n5
     Index=Is Valid Key!(Key:"Key Structure (by ref)"=@n6) -> (ReturnValue:bool)
    ) -> (ReturnValue:Name)
    InteractionType=Get Interaction Type() -> (Interaction Type=None)
   )
   n7 = Get FortInventoryContext()
   n8 = TextToUpper!(InText:"Text (by ref)"=GetTrapAttachTypeName!(self:Fort Inventory Context Object Reference=@n7, Actor:Building Static Mesh Object Reference=@n1) -> (ReturnValue:Text)) -> (ReturnValue:Text)
   > "SetText (Text)"(
    self=Get Text()
    InText=Select(
     Option 0:Text=Format Text!(Format="LOCTEXT(\"PICK {BuildingTypeName} TRAP\")", BuildingTypeName:Text=@n8) -> (Result)
     Option 1:Text=Format Text(Format="LOCTEXT(\"EQUIP {BuildingTypeName} TRAP\")", BuildingTypeName=@n8)
     Index=HasTrapReadyForBuilding!(self:Fort Inventory Context Object Reference=@n7, BuildingToAttachTo:Building Static Mesh Object Reference=@n1) -> (ReturnValue:bool)
    ) -> (ReturnValue:Text)
   )
  ]}
]
variables [
 bHasValidQuickEditOption {type=boolean}
 bIsEditAvailable {type=boolean}
 bIsPipEnabled {type=boolean}
 bIsQuickEditEnabled {type=boolean}
 bPlayerFacingFromSide {type=boolean}
 BuildingInfo {
  type=FortFocusedBuildingInfo
  default {
   attachedTrapMaxDurability=0
   bCanBePlayerEdited=false
   bIsAnyTrapAttached=false
   bIsInteractable=false
   bIsPlayerBuilt=false
   bIsPreviewTrapAttached=false
   bIsTrapAttachedFacingPlayer=false
   buildingMaterial=None
   healthDisplayRule=Never
   interactionCost=0
   interactionType=None
   maxHealth=0
   quickEditClass=None
   indicatorRelativeLocation=<0 0 0>
  }
 }
 Current Durability {type=number}
 Current Health {type=number}
 CurrentQuickEditClass {type=string, default=None}
 GamePhaseUpdatedAsyncTask {type=AsyncAction_StartListeningToStatefulEvent, default=null}
 HealthFadeTimerHandle {type=TimerHandle}
 MashPreviewResetNames {
  type="array<string>"
  default [
   pbwa_w1_solid_c
   pbwa_w1_floor_c
   pbwa_w1_stairw_c
   pbwa_w1_roofc_c
   pbwa_s1_solid_c
   pbwa_s1_floor_c
   pbwa_s1_stairw_c
   pbwa_s1_roofc_c
   pbwa_m1_solid_c
   pbwa_m1_floor_c
   pbwa_m1_stairw_c
   pbwa_m1_roofc_c
  ]
 }
 MeshPreviewMap_Metal {
  type="map<Texture>"
  default {
   pBWA_M1_ArchwayLarge_C_0=/Game/UI/InGame/HUD/SimpleEditPreviews/_0011_PBWA_W1_Brace_C_mirrored
   pBWA_M1_Brace_C_0=&_0011_PBWA_W1_Brace_C_mirrored
   pBWA_M1_Brace_C_0_mirrored=/Game/UI/InGame/HUD/SimpleEditPreviews/_0010_PBWA_W1_Brace_C
   pBWA_M1_DoorC_C_0=/Game/UI/InGame/HUD/SimpleEditPreviews/_0015_PBWA_W1_DoorC_C
   pBWA_M1_RoofO_C_0=/Game/UI/InGame/HUD/SimpleEditPreviews/_0002_PBWA_W1_Roof0_C_0
   pBWA_M1_RoofO_C_1=/Game/UI/InGame/HUD/SimpleEditPreviews/_0001_PBWA_W1_Roof0_C_1
   pBWA_M1_RoofO_C_2=/Game/UI/InGame/HUD/SimpleEditPreviews/_0000_PBWA_W1_Roof0_C_2
   pBWA_M1_RoofO_C_3=/Game/UI/InGame/HUD/SimpleEditPreviews/_0003_PBWA_W1_Roof0_C_3
   pBWA_M1_RoofS_C_0=/Game/UI/InGame/HUD/SimpleEditPreviews/_0007_PBWA_W1_StairF_C_Full
   pBWA_M1_RoofWall_C_0=/Game/UI/InGame/HUD/SimpleEditPreviews/_0012_PBWA_W1_RoofWall_C
   pBWA_M1_RoofWall_C_0_mirrored=/Game/UI/InGame/HUD/SimpleEditPreviews/_0013_PBWA_W1_RoofWall_C_mirrored
   pBWA_M1_StairF_C_0=/Game/UI/InGame/HUD/SimpleEditPreviews/_0005_PBWA_W1_StairF_C
   pBWA_M1_StairF_C_0_mirrored=/Game/UI/InGame/HUD/SimpleEditPreviews/_0004_PBWA_W1_StairF_C_mirrored
   pBWA_M1_WindowC_C_0=/Game/UI/InGame/HUD/SimpleEditPreviews/_0014_PBWA_W1_WindowC_C
   pBWA_M1_balconyi_c_0=/Game/UI/InGame/HUD/SimpleEditPreviews/_0008_PBWA_W1_Balconyl_C_0
   pBWA_M1_balconyi_c_1=/Game/UI/InGame/HUD/SimpleEditPreviews/_0007_PBWA_W1_Balconyl_C_1
   pBWA_M1_balconyi_c_2=/Game/UI/InGame/HUD/SimpleEditPreviews/_0006_PBWA_W1_Balconyl_C_2
   pBWA_M1_balconyi_c_3=/Game/UI/InGame/HUD/SimpleEditPreviews/_0009_PBWA_W1_Balconyl_C_3
   pBWA_M1_floor_c_0=/Game/UI/InGame/HUD/SimpleEditPreviews/_0011_PBWA_W1_Balconyl_Full
   pBWA_M1_roofc_c_0=/Game/UI/InGame/HUD/SimpleEditPreviews/_0000_PBWA_W1_Roof0_C_Full
   pBWA_M1_solid_c_0=/Game/UI/InGame/HUD/SimpleEditPreviews/_0019_PBWA_W1_RoofWall_Full
   pBWA_M1_stairw_c_0=&_0007_PBWA_W1_StairF_C_Full
   pBWA_M2_ArchwayLarge_C_0=&_0011_PBWA_W1_Brace_C_mirrored
   pBWA_M2_Brace_C_0=&_0011_PBWA_W1_Brace_C_mirrored
   pBWA_M2_Brace_C_0_mirrored=&_0010_PBWA_W1_Brace_C
   pBWA_M2_DoorC_C_0=&_0015_PBWA_W1_DoorC_C
   pBWA_M2_RoofO_C_0=&_0002_PBWA_W1_Roof0_C_0
   pBWA_M2_RoofO_C_1=&_0001_PBWA_W1_Roof0_C_1
   pBWA_M2_RoofO_C_2=&_0000_PBWA_W1_Roof0_C_2
   pBWA_M2_RoofO_C_3=&_0003_PBWA_W1_Roof0_C_3
   pBWA_M2_RoofS_C_0=&_0007_PBWA_W1_StairF_C_Full
   pBWA_M2_RoofWall_C_0=&_0012_PBWA_W1_RoofWall_C
   pBWA_M2_RoofWall_C_0_mirrored=&_0013_PBWA_W1_RoofWall_C_mirrored
   pBWA_M2_StairF_C_0=&_0005_PBWA_W1_StairF_C
   pBWA_M2_StairF_C_0_mirrored=&_0004_PBWA_W1_StairF_C_mirrored
   pBWA_M2_WindowC_C_0=&_0014_PBWA_W1_WindowC_C
   pBWA_M2_balconyi_c_0=&_0008_PBWA_W1_Balconyl_C_0
   pBWA_M2_balconyi_c_1=&_0007_PBWA_W1_Balconyl_C_1
   pBWA_M2_balconyi_c_2=&_0006_PBWA_W1_Balconyl_C_2
   pBWA_M2_balconyi_c_3=&_0009_PBWA_W1_Balconyl_C_3
   pBWA_M2_floor_c_0=&_0011_PBWA_W1_Balconyl_Full
   pBWA_M2_roofc_c_0=&_0000_PBWA_W1_Roof0_C_Full
   pBWA_M2_solid_c_0=&_0019_PBWA_W1_RoofWall_Full
   pBWA_M2_stairw_c_0=&_0007_PBWA_W1_StairF_C_Full
   pBWA_M3_ArchwayLarge_C_0=&_0011_PBWA_W1_Brace_C_mirrored
   pBWA_M3_Brace_C_0=&_0011_PBWA_W1_Brace_C_mirrored
   pBWA_M3_Brace_C_0_mirrored=&_0010_PBWA_W1_Brace_C
   pBWA_M3_DoorC_C_0=&_0015_PBWA_W1_DoorC_C
   pBWA_M3_RoofO_C_0=&_0002_PBWA_W1_Roof0_C_0
   pBWA_M3_RoofO_C_1=&_0001_PBWA_W1_Roof0_C_1
   pBWA_M3_RoofO_C_2=&_0000_PBWA_W1_Roof0_C_2
   pBWA_M3_RoofO_C_3=&_0003_PBWA_W1_Roof0_C_3
   pBWA_M3_RoofS_C_0=&_0007_PBWA_W1_StairF_C_Full
   pBWA_M3_RoofWall_C_0=&_0012_PBWA_W1_RoofWall_C
   pBWA_M3_RoofWall_C_0_mirrored=&_0013_PBWA_W1_RoofWall_C_mirrored
   pBWA_M3_StairF_C_0=&_0005_PBWA_W1_StairF_C
   pBWA_M3_StairF_C_0_mirrored=&_0004_PBWA_W1_StairF_C_mirrored
   pBWA_M3_WindowC_C_0=&_0014_PBWA_W1_WindowC_C
   pBWA_M3_balconyi_c_0=&_0008_PBWA_W1_Balconyl_C_0
   pBWA_M3_balconyi_c_1=&_0007_PBWA_W1_Balconyl_C_1
   pBWA_M3_balconyi_c_2=&_0006_PBWA_W1_Balconyl_C_2
   pBWA_M3_balconyi_c_3=&_0009_PBWA_W1_Balconyl_C_3
   pBWA_M3_floor_c_0=&_0011_PBWA_W1_Balconyl_Full
   pBWA_M3_roofc_c_0=&_0000_PBWA_W1_Roof0_C_Full
   pBWA_M3_solid_c_0=&_0019_PBWA_W1_RoofWall_Full
   pBWA_M3_stairw_c_0=&_0007_PBWA_W1_StairF_C_Full
  }
 }
 MeshPreviewMap_Stone {
  type="map<Texture>"
  default {
   pBWA_S1_ArchwayLarge_C_0=&_0011_PBWA_W1_Brace_C_mirrored
   pBWA_S1_Brace_C_0=&_0011_PBWA_W1_Brace_C_mirrored
   pBWA_S1_Brace_C_0_mirrored=&_0010_PBWA_W1_Brace_C
   pBWA_S1_DoorC_C_0=&_0015_PBWA_W1_DoorC_C
   pBWA_S1_RoofO_C_0=&_0002_PBWA_W1_Roof0_C_0
   pBWA_S1_RoofO_C_1=&_0001_PBWA_W1_Roof0_C_1
   pBWA_S1_RoofO_C_2=&_0000_PBWA_W1_Roof0_C_2
   pBWA_S1_RoofO_C_3=&_0003_PBWA_W1_Roof0_C_3
   pBWA_S1_RoofS_C_0=&_0007_PBWA_W1_StairF_C_Full
   pBWA_S1_RoofWall_C_0=&_0012_PBWA_W1_RoofWall_C
   pBWA_S1_RoofWall_C_0_mirrored=&_0013_PBWA_W1_RoofWall_C_mirrored
   pBWA_S1_StairF_C_0=&_0005_PBWA_W1_StairF_C
   pBWA_S1_StairF_C_0_mirrored=&_0004_PBWA_W1_StairF_C_mirrored
   pBWA_S1_WindowC_C_0=&_0014_PBWA_W1_WindowC_C
   pBWA_S1_balconyi_c_0=&_0008_PBWA_W1_Balconyl_C_0
   pBWA_S1_balconyi_c_1=&_0007_PBWA_W1_Balconyl_C_1
   pBWA_S1_balconyi_c_2=&_0006_PBWA_W1_Balconyl_C_2
   pBWA_S1_balconyi_c_3=&_0009_PBWA_W1_Balconyl_C_3
   pBWA_S1_floor_c_0=&_0011_PBWA_W1_Balconyl_Full
   pBWA_S1_roofc_c_0=&_0000_PBWA_W1_Roof0_C_Full
   pBWA_S1_solid_c_0=&_0019_PBWA_W1_RoofWall_Full
   pBWA_S1_stairw_c_0=&_0007_PBWA_W1_StairF_C_Full
   pBWA_S2_ArchwayLarge_C_0=&_0011_PBWA_W1_Brace_C_mirrored
   pBWA_S2_Brace_C_0=&_0011_PBWA_W1_Brace_C_mirrored
   pBWA_S2_Brace_C_0_mirrored=&_0010_PBWA_W1_Brace_C
   pBWA_S2_DoorC_C_0=&_0015_PBWA_W1_DoorC_C
   pBWA_S2_RoofO_C_0=&_0002_PBWA_W1_Roof0_C_0
   pBWA_S2_RoofO_C_1=&_0001_PBWA_W1_Roof0_C_1
   pBWA_S2_RoofO_C_2=&_0000_PBWA_W1_Roof0_C_2
   pBWA_S2_RoofO_C_3=&_0003_PBWA_W1_Roof0_C_3
   pBWA_S2_RoofS_C_0=&_0007_PBWA_W1_StairF_C_Full
   pBWA_S2_RoofWall_C_0=&_0012_PBWA_W1_RoofWall_C
   pBWA_S2_RoofWall_C_0_mirrored=&_0013_PBWA_W1_RoofWall_C_mirrored
   pBWA_S2_StairF_C_0=&_0005_PBWA_W1_StairF_C
   pBWA_S2_StairF_C_0_mirrored=&_0004_PBWA_W1_StairF_C_mirrored
   pBWA_S2_WindowC_C_0=&_0014_PBWA_W1_WindowC_C
   pBWA_S2_balconyi_c_0=&_0008_PBWA_W1_Balconyl_C_0
   pBWA_S2_balconyi_c_1=&_0007_PBWA_W1_Balconyl_C_1
   pBWA_S2_balconyi_c_2=&_0006_PBWA_W1_Balconyl_C_2
   pBWA_S2_balconyi_c_3=&_0009_PBWA_W1_Balconyl_C_3
   pBWA_S2_floor_c_0=&_0011_PBWA_W1_Balconyl_Full
   pBWA_S2_roofc_c_0=&_0000_PBWA_W1_Roof0_C_Full
   pBWA_S2_solid_c_0=&_0019_PBWA_W1_RoofWall_Full
   pBWA_S2_stairw_c_0=&_0007_PBWA_W1_StairF_C_Full
   pBWA_S3_ArchwayLarge_C_0=&_0011_PBWA_W1_Brace_C_mirrored
   pBWA_S3_Brace_C_0=&_0011_PBWA_W1_Brace_C_mirrored
   pBWA_S3_Brace_C_0_mirrored=&_0010_PBWA_W1_Brace_C
   pBWA_S3_DoorC_C_0=&_0015_PBWA_W1_DoorC_C
   pBWA_S3_RoofO_C_0=&_0002_PBWA_W1_Roof0_C_0
   pBWA_S3_RoofO_C_1=&_0001_PBWA_W1_Roof0_C_1
   pBWA_S3_RoofO_C_2=&_0000_PBWA_W1_Roof0_C_2
   pBWA_S3_RoofO_C_3=&_0003_PBWA_W1_Roof0_C_3
   pBWA_S3_RoofS_C_0=&_0007_PBWA_W1_StairF_C_Full
   pBWA_S3_RoofWall_C_0=&_0012_PBWA_W1_RoofWall_C
   pBWA_S3_RoofWall_C_0_mirrored=&_0013_PBWA_W1_RoofWall_C_mirrored
   pBWA_S3_StairF_C_0=&_0005_PBWA_W1_StairF_C
   pBWA_S3_StairF_C_0_mirrored=&_0004_PBWA_W1_StairF_C_mirrored
   pBWA_S3_WindowC_C_0=&_0014_PBWA_W1_WindowC_C
   pBWA_S3_balconyi_c_0=&_0008_PBWA_W1_Balconyl_C_0
   pBWA_S3_balconyi_c_1=&_0007_PBWA_W1_Balconyl_C_1
   pBWA_S3_balconyi_c_2=&_0006_PBWA_W1_Balconyl_C_2
   pBWA_S3_balconyi_c_3=&_0009_PBWA_W1_Balconyl_C_3
   pBWA_S3_floor_c_0=&_0011_PBWA_W1_Balconyl_Full
   pBWA_S3_roofc_c_0=&_0000_PBWA_W1_Roof0_C_Full
   pBWA_S3_solid_c_0=&_0019_PBWA_W1_RoofWall_Full
   pBWA_S3_stairw_c_0=&_0007_PBWA_W1_StairF_C_Full
  }
 }
 MeshPreviewMap_Wood {
  type="map<Texture>"
  default {
   pBWA_W1_ArchwayLarge_C_0=&_0011_PBWA_W1_Brace_C_mirrored
   pBWA_W1_Brace_C_0=&_0011_PBWA_W1_Brace_C_mirrored
   pBWA_W1_Brace_C_0_mirrored=&_0010_PBWA_W1_Brace_C
   pBWA_W1_DoorC_C_0=&_0015_PBWA_W1_DoorC_C
   pBWA_W1_RoofO_C_0=&_0002_PBWA_W1_Roof0_C_0
   pBWA_W1_RoofO_C_1=&_0001_PBWA_W1_Roof0_C_1
   pBWA_W1_RoofO_C_2=&_0000_PBWA_W1_Roof0_C_2
   pBWA_W1_RoofO_C_3=&_0003_PBWA_W1_Roof0_C_3
   pBWA_W1_RoofS_C_0=&_0007_PBWA_W1_StairF_C_Full
   pBWA_W1_RoofWall_C_0=&_0012_PBWA_W1_RoofWall_C
   pBWA_W1_RoofWall_C_0_mirrored=&_0013_PBWA_W1_RoofWall_C_mirrored
   pBWA_W1_StairF_C_0=&_0005_PBWA_W1_StairF_C
   pBWA_W1_StairF_C_0_mirrored=&_0004_PBWA_W1_StairF_C_mirrored
   pBWA_W1_WindowC_C_0=&_0014_PBWA_W1_WindowC_C
   pBWA_W1_balconyi_c_0=&_0008_PBWA_W1_Balconyl_C_0
   pBWA_W1_balconyi_c_1=&_0007_PBWA_W1_Balconyl_C_1
   pBWA_W1_balconyi_c_2=&_0006_PBWA_W1_Balconyl_C_2
   pBWA_W1_balconyi_c_3=&_0009_PBWA_W1_Balconyl_C_3
   pBWA_W1_floor_c_0=&_0011_PBWA_W1_Balconyl_Full
   pBWA_W1_roofc_c_0=&_0000_PBWA_W1_Roof0_C_Full
   pBWA_W1_solid_c_0=&_0019_PBWA_W1_RoofWall_Full
   pBWA_W1_stairw_c_0=&_0007_PBWA_W1_StairF_C_Full
   pBWA_W2_ArchwayLarge_C_0=&_0011_PBWA_W1_Brace_C_mirrored
   pBWA_W2_Brace_C_0=&_0011_PBWA_W1_Brace_C_mirrored
   pBWA_W2_Brace_C_0_mirrored=&_0010_PBWA_W1_Brace_C
   pBWA_W2_DoorC_C_0=&_0015_PBWA_W1_DoorC_C
   pBWA_W2_RoofO_C_0=&_0002_PBWA_W1_Roof0_C_0
   pBWA_W2_RoofO_C_1=&_0001_PBWA_W1_Roof0_C_1
   pBWA_W2_RoofO_C_2=&_0000_PBWA_W1_Roof0_C_2
   pBWA_W2_RoofO_C_3=&_0003_PBWA_W1_Roof0_C_3
   pBWA_W2_RoofS_C_0=&_0007_PBWA_W1_StairF_C_Full
   pBWA_W2_RoofWall_C_0=&_0012_PBWA_W1_RoofWall_C
   pBWA_W2_RoofWall_C_0_mirrored=&_0013_PBWA_W1_RoofWall_C_mirrored
   pBWA_W2_StairF_C_0=&_0005_PBWA_W1_StairF_C
   pBWA_W2_StairF_C_0_mirrored=&_0004_PBWA_W1_StairF_C_mirrored
   pBWA_W2_WindowC_C_0=&_0014_PBWA_W1_WindowC_C
   pBWA_W2_balconyi_c_0=&_0008_PBWA_W1_Balconyl_C_0
   pBWA_W2_balconyi_c_1=&_0007_PBWA_W1_Balconyl_C_1
   pBWA_W2_balconyi_c_2=&_0006_PBWA_W1_Balconyl_C_2
   pBWA_W2_balconyi_c_3=&_0009_PBWA_W1_Balconyl_C_3
   pBWA_W2_floor_c_0=&_0011_PBWA_W1_Balconyl_Full
   pBWA_W2_roofc_c_0=&_0000_PBWA_W1_Roof0_C_Full
   pBWA_W2_solid_c_0=&_0019_PBWA_W1_RoofWall_Full
   pBWA_W2_stairw_c_0=&_0007_PBWA_W1_StairF_C_Full
   pBWA_W3_ArchwayLarge_C_0=&_0011_PBWA_W1_Brace_C_mirrored
   pBWA_W3_Brace_C_0=&_0011_PBWA_W1_Brace_C_mirrored
   pBWA_W3_Brace_C_0_mirrored=&_0010_PBWA_W1_Brace_C
   pBWA_W3_DoorC_C_0=&_0015_PBWA_W1_DoorC_C
   pBWA_W3_RoofO_C_0=&_0002_PBWA_W1_Roof0_C_0
   pBWA_W3_RoofO_C_1=&_0001_PBWA_W1_Roof0_C_1
   pBWA_W3_RoofO_C_2=&_0000_PBWA_W1_Roof0_C_2
   pBWA_W3_RoofO_C_3=&_0003_PBWA_W1_Roof0_C_3
   pBWA_W3_RoofS_C_0=&_0007_PBWA_W1_StairF_C_Full
   pBWA_W3_RoofWall_C_0=&_0012_PBWA_W1_RoofWall_C
   pBWA_W3_RoofWall_C_0_mirrored=&_0013_PBWA_W1_RoofWall_C_mirrored
   pBWA_W3_StairF_C_0=&_0005_PBWA_W1_StairF_C
   pBWA_W3_StairF_C_0_mirrored=&_0004_PBWA_W1_StairF_C_mirrored
   pBWA_W3_WindowC_C_0=&_0014_PBWA_W1_WindowC_C
   pBWA_W3_balconyi_c_0=&_0008_PBWA_W1_Balconyl_C_0
   pBWA_W3_balconyi_c_1=&_0007_PBWA_W1_Balconyl_C_1
   pBWA_W3_balconyi_c_2=&_0006_PBWA_W1_Balconyl_C_2
   pBWA_W3_balconyi_c_3=&_0009_PBWA_W1_Balconyl_C_3
   pBWA_W3_floor_c_0=&_0011_PBWA_W1_Balconyl_Full
   pBWA_W3_roofc_c_0=&_0000_PBWA_W1_Roof0_C_Full
   pBWA_W3_solid_c_0=&_0019_PBWA_W1_RoofWall_Full
   pBWA_W3_stairw_c_0=&_0007_PBWA_W1_StairF_C_Full
  }
 }
 QuickEditMap {
  type="map<MaterialInstance>"
  default {
   pBWA_M1_ArchwayLarge_C=/Game/UI/InGame/HUD/Textures/M_UI_QuickEdit_10
   pBWA_M1_RoofS_C=/Game/UI/InGame/HUD/Textures/M_UI_QuickEdit_11
   pBWA_S1_ArchwayLarge_C=&M_UI_QuickEdit_10
   pBWA_S1_RoofS_C=&M_UI_QuickEdit_11
   pBWA_W1_ArchwayLarge_C=&M_UI_QuickEdit_10
   pBWA_W1_RoofS_C=&M_UI_QuickEdit_11
   pBWA_m2_ArchwayLarge_C=&M_UI_QuickEdit_10
   pBWA_m2_RoofS_C=&M_UI_QuickEdit_11
   pBWA_m3_ArchwayLarge_C=&M_UI_QuickEdit_10
   pBWA_m3_RoofS_C=&M_UI_QuickEdit_11
   pBWA_s2_ArchwayLarge_C=&M_UI_QuickEdit_10
   pBWA_s2_RoofS_C=&M_UI_QuickEdit_11
   pBWA_s3_ArchwayLarge_C=&M_UI_QuickEdit_10
   pBWA_s3_RoofS_C=&M_UI_QuickEdit_11
   pBWA_w2_ArchwayLarge_C=&M_UI_QuickEdit_10
   pBWA_w2_RoofS_C=&M_UI_QuickEdit_11
   pBWA_w3_ArchwayLarge_C=&M_UI_QuickEdit_10
   pBWA_w3_RoofS_C=&M_UI_QuickEdit_11
   pbwa_m1_balconyi_c=/Game/UI/InGame/HUD/Textures/M_UI_QuickEdit_8
   pbwa_m1_brace_c=/Game/UI/InGame/HUD/Textures/M_UI_QuickEdit_3
   pbwa_m1_doorc_c=/Game/UI/InGame/HUD/Textures/M_UI_QuickEdit_2
   pbwa_m1_floor_c=/Game/UI/InGame/HUD/Textures/M_UI_QuickEdit_7
   pbwa_m1_roofc_c=&M_UI_QuickEdit_7
   pbwa_m1_roofo_c=&M_UI_QuickEdit_8
   pbwa_m1_roofwall_c=/Game/UI/InGame/HUD/Textures/M_UI_QuickEdit_9
   pbwa_m1_solid_c=&M_UI_QuickEdit_0
   pbwa_m1_stairf_c=/Game/UI/InGame/HUD/Textures/M_UI_QuickEdit_4
   pbwa_m1_stairw_c=&M_UI_QuickEdit_7
   pbwa_m1_windowc_c=/Game/UI/InGame/HUD/Textures/M_UI_QuickEdit_1
   pbwa_m1_windowside_c=/Game/UI/InGame/HUD/Textures/M_UI_QuickEdit_6
   pbwa_m2_balconyi_c=&M_UI_QuickEdit_8
   pbwa_m2_brace_c=&M_UI_QuickEdit_3
   pbwa_m2_doorc_c=&M_UI_QuickEdit_2
   pbwa_m2_floor_c=&M_UI_QuickEdit_7
   pbwa_m2_roofc_c=&M_UI_QuickEdit_7
   pbwa_m2_roofo_c=&M_UI_QuickEdit_8
   pbwa_m2_roofwall_c=&M_UI_QuickEdit_9
   pbwa_m2_solid_c=&M_UI_QuickEdit_0
   pbwa_m2_stairf_c=&M_UI_QuickEdit_4
   pbwa_m2_stairw_c=&M_UI_QuickEdit_7
   pbwa_m2_windowc_c=&M_UI_QuickEdit_1
   pbwa_m2_windowside_c=&M_UI_QuickEdit_6
   pbwa_m3_balconyi_c=&M_UI_QuickEdit_8
   pbwa_m3_brace_c=&M_UI_QuickEdit_3
   pbwa_m3_doorc_c=&M_UI_QuickEdit_2
   pbwa_m3_floor_c=&M_UI_QuickEdit_7
   pbwa_m3_roofc_c=&M_UI_QuickEdit_7
   pbwa_m3_roofo_c=&M_UI_QuickEdit_8
   pbwa_m3_roofwall_c=&M_UI_QuickEdit_9
   pbwa_m3_solid_c=&M_UI_QuickEdit_0
   pbwa_m3_stairf_c=&M_UI_QuickEdit_4
   pbwa_m3_stairw_c=&M_UI_QuickEdit_7
   pbwa_m3_windowc_c=&M_UI_QuickEdit_1
   pbwa_m3_windowside_c=&M_UI_QuickEdit_6
   pbwa_s1_balconyi_c=&M_UI_QuickEdit_8
   pbwa_s1_brace_c=&M_UI_QuickEdit_3
   pbwa_s1_doorc_c=&M_UI_QuickEdit_2
   pbwa_s1_floor_c=&M_UI_QuickEdit_7
   pbwa_s1_roofc_c=&M_UI_QuickEdit_7
   pbwa_s1_roofo_c=&M_UI_QuickEdit_8
   pbwa_s1_roofwall_c=&M_UI_QuickEdit_9
   pbwa_s1_solid_c=&M_UI_QuickEdit_0
   pbwa_s1_stairf_c=&M_UI_QuickEdit_4
   pbwa_s1_stairw_c=&M_UI_QuickEdit_7
   pbwa_s1_windowsc_c=&M_UI_QuickEdit_1
   pbwa_s1_windowsside_c=&M_UI_QuickEdit_6
   pbwa_s2_balconyi_c=&M_UI_QuickEdit_8
   pbwa_s2_brace_c=&M_UI_QuickEdit_3
   pbwa_s2_doorc_c=&M_UI_QuickEdit_2
   pbwa_s2_floor_c=&M_UI_QuickEdit_7
   pbwa_s2_roofc_c=&M_UI_QuickEdit_7
   pbwa_s2_roofo_c=&M_UI_QuickEdit_8
   pbwa_s2_roofwall_c=&M_UI_QuickEdit_9
   pbwa_s2_solid_c=&M_UI_QuickEdit_0
   pbwa_s2_stairf_c=&M_UI_QuickEdit_4
   pbwa_s2_stairw_c=&M_UI_QuickEdit_7
   pbwa_s2_windowsc_c=&M_UI_QuickEdit_1
   pbwa_s2_windowsside_c=&M_UI_QuickEdit_6
   pbwa_s3_balconyi_c=&M_UI_QuickEdit_8
   pbwa_s3_brace_c=&M_UI_QuickEdit_3
   pbwa_s3_doorc_c=&M_UI_QuickEdit_2
   pbwa_s3_floor_c=&M_UI_QuickEdit_7
   pbwa_s3_roofc_c=&M_UI_QuickEdit_7
   pbwa_s3_roofo_c=&M_UI_QuickEdit_8
   pbwa_s3_roofwall_c=&M_UI_QuickEdit_9
   pbwa_s3_solid_c=&M_UI_QuickEdit_0
   pbwa_s3_stairf_c=&M_UI_QuickEdit_4
   pbwa_s3_stairw_c=&M_UI_QuickEdit_7
   pbwa_s3_windowsc_c=&M_UI_QuickEdit_1
   pbwa_s3_windowsside_c=&M_UI_QuickEdit_6
   pbwa_w1_RoofWall_c=&M_UI_QuickEdit_9
   pbwa_w1_balconyi_c=&M_UI_QuickEdit_8
   pbwa_w1_brace_c=&M_UI_QuickEdit_3
   pbwa_w1_doorc_c=&M_UI_QuickEdit_2
   pbwa_w1_floor_c=&M_UI_QuickEdit_7
   pbwa_w1_roofc_c=&M_UI_QuickEdit_7
   pbwa_w1_roofo_c=&M_UI_QuickEdit_8
   pbwa_w1_solid_c=&M_UI_QuickEdit_0
   pbwa_w1_stairf_c=&M_UI_QuickEdit_4
   pbwa_w1_stairw_c=&M_UI_QuickEdit_7
   pbwa_w1_windowc_c=&M_UI_QuickEdit_1
   pbwa_w1_windowside_c=&M_UI_QuickEdit_6
   pbwa_w2_RoofWall_c=&M_UI_QuickEdit_9
   pbwa_w2_balconyi_c=&M_UI_QuickEdit_8
   pbwa_w2_brace_c=&M_UI_QuickEdit_3
   pbwa_w2_doorc_c=&M_UI_QuickEdit_2
   pbwa_w2_floor_c=&M_UI_QuickEdit_7
   pbwa_w2_roofc_c=&M_UI_QuickEdit_7
   pbwa_w2_roofo_c=&M_UI_QuickEdit_8
   pbwa_w2_solid_c=&M_UI_QuickEdit_0
   pbwa_w2_stairf_c=&M_UI_QuickEdit_4
   pbwa_w2_stairw_c=&M_UI_QuickEdit_7
   pbwa_w2_windowc_c=&M_UI_QuickEdit_1
   pbwa_w2_windowside_c=&M_UI_QuickEdit_6
   pbwa_w3_RoofWall_c=&M_UI_QuickEdit_9
   pbwa_w3_balconyi_c=&M_UI_QuickEdit_8
   pbwa_w3_brace_c=&M_UI_QuickEdit_3
   pbwa_w3_doorc_c=&M_UI_QuickEdit_2
   pbwa_w3_floor_c=&M_UI_QuickEdit_7
   pbwa_w3_roofc_c=&M_UI_QuickEdit_7
   pbwa_w3_roofo_c=&M_UI_QuickEdit_8
   pbwa_w3_solid_c=&M_UI_QuickEdit_0
   pbwa_w3_stairf_c=&M_UI_QuickEdit_4
   pbwa_w3_stairw_c=&M_UI_QuickEdit_7
   pbwa_w3_windowc_c=&M_UI_QuickEdit_1
   pbwa_w3_windowside_c=&M_UI_QuickEdit_6
  }
 }
 QuickEditPreviewMethod {type=EQuickEditPreviewMethod, default=MeshPreview}
]
widgets {
 info {inheritedWidgetCount=2, parentClass=/Script/FortniteUI.FortBuildingInfoIndicatorWidget, rootWidgetClass=&Overlay, widgetCount=80}
 widgets [
  Overlay_0 {widgetClassPath=&Overlay, widget=Overlay_0}
  WBP_QuickEditGrid {
   widgetClassPath=&/WBP_QuickEditGrid
   bIsVariable=true
   parent=Overlay_0
   widget=WBP_QuickEditGrid
   properties {visibility=Collapsed}
   slot {class=OverlaySlot, properties {horizontalAlignment=HAlign_Fill, verticalAlignment=VAlign_Fill}}
  }
  VerticalBox_0 {widgetClassPath=&VerticalBox, parent=Overlay_0, widget=VerticalBox_0, slot {class=OverlaySlot}}
  "[Widget] Building HP && Trap Bars" {
   widgetClassPath=&/BuildingBar
   bIsVariable=true
   parent=VerticalBox_0
   widget="[Widget] Building HP && Trap Bars"
   slot {class=VerticalBoxSlot}
  }
  Mesh_Preview {
   widgetClassPath=&Image
   bInherited=true
   bIsVariable=true
   parent=VerticalBox_0
   widget=Mesh_Preview
   properties {
    visibility=Collapsed
    brush {
     bIsDynamicallyLoaded=false
     drawAs=Image
     imageType=FullColor
     mirroring=NoMirror
     resourceName=None
     resourceObject=/Game/UI/InGame/HUD/Textures/M_UI_SimpleEditMeshPreview
     tiling=NoTile
     imageSize=<108 108>
     margin=<ltrb 0 0 0 0>
     outlineSettings {
      bUseBrushTransparency=false
      roundingType=HalfHeightRadius
      width=0
      color {colorUseRule=UseColor_Specified, specifiedColor=<rgba 0 0 0 0>}
      cornerRadii=<0 0 0 0>
     }
     tintColor {colorUseRule=UseColor_Specified, specifiedColor=<rgba 1 1 1 1>}
     uVRegion {bIsValid=false, max=<0 0>, min=<0 0>}
    }
   }
   slot {class=VerticalBoxSlot, properties {horizontalAlignment=HAlign_Center, verticalAlignment=VAlign_Center}}
  }
  Input {
   widgetClassPath=&WidgetSwitcher
   bIsVariable=true
   parent=VerticalBox_0
   widget=Input
   properties {activeWidgetIndex=1}
   slot {class=VerticalBoxSlot, properties {padding=<ltrb 0 8 0 0>}}
  }
  SwitcherInteractions {
   widgetClassPath=&CommonWidgetSwitcherLegacy
   bIsVariable=true
   parent=Input
   widget=SwitcherInteractions
   properties {transitionDuration=0}
   slot {class=WidgetSwitcherSlot}
  }
  SizeBoxTriOption {
   widgetClassPath=&SizeBox
   bIsVariable=true
   parent=SwitcherInteractions
   widget=SizeBoxTriOption
   properties {bOverride_MinDesiredWidth=true, minDesiredWidth=512, widthOverride=512}
   slot {class=WidgetSwitcherSlot}
  }
  VerticalBox_5 {
   widgetClassPath=&VerticalBox
   parent=SizeBoxTriOption
   widget=VerticalBox_5
   slot {class=SizeBoxSlot, properties {verticalAlignment=VAlign_Top}}
  }
  "[Horizontal Box]" {widgetClassPath=&HorizontalBox, parent=VerticalBox_5, widget="[Horizontal Box]", slot {class=VerticalBoxSlot}}
  SwitcherSideOption2 {
   widgetClassPath=&CommonWidgetSwitcherLegacy
   bIsVariable=true
   parent="[Horizontal Box]"
   widget=SwitcherSideOption2
   properties {transitionDuration=0}
   slot {class=HorizontalBoxSlot, properties {horizontalAlignment=HAlign_Right, verticalAlignment=VAlign_Top, size {sizeRule=Fill, value=1}}}
  }
  "[Vertical Box] Build Subset" {
   widgetClassPath=&VerticalBox
   bIsVariable=true
   parent=SwitcherSideOption2
   widget="[Vertical Box] Build Subset"
   slot {class=WidgetSwitcherSlot}
  }
  "[Horizontal Box] Build-Edit" {
   widgetClassPath=&HorizontalBox
   bIsVariable=true
   parent="[Vertical Box] Build Subset"
   widget="[Horizontal Box] Build-Edit"
   slot {class=VerticalBoxSlot, properties {horizontalAlignment=HAlign_Right, verticalAlignment=VAlign_Center}}
  }
  Image_QuickEditPreview_2 {
   widgetClassPath=&Image
   bIsVariable=true
   parent="[Horizontal Box] Build-Edit"
   widget=Image_QuickEditPreview_2
   properties {
    visibility=Collapsed
    brush {
     bIsDynamicallyLoaded=false
     drawAs=Image
     imageType=NoImage
     mirroring=NoMirror
     resourceName=None
     resourceObject=&M_UI_QuickEdit_0
     tiling=NoTile
     imageSize=<64 64>
     margin=<ltrb 0 0 0 0>
     outlineSettings {
      bUseBrushTransparency=false
      roundingType=HalfHeightRadius
      width=0
      color {colorUseRule=UseColor_Specified, specifiedColor=<rgba 0 0 0 0>}
      cornerRadii=<0 0 0 0>
     }
     tintColor {colorUseRule=UseColor_Specified, specifiedColor=<rgba 1 1 1 1>}
     uVRegion {bIsValid=false, max=<0 0>, min=<0 0>}
    }
   }
   slot {class=HorizontalBoxSlot, properties {horizontalAlignment=HAlign_Center, verticalAlignment=VAlign_Center, padding=<ltrb 0 0 8 0>}}
  }
  TextBuild-Edit {
   widgetClassPath=&CommonTextBlock
   bIsVariable=true
   parent="[Horizontal Box] Build-Edit"
   widget=TextBuild-Edit
   properties {
    justification=Right
    minDesiredWidth=1
    style=/FortUILibrary/Styles/Typography/Foundation/TextStyle-Base-XS-B-S.TextStyle-Base-XS-B-S_C
    text=EDIT
    font {
     bForceMonospaced=false
     bMaterialIsStencil=false
     fontMaterial=null
     fontObject=&NotoSans
     letterSpacing=0
     monospacedWidth=1
     size=16
     skewAmount=0
     typefaceFontName=NotoSans-Bold
     outlineSettings {
      bApplyOutlineToDropShadows=true
      bMiteredCorners=false
      bSeparateFillAlpha=true
      outlineMaterial=&M-FontShadow
      outlineSize=2
      outlineColor=<rgba 0 0 0 0.5>
     }
    }
    shadowOffset=<0 0>
   }
   slot {class=HorizontalBoxSlot, properties {horizontalAlignment=HAlign_Left, verticalAlignment=VAlign_Center}}
  }
  KeyBindBuild-Edit {
   widgetClassPath=/FortUILibrary/Input/KeybindWidgetDark
   bIsVariable=true
   parent="[Horizontal Box] Build-Edit"
   widget=KeyBindBuild-Edit
   slot {class=HorizontalBoxSlot, properties {horizontalAlignment=HAlign_Center, verticalAlignment=VAlign_Center, padding=<ltrb 8 8 8 8>}}
  }
  "[Horizontal Box] Material" {
   widgetClassPath=&HorizontalBox
   bIsVariable=true
   parent="[Vertical Box] Build Subset"
   widget="[Horizontal Box] Material"
   slot {class=VerticalBoxSlot, properties {padding=<ltrb 0 4 0 0>}}
  }
  TextMaterial {
   widgetClassPath=&CommonTextBlock
   parent="[Horizontal Box] Material"
   widget=TextMaterial
   properties {
    justification=Right
    minDesiredWidth=1
    style=&TextStyle-Base-XS-B-S_C
    text=MATERIAL
    font {
     bForceMonospaced=false
     bMaterialIsStencil=false
     fontMaterial=null
     fontObject=&NotoSans
     letterSpacing=0
     monospacedWidth=1
     size=16
     skewAmount=0
     typefaceFontName=NotoSans-Bold
     outlineSettings {
      bApplyOutlineToDropShadows=true
      bMiteredCorners=false
      bSeparateFillAlpha=true
      outlineMaterial=&M-FontShadow
      outlineSize=2
      outlineColor=<rgba 0 0 0 0.5>
     }
    }
    shadowOffset=<0 0>
   }
   slot {class=HorizontalBoxSlot, properties {verticalAlignment=VAlign_Center}}
  }
  KeybindMaterial {
   widgetClassPath=&KeybindWidgetDark
   bIsVariable=true
   parent="[Horizontal Box] Material"
   widget=KeybindMaterial
   slot {class=HorizontalBoxSlot, properties {horizontalAlignment=HAlign_Center, verticalAlignment=VAlign_Center, padding=<ltrb 8 8 8 8>}}
  }
  "[Horizontal Box] Edit" {
   widgetClassPath=&HorizontalBox
   bIsVariable=true
   parent=SwitcherSideOption2
   widget="[Horizontal Box] Edit"
   slot {class=WidgetSwitcherSlot, properties {horizontalAlignment=HAlign_Right, verticalAlignment=VAlign_Top}}
  }
  Image_QuickEditPreview {
   widgetClassPath=&Image
   bIsVariable=true
   parent="[Horizontal Box] Edit"
   widget=Image_QuickEditPreview
   properties {
    visibility=Collapsed
    brush {
     bIsDynamicallyLoaded=false
     drawAs=Image
     imageType=NoImage
     mirroring=NoMirror
     resourceName=None
     resourceObject=&M_UI_QuickEdit_0
     tiling=NoTile
     imageSize=<64 64>
     margin=<ltrb 0 0 0 0>
     outlineSettings {
      bUseBrushTransparency=false
      roundingType=HalfHeightRadius
      width=0
      color {colorUseRule=UseColor_Specified, specifiedColor=<rgba 0 0 0 0>}
      cornerRadii=<0 0 0 0>
     }
     tintColor {colorUseRule=UseColor_Specified, specifiedColor=<rgba 1 1 1 1>}
     uVRegion {bIsValid=false, max=<0 0>, min=<0 0>}
    }
   }
   slot {class=HorizontalBoxSlot, properties {horizontalAlignment=HAlign_Center, verticalAlignment=VAlign_Center, padding=<ltrb 0 0 8 0>}}
  }
  TextEditLabel {
   widgetClassPath=&CommonTextBlock
   bIsVariable=true
   parent="[Horizontal Box] Edit"
   widget=TextEditLabel
   properties {
    justification=Center
    style=&TextStyle-Base-XS-B-S_C
    text=EDIT
    font {
     bForceMonospaced=false
     bMaterialIsStencil=false
     fontMaterial=null
     fontObject=&NotoSans
     letterSpacing=0
     monospacedWidth=1
     size=16
     skewAmount=0
     typefaceFontName=NotoSans-Bold
     outlineSettings {
      bApplyOutlineToDropShadows=true
      bMiteredCorners=false
      bSeparateFillAlpha=true
      outlineMaterial=&M-FontShadow
      outlineSize=2
      outlineColor=<rgba 0 0 0 0.5>
     }
    }
    shadowOffset=<0 0>
   }
   slot {class=HorizontalBoxSlot, properties {verticalAlignment=VAlign_Center, padding=<ltrb 0 0 4 0>}}
  }
  KeybindEdit {
   widgetClassPath=&KeybindWidgetDark
   bIsVariable=true
   parent="[Horizontal Box] Edit"
   widget=KeybindEdit
   slot {class=HorizontalBoxSlot, properties {horizontalAlignment=HAlign_Center, verticalAlignment=VAlign_Center, padding=<ltrb 8 8 8 8>}}
  }
  SwitcherSideOption1 {
   widgetClassPath=&CommonWidgetSwitcherLegacy
   bIsVariable=true
   parent="[Horizontal Box]"
   widget=SwitcherSideOption1
   properties {transitionDuration=0}
   slot {
    class=HorizontalBoxSlot
    properties {horizontalAlignment=HAlign_Left, verticalAlignment=VAlign_Top, padding=<ltrb 20 0 0 0>, size {sizeRule=Fill, value=1}}
   }
  }
  VerticalBoxBuildWithRotate {
   widgetClassPath=&VerticalBox
   bIsVariable=true
   parent=SwitcherSideOption1
   widget=VerticalBoxBuildWithRotate
   slot {class=WidgetSwitcherSlot, properties {verticalAlignment=VAlign_Top}}
  }
  HorizontalBox {widgetClassPath=&HorizontalBox, parent=VerticalBoxBuildWithRotate, widget=HorizontalBox, slot {class=VerticalBoxSlot}}
  KeybindBuild {
   widgetClassPath=&KeybindWidgetDark
   bIsVariable=true
   parent=HorizontalBox
   widget=KeybindBuild
   slot {class=HorizontalBoxSlot, properties {horizontalAlignment=HAlign_Center, verticalAlignment=VAlign_Top, padding=<ltrb 8 8 8 8>}}
  }
  HorizontalBox_1 {
   widgetClassPath=&HorizontalBox
   parent=HorizontalBox
   widget=HorizontalBox_1
   slot {class=HorizontalBoxSlot, properties {horizontalAlignment=HAlign_Left}}
  }
  CommonTextBlock_0 {
   widgetClassPath=&CommonTextBlock
   bIsVariable=true
   parent=HorizontalBox_1
   widget=CommonTextBlock_0
   properties {
    justification=Right
    minDesiredWidth=1
    style=&TextStyle-Base-XS-B-S_C
    text=BUILD
    font {
     bForceMonospaced=false
     bMaterialIsStencil=false
     fontMaterial=null
     fontObject=&NotoSans
     letterSpacing=0
     monospacedWidth=1
     size=16
     skewAmount=0
     typefaceFontName=NotoSans-Bold
     outlineSettings {
      bApplyOutlineToDropShadows=true
      bMiteredCorners=false
      bSeparateFillAlpha=true
      outlineMaterial=&M-FontShadow
      outlineSize=2
      outlineColor=<rgba 0 0 0 0.5>
     }
    }
    shadowOffset=<0 0>
   }
   slot {class=HorizontalBoxSlot, properties {verticalAlignment=VAlign_Center}}
  }
  CommonBorder_0 {
   widgetClassPath=&CommonBorder
   parent=HorizontalBox_1
   widget=CommonBorder_0
   properties {
    background {
     bIsDynamicallyLoaded=false
     drawAs=Image
     imageType=NoImage
     mirroring=NoMirror
     resourceName=None
     resourceObject=null
     tiling=NoTile
     imageSize=<32 32>
     margin=<ltrb 0 0 0 0>
     outlineSettings {
      bUseBrushTransparency=false
      roundingType=HalfHeightRadius
      width=0
      color {colorUseRule=UseColor_Specified, specifiedColor=<rgba 0 0 0 0>}
      cornerRadii=<0 0 0 0>
     }
     tintColor {colorUseRule=UseColor_Specified, specifiedColor=<rgba 1 1 1 1>}
     uVRegion {bIsValid=false, max=<0 0>, min=<0 0>}
    }
    brushColor=<rgba 0.0273 0.0273 0.0273 0.5>
   }
   slot {class=HorizontalBoxSlot, properties {verticalAlignment=VAlign_Center, padding=<ltrb 5 0 0 0>}}
  }
  HorizontalBox_3 {widgetClassPath=&HorizontalBox, parent=CommonBorder_0, widget=HorizontalBox_3, slot {class=BorderSlot}}
  ImageResourceTypeIconwithRotate {
   widgetClassPath=&FortLazyImage
   bIsVariable=true
   parent=HorizontalBox_3
   widget=ImageResourceTypeIconwithRotate
   properties {brush {
     bIsDynamicallyLoaded=false
     drawAs=Image
     imageType=NoImage
     mirroring=NoMirror
     resourceName=None
     resourceObject=null
     tiling=NoTile
     imageSize=<40 40>
     margin=<ltrb 0 0 0 0>
     outlineSettings {
      bUseBrushTransparency=false
      roundingType=HalfHeightRadius
      width=0
      color {colorUseRule=UseColor_Specified, specifiedColor=<rgba 0 0 0 0>}
      cornerRadii=<0 0 0 0>
     }
     tintColor {colorUseRule=UseColor_Specified, specifiedColor=<rgba 1 1 1 1>}
     uVRegion {bIsValid=false, max=<0 0>, min=<0 0>}
    }}
   slot {class=HorizontalBoxSlot, properties {verticalAlignment=VAlign_Top}}
  }
  TextInteractionCostwithRotate {
   widgetClassPath=&CommonTextBlock
   bIsVariable=true
   parent=HorizontalBox_3
   widget=TextInteractionCostwithRotate
   properties {
    justification=Right
    style=&TextStyle-Base-XS-B-S_C
    text=999
    font {
     bForceMonospaced=false
     bMaterialIsStencil=false
     fontMaterial=null
     fontObject=&NotoSans
     letterSpacing=0
     monospacedWidth=1
     size=16
     skewAmount=0
     typefaceFontName=NotoSans-Bold
     outlineSettings {
      bApplyOutlineToDropShadows=true
      bMiteredCorners=false
      bSeparateFillAlpha=true
      outlineMaterial=&M-FontShadow
      outlineSize=2
      outlineColor=<rgba 0 0 0 0.5>
     }
    }
    shadowOffset=<0 0>
   }
   slot {class=HorizontalBoxSlot, properties {verticalAlignment=VAlign_Center}}
  }
  RotateBox {
   widgetClassPath=&HorizontalBox
   bIsVariable=true
   parent=VerticalBoxBuildWithRotate
   widget=RotateBox
   slot {class=VerticalBoxSlot, properties {horizontalAlignment=HAlign_Left, padding=<ltrb 0 4 0 0>}}
  }
  KeybindRotate {
   widgetClassPath=&KeybindWidgetDark
   bIsVariable=true
   parent=RotateBox
   widget=KeybindRotate
   slot {class=HorizontalBoxSlot, properties {horizontalAlignment=HAlign_Center, verticalAlignment=VAlign_Center, padding=<ltrb 8 8 8 8>}}
  }
  CommonTextBlock_2 {
   widgetClassPath=&CommonTextBlock
   parent=RotateBox
   widget=CommonTextBlock_2
   properties {
    justification=Right
    minDesiredWidth=1
    style=&TextStyle-Base-XS-B-S_C
    text=ROTATE
    font {
     bForceMonospaced=false
     bMaterialIsStencil=false
     fontMaterial=null
     fontObject=&NotoSans
     letterSpacing=0
     monospacedWidth=1
     size=16
     skewAmount=0
     typefaceFontName=NotoSans-Bold
     outlineSettings {
      bApplyOutlineToDropShadows=true
      bMiteredCorners=false
      bSeparateFillAlpha=true
      outlineMaterial=&M-FontShadow
      outlineSize=2
      outlineColor=<rgba 0 0 0 0.5>
     }
    }
    shadowOffset=<0 0>
   }
   slot {class=HorizontalBoxSlot, properties {horizontalAlignment=HAlign_Left, verticalAlignment=VAlign_Center}}
  }
  HorizontalBoxUpgradeRepair {
   widgetClassPath=&HorizontalBox
   bIsVariable=true
   parent=SwitcherSideOption1
   widget=HorizontalBoxUpgradeRepair
   slot {class=WidgetSwitcherSlot, properties {verticalAlignment=VAlign_Top}}
  }
  KeybindUpgradeRepair {
   widgetClassPath=&KeybindWidgetDark
   bIsVariable=true
   parent=HorizontalBoxUpgradeRepair
   widget=KeybindUpgradeRepair
   properties {forcedHoldKeybindStatus=ForcedHold}
   slot {class=HorizontalBoxSlot, properties {horizontalAlignment=HAlign_Center, verticalAlignment=VAlign_Top, padding=<ltrb 8 8 8 8>}}
  }
  VerticalBox_176 {
   widgetClassPath=&VerticalBox
   parent=HorizontalBoxUpgradeRepair
   widget=VerticalBox_176
   slot {class=HorizontalBoxSlot, properties {horizontalAlignment=HAlign_Left, verticalAlignment=VAlign_Center, padding=<ltrb 4 0 0 0>}}
  }
  TextInteractionName {
   widgetClassPath=&CommonTextBlock
   bIsVariable=true
   parent=VerticalBox_176
   widget=TextInteractionName
   properties {
    justification=Right
    style=&TextStyle-Base-XS-B-S_C
    text=UPGRADE
    font {
     bForceMonospaced=false
     bMaterialIsStencil=false
     fontMaterial=null
     fontObject=&NotoSans
     letterSpacing=0
     monospacedWidth=1
     size=16
     skewAmount=0
     typefaceFontName=NotoSans-Bold
     outlineSettings {
      bApplyOutlineToDropShadows=true
      bMiteredCorners=false
      bSeparateFillAlpha=true
      outlineMaterial=&M-FontShadow
      outlineSize=2
      outlineColor=<rgba 0 0 0 0.5>
     }
    }
    shadowOffset=<0 0>
   }
   slot {class=VerticalBoxSlot, properties {horizontalAlignment=HAlign_Left, verticalAlignment=VAlign_Center}}
  }
  HorizontalBox_488 {
   widgetClassPath=&HorizontalBox
   parent=VerticalBox_176
   widget=HorizontalBox_488
   slot {class=VerticalBoxSlot, properties {horizontalAlignment=HAlign_Left, verticalAlignment=VAlign_Bottom}}
  }
  "[Image] Resource Type Icon" {
   widgetClassPath=&FortLazyImage
   bIsVariable=true
   parent=HorizontalBox_488
   widget="[Image] Resource Type Icon"
   properties {brush {
     bIsDynamicallyLoaded=false
     drawAs=Image
     imageType=NoImage
     mirroring=NoMirror
     resourceName=None
     resourceObject=null
     tiling=NoTile
     imageSize=<40 40>
     margin=<ltrb 0 0 0 0>
     outlineSettings {
      bUseBrushTransparency=false
      roundingType=HalfHeightRadius
      width=0
      color {colorUseRule=UseColor_Specified, specifiedColor=<rgba 0 0 0 0>}
      cornerRadii=<0 0 0 0>
     }
     tintColor {colorUseRule=UseColor_Specified, specifiedColor=<rgba 1 1 1 1>}
     uVRegion {bIsValid=false, max=<0 0>, min=<0 0>}
    }}
   slot {class=HorizontalBoxSlot, properties {verticalAlignment=VAlign_Top}}
  }
  TextInteractionCost {
   widgetClassPath=&CommonTextBlock
   bIsVariable=true
   parent=HorizontalBox_488
   widget=TextInteractionCost
   properties {
    justification=Right
    style=&TextStyle-Base-XS-B-S_C
    text=999
    font {
     bForceMonospaced=false
     bMaterialIsStencil=false
     fontMaterial=null
     fontObject=&NotoSans
     letterSpacing=0
     monospacedWidth=1
     size=16
     skewAmount=0
     typefaceFontName=NotoSans-Bold
     outlineSettings {
      bApplyOutlineToDropShadows=true
      bMiteredCorners=false
      bSeparateFillAlpha=true
      outlineMaterial=&M-FontShadow
      outlineSize=2
      outlineColor=<rgba 0 0 0 0.5>
     }
    }
    shadowOffset=<0 0>
   }
   slot {class=HorizontalBoxSlot, properties {verticalAlignment=VAlign_Center}}
  }
  TrapOptions {
   widgetClassPath=&VerticalBox
   bIsVariable=true
   parent=SwitcherSideOption1
   widget=TrapOptions
   slot {class=WidgetSwitcherSlot, properties {verticalAlignment=VAlign_Top}}
  }
  HorizontalBoxPlaceTrap {
   widgetClassPath=&HorizontalBox
   bIsVariable=true
   parent=TrapOptions
   widget=HorizontalBoxPlaceTrap
   slot {class=VerticalBoxSlot, properties {horizontalAlignment=HAlign_Left, padding=<ltrb 0 0 0 4>}}
  }
  KeybindPlaceTrapLeft {
   widgetClassPath=&KeybindWidgetDark
   bIsVariable=true
   parent=HorizontalBoxPlaceTrap
   widget=KeybindPlaceTrapLeft
   slot {class=HorizontalBoxSlot, properties {horizontalAlignment=HAlign_Center, verticalAlignment=VAlign_Center, padding=<ltrb 8 8 8 8>}}
  }
  CommonTextBlock_8 {
   widgetClassPath=&CommonTextBlock
   parent=HorizontalBoxPlaceTrap
   widget=CommonTextBlock_8
   properties {
    style=&TextStyle-Base-XS-B-S_C
    text=PLACE TRAP
    font {
     bForceMonospaced=false
     bMaterialIsStencil=false
     fontMaterial=null
     fontObject=&NotoSans
     letterSpacing=0
     monospacedWidth=1
     size=16
     skewAmount=0
     typefaceFontName=NotoSans-Bold
     outlineSettings {
      bApplyOutlineToDropShadows=true
      bMiteredCorners=false
      bSeparateFillAlpha=true
      outlineMaterial=&M-FontShadow
      outlineSize=2
      outlineColor=<rgba 0 0 0 0.5>
     }
    }
    shadowOffset=<0 0>
   }
   slot {class=HorizontalBoxSlot, properties {horizontalAlignment=HAlign_Right, verticalAlignment=VAlign_Center}}
  }
  HorizontalBoxCycleTrap {widgetClassPath=&HorizontalBox, bIsVariable=true, parent=TrapOptions, widget=HorizontalBoxCycleTrap, slot {class=VerticalBoxSlot}}
  KeybindCycleTrapLeft {
   widgetClassPath=&KeybindWidgetDark
   bIsVariable=true
   parent=HorizontalBoxCycleTrap
   widget=KeybindCycleTrapLeft
   slot {class=HorizontalBoxSlot, properties {horizontalAlignment=HAlign_Left, verticalAlignment=VAlign_Center, padding=<ltrb 8 8 8 8>}}
  }
  TextPickTrap {
   widgetClassPath=&CommonTextBlock
   bIsVariable=true
   parent=HorizontalBoxCycleTrap
   widget=TextPickTrap
   properties {
    style=&TextStyle-Base-XS-B-S_C
    text=PICK TRAP
    font {
     bForceMonospaced=false
     bMaterialIsStencil=false
     fontMaterial=null
     fontObject=&NotoSans
     letterSpacing=0
     monospacedWidth=1
     size=16
     skewAmount=0
     typefaceFontName=NotoSans-Bold
     outlineSettings {
      bApplyOutlineToDropShadows=true
      bMiteredCorners=false
      bSeparateFillAlpha=true
      outlineMaterial=&M-FontShadow
      outlineSize=2
      outlineColor=<rgba 0 0 0 0.5>
     }
    }
    shadowOffset=<0 0>
   }
   slot {class=HorizontalBoxSlot, properties {verticalAlignment=VAlign_Center, size {sizeRule=Fill, value=1}}}
  }
  VerticalBoxLowerOption {
   widgetClassPath=&VerticalBox
   bIsVariable=true
   parent=VerticalBox_5
   widget=VerticalBoxLowerOption
   slot {class=VerticalBoxSlot, properties {horizontalAlignment=HAlign_Center, padding=<ltrb 0 8 0 0>}}
  }
  KeybindTrapLower {
   widgetClassPath=&KeybindWidgetDark
   bIsVariable=true
   parent=VerticalBoxLowerOption
   widget=KeybindTrapLower
   slot {class=VerticalBoxSlot, properties {horizontalAlignment=HAlign_Center, padding=<ltrb 8 8 8 8>}}
  }
  TextPickTrapBottom {
   widgetClassPath=&CommonTextBlock
   bIsVariable=true
   parent=VerticalBoxLowerOption
   widget=TextPickTrapBottom
   properties {
    justification=Center
    minDesiredWidth=1
    style=&TextStyle-Base-XS-B-S_C
    text=PICK TRAP
    font {
     bForceMonospaced=false
     bMaterialIsStencil=false
     fontMaterial=null
     fontObject=&NotoSans
     letterSpacing=0
     monospacedWidth=1
     size=16
     skewAmount=0
     typefaceFontName=NotoSans-Bold
     outlineSettings {
      bApplyOutlineToDropShadows=true
      bMiteredCorners=false
      bSeparateFillAlpha=true
      outlineMaterial=&M-FontShadow
      outlineSize=2
      outlineColor=<rgba 0 0 0 0.5>
     }
    }
    shadowOffset=<0 0>
   }
   slot {class=VerticalBoxSlot}
  }
  "[Size Box] Confirm" {
   widgetClassPath=&SizeBox
   bIsVariable=true
   parent=SwitcherInteractions
   widget="[Size Box] Confirm"
   properties {bOverride_MinDesiredWidth=true, minDesiredWidth=512}
   slot {class=WidgetSwitcherSlot}
  }
  HorizontalBox_0 {
   widgetClassPath=&HorizontalBox
   parent="[Size Box] Confirm"
   widget=HorizontalBox_0
   slot {class=SizeBoxSlot, properties {verticalAlignment=VAlign_Top}}
  }
  VerticalBox_432 {
   widgetClassPath=&VerticalBox
   parent=HorizontalBox_0
   widget=VerticalBox_432
   slot {class=HorizontalBoxSlot, properties {horizontalAlignment=HAlign_Right, size {sizeRule=Fill, value=1}}}
  }
  KeybindConfirm {
   widgetClassPath=&KeybindWidgetDark
   bIsVariable=true
   parent=VerticalBox_432
   widget=KeybindConfirm
   properties {boundAction=CompleteBuildingEditInteraction}
   slot {class=VerticalBoxSlot, properties {horizontalAlignment=HAlign_Center, padding=<ltrb 8 8 8 8>}}
  }
  TextConfirmLabel {
   widgetClassPath=&CommonTextBlock
   parent=VerticalBox_432
   widget=TextConfirmLabel
   properties {
    justification=Right
    style=&TextStyle-Base-XS-B-S_C
    text=CONFIRM
    font {
     bForceMonospaced=false
     bMaterialIsStencil=false
     fontMaterial=null
     fontObject=&NotoSans
     letterSpacing=0
     monospacedWidth=1
     size=16
     skewAmount=0
     typefaceFontName=NotoSans-Bold
     outlineSettings {
      bApplyOutlineToDropShadows=true
      bMiteredCorners=false
      bSeparateFillAlpha=true
      outlineMaterial=&M-FontShadow
      outlineSize=2
      outlineColor=<rgba 0 0 0 0.5>
     }
    }
    shadowOffset=<0 0>
   }
   slot {class=VerticalBoxSlot}
  }
  VerticalBox_3 {
   widgetClassPath=&VerticalBox
   parent=HorizontalBox_0
   widget=VerticalBox_3
   slot {class=HorizontalBoxSlot, properties {horizontalAlignment=HAlign_Left, padding=<ltrb 20 0 0 0>, size {sizeRule=Fill, value=1}}}
  }
  KeybindReset {
   widgetClassPath=&KeybindWidgetDark
   bIsVariable=true
   parent=VerticalBox_3
   widget=KeybindReset
   properties {boundAction=BuildingEditReset}
   slot {class=VerticalBoxSlot, properties {horizontalAlignment=HAlign_Center, padding=<ltrb 8 8 8 8>}}
  }
  TextResetLabel {
   widgetClassPath=&CommonTextBlock
   parent=VerticalBox_3
   widget=TextResetLabel
   properties {
    style=&TextStyle-Base-XS-B-S_C
    text=RESET
    font {
     bForceMonospaced=false
     bMaterialIsStencil=false
     fontMaterial=null
     fontObject=&NotoSans
     letterSpacing=0
     monospacedWidth=1
     size=16
     skewAmount=0
     typefaceFontName=NotoSans-Bold
     outlineSettings {
      bApplyOutlineToDropShadows=true
      bMiteredCorners=false
      bSeparateFillAlpha=true
      outlineMaterial=&M-FontShadow
      outlineSize=2
      outlineColor=<rgba 0 0 0 0.5>
     }
    }
    shadowOffset=<0 0>
   }
   slot {class=VerticalBoxSlot}
  }
  "[Size Box] Edit (standalone)" {
   widgetClassPath=&SizeBox
   bIsVariable=true
   parent=SwitcherInteractions
   widget="[Size Box] Edit (standalone)"
   properties {bOverride_MinDesiredWidth=true, minDesiredWidth=512}
   slot {class=WidgetSwitcherSlot}
  }
  VerticalBox_1 {
   widgetClassPath=&VerticalBox
   parent="[Size Box] Edit (standalone)"
   widget=VerticalBox_1
   slot {class=SizeBoxSlot, properties {horizontalAlignment=HAlign_Center, verticalAlignment=VAlign_Top}}
  }
  Image_QuickEditPreview_1 {
   widgetClassPath=&Image
   bIsVariable=true
   parent=VerticalBox_1
   widget=Image_QuickEditPreview_1
   properties {
    visibility=Collapsed
    brush {
     bIsDynamicallyLoaded=false
     drawAs=Image
     imageType=NoImage
     mirroring=NoMirror
     resourceName=None
     resourceObject=&M_UI_QuickEdit_0
     tiling=NoTile
     imageSize=<64 64>
     margin=<ltrb 0 0 0 0>
     outlineSettings {
      bUseBrushTransparency=false
      roundingType=HalfHeightRadius
      width=0
      color {colorUseRule=UseColor_Specified, specifiedColor=<rgba 0 0 0 0>}
      cornerRadii=<0 0 0 0>
     }
     tintColor {colorUseRule=UseColor_Specified, specifiedColor=<rgba 1 1 1 1>}
     uVRegion {bIsValid=false, max=<0 0>, min=<0 0>}
    }
   }
   slot {class=VerticalBoxSlot, properties {horizontalAlignment=HAlign_Center, verticalAlignment=VAlign_Center}}
  }
  KeybindEditstandalone {
   widgetClassPath=&KeybindWidgetDark
   bIsVariable=true
   parent=VerticalBox_1
   widget=KeybindEditstandalone
   properties {boundAction=PerformBuildingEditInteraction}
   slot {class=VerticalBoxSlot, properties {horizontalAlignment=HAlign_Center, padding=<ltrb 8 8 8 8>}}
  }
  TextEdit {
   widgetClassPath=&CommonTextBlock
   bIsVariable=true
   parent=VerticalBox_1
   widget=TextEdit
   properties {
    justification=Center
    style=&TextStyle-Base-XS-B-S_C
    text=EDIT
    font {
     bForceMonospaced=false
     bMaterialIsStencil=false
     fontMaterial=null
     fontObject=&NotoSans
     letterSpacing=0
     monospacedWidth=1
     size=16
     skewAmount=0
     typefaceFontName=NotoSans-Bold
     outlineSettings {
      bApplyOutlineToDropShadows=true
      bMiteredCorners=false
      bSeparateFillAlpha=true
      outlineMaterial=&M-FontShadow
      outlineSize=2
      outlineColor=<rgba 0 0 0 0.5>
     }
    }
    shadowOffset=<0 0>
   }
   slot {class=VerticalBoxSlot, properties {padding=<ltrb 0 -2 0 0>}}
  }
  DynamicInteraction {
   widgetClassPath=&SizeBox
   bIsVariable=true
   parent=SwitcherInteractions
   widget=DynamicInteraction
   properties {bOverride_MinDesiredWidth=true, minDesiredWidth=512}
   slot {class=WidgetSwitcherSlot}
  }
  VerticalBox {
   widgetClassPath=&VerticalBox
   parent=DynamicInteraction
   widget=VerticalBox
   slot {class=SizeBoxSlot, properties {horizontalAlignment=HAlign_Center, verticalAlignment=VAlign_Top}}
  }
  DynamicInteractionTexture {
   widgetClassPath=&Image
   bIsVariable=true
   parent=VerticalBox
   widget=DynamicInteractionTexture
   slot {class=VerticalBoxSlot, properties {horizontalAlignment=HAlign_Center, verticalAlignment=VAlign_Center}}
  }
  DynamicInteractionPrompt {
   widgetClassPath=&CommonTextBlock
   bIsVariable=true
   parent=VerticalBox
   widget=DynamicInteractionPrompt
   properties {
    justification=Center
    style=&TextStyle-Base-XS-B-S_C
    text=INTERACT
    font {
     bForceMonospaced=false
     bMaterialIsStencil=false
     fontMaterial=null
     fontObject=&NotoSans
     letterSpacing=0
     monospacedWidth=1
     size=16
     skewAmount=0
     typefaceFontName=NotoSans-Bold
     outlineSettings {
      bApplyOutlineToDropShadows=true
      bMiteredCorners=false
      bSeparateFillAlpha=true
      outlineMaterial=&M-FontShadow
      outlineSize=2
      outlineColor=<rgba 0 0 0 0.5>
     }
    }
    shadowOffset=<0 0>
   }
   slot {class=VerticalBoxSlot, properties {padding=<ltrb 0 -2 0 0>}}
  }
  SizeBoxTrapstandalone {
   widgetClassPath=&SizeBox
   bIsVariable=true
   parent=SwitcherInteractions
   widget=SizeBoxTrapstandalone
   properties {bOverride_MinDesiredWidth=true, minDesiredWidth=512}
   slot {class=WidgetSwitcherSlot}
  }
  HorizontalBox_2 {
   widgetClassPath=&HorizontalBox
   parent=SizeBoxTrapstandalone
   widget=HorizontalBox_2
   slot {class=SizeBoxSlot, properties {horizontalAlignment=HAlign_Center}}
  }
  VerticalBoxPlaceTrapstandalone {
   widgetClassPath=&VerticalBox
   bIsVariable=true
   parent=HorizontalBox_2
   widget=VerticalBoxPlaceTrapstandalone
   slot {class=HorizontalBoxSlot}
  }
  KeybindPlaceTrapstandalone {
   widgetClassPath=&KeybindWidgetDark
   bIsVariable=true
   parent=VerticalBoxPlaceTrapstandalone
   widget=KeybindPlaceTrapstandalone
   slot {class=VerticalBoxSlot, properties {horizontalAlignment=HAlign_Center, padding=<ltrb 8 8 8 8>}}
  }
  TextPlaceTrapLabel {
   widgetClassPath=&CommonTextBlock
   parent=VerticalBoxPlaceTrapstandalone
   widget=TextPlaceTrapLabel
   properties {
    justification=Center
    style=&TextStyle-Base-XS-B-S_C
    text=PLACE TRAP
    font {
     bForceMonospaced=false
     bMaterialIsStencil=false
     fontMaterial=null
     fontObject=&NotoSans
     letterSpacing=0
     monospacedWidth=1
     size=16
     skewAmount=0
     typefaceFontName=NotoSans-Bold
     outlineSettings {
      bApplyOutlineToDropShadows=true
      bMiteredCorners=false
      bSeparateFillAlpha=true
      outlineMaterial=&M-FontShadow
      outlineSize=2
      outlineColor=<rgba 0 0 0 0.5>
     }
    }
    shadowOffset=<0 0>
   }
   slot {class=VerticalBoxSlot, properties {verticalAlignment=VAlign_Top}}
  }
  VerticalBox_7 {
   widgetClassPath=&VerticalBox
   parent=HorizontalBox_2
   widget=VerticalBox_7
   slot {class=HorizontalBoxSlot, properties {horizontalAlignment=HAlign_Left, verticalAlignment=VAlign_Top, padding=<ltrb 20 0 0 0>}}
  }
  KeybindPickTrapstandalone {
   widgetClassPath=&KeybindWidgetDark
   bIsVariable=true
   parent=VerticalBox_7
   widget=KeybindPickTrapstandalone
   slot {class=VerticalBoxSlot, properties {horizontalAlignment=HAlign_Center, padding=<ltrb 8 8 8 8>}}
  }
  TextPickTrapLabel {
   widgetClassPath=&CommonTextBlock
   bIsVariable=true
   parent=VerticalBox_7
   widget=TextPickTrapLabel
   properties {
    justification=Center
    style=&TextStyle-Base-XS-B-S_C
    text=PICK TRAP
    wrapTextAt=150
    font {
     bForceMonospaced=false
     bMaterialIsStencil=false
     fontMaterial=null
     fontObject=&NotoSans
     letterSpacing=0
     monospacedWidth=1
     size=16
     skewAmount=0
     typefaceFontName=NotoSans-Bold
     outlineSettings {
      bApplyOutlineToDropShadows=true
      bMiteredCorners=false
      bSeparateFillAlpha=true
      outlineMaterial=&M-FontShadow
      outlineSize=2
      outlineColor=<rgba 0 0 0 0.5>
     }
    }
    shadowOffset=<0 0>
   }
   slot {class=VerticalBoxSlot, properties {horizontalAlignment=HAlign_Center, verticalAlignment=VAlign_Center}}
  }
  Image_QuickEditPreview_Touch {
   widgetClassPath=&Image
   bIsVariable=true
   parent=Input
   widget=Image_QuickEditPreview_Touch
   properties {
    visibility=Collapsed
    brush {
     bIsDynamicallyLoaded=false
     drawAs=Image
     imageType=NoImage
     mirroring=NoMirror
     resourceName=None
     resourceObject=&M_UI_QuickEdit_0
     tiling=NoTile
     imageSize=<64 64>
     margin=<ltrb 0 0 0 0>
     outlineSettings {
      bUseBrushTransparency=false
      roundingType=HalfHeightRadius
      width=0
      color {colorUseRule=UseColor_Specified, specifiedColor=<rgba 0 0 0 0>}
      cornerRadii=<0 0 0 0>
     }
     tintColor {colorUseRule=UseColor_Specified, specifiedColor=<rgba 1 1 1 1>}
     uVRegion {bIsValid=false, max=<0 0>, min=<0 0>}
    }
   }
   slot {class=WidgetSwitcherSlot, properties {horizontalAlignment=HAlign_Center, verticalAlignment=VAlign_Center, padding=<ltrb 0 0 8 0>}}
  }
  TutorialHighlight {widgetClassPath=&TutorialHighlight}
 ]
}
