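# Widget Blueprint for a top quickbar keybind prompt in the in-game HUD, laying out the inherited widgets of FortQuickbarKeybindWidget.
# A CommonVisibilitySwitcher holds three layouts: a single keybind with side arrows, a "+" combo of two keybinds, and a double bind with arrows.
# The paired keybinds bind to GamepadAbility1 and GamepadAbility2, and right arrows reuse the left arrow texture rotated 180 degrees.
# The event graph contains only empty events, and inherited HoldText has no widget here.
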WidgetBlueprint /Game/UI/InGame/HUD/QuickbarKeybindTop
graphs [EventGraph {nodes [
   Event PreConstruct()
   Event Construct()
   Event Tick()
  ]}]
variables []
widgets {
 info {inheritedWidgetCount=11, parentClass=/Script/FortniteUI.FortQuickbarKeybindWidget, rootWidgetClass=&SizeBox, widgetCount=22}
 widgets [
  SizeBox_TopKeybind {widgetClassPath=&SizeBox, widget=SizeBox_TopKeybind, properties {bOverride_HeightOverride=true, heightOverride=32}}
  MainSwitcher {
   widgetClassPath=&CommonVisibilitySwitcher
   bInherited=true
   parent=SizeBox_TopKeybind
   widget=MainSwitcher
   properties {bOverride_Cursor=true}
   slot {class=SizeBoxSlot}
  }
  HorizontalBox_3 {widgetClassPath=&HorizontalBox, parent=MainSwitcher, widget=HorizontalBox_3, slot {class=CommonVisibilitySwitcherSlot}}
  LeftArrowImage {
   widgetClassPath=&Image
   bInherited=true
   parent=HorizontalBox_3
   widget=LeftArrowImage
   properties {
    visibility=Collapsed
    brush {
     bIsDynamicallyLoaded=false
     drawAs=Image
     imageType=NoImage
     mirroring=NoMirror
     resourceName=None
     resourceObject=/Game/UI/Foundation/Textures/Buttons/T-HUD-Slider-Arrow
     tiling=NoTile
     imageSize=<12 12>
     margin=<ltrb 0 0 0 0>
     outlineSettings {
      bUseBrushTransparency=false
      roundingType=HalfHeightRadius
      width=0
      color {colorUseRule=UseColor_Specified, specifiedColor=<rgba 0 0 0 0>}
      cornerRadii=<0 0 0 0>
     }
     tintColor {colorUseRule=UseColor_Specified, specifiedColor=<rgba 1 1 1 1>}
     uVRegion {bIsValid=false, max=<0 0>, min=<0 0>}
    }
   }
   slot {class=HorizontalBoxSlot, properties {horizontalAlignment=HAlign_Right, verticalAlignment=VAlign_Center}}
  }
  SizeBoxKeybindWidget {
   widgetClassPath=&SizeBox
   parent=HorizontalBox_3
   widget=SizeBoxKeybindWidget
   properties {bOverride_HeightOverride=true, bOverride_WidthOverride=true, heightOverride=32, minDesiredWidth=32, widthOverride=32}
   slot {class=HorizontalBoxSlot}
  }
  KeybindWidget {
   widgetClassPath=/FortUILibrary/Input/KeybindWidget
   bInherited=true
   parent=SizeBoxKeybindWidget
   widget=KeybindWidget
   properties {visibility=Collapsed}
   slot {class=SizeBoxSlot}
  }
  RightArrowImage {
   widgetClassPath=&Image
   bInherited=true
   parent=HorizontalBox_3
   widget=RightArrowImage
   properties {
    visibility=Collapsed
    brush {
     bIsDynamicallyLoaded=false
     drawAs=Image
     imageType=NoImage
     mirroring=NoMirror
     resourceName=None
     resourceObject=&T-HUD-Slider-Arrow
     tiling=NoTile
     imageSize=<12 12>
     margin=<ltrb 0 0 0 0>
     outlineSettings {
      bUseBrushTransparency=false
      roundingType=HalfHeightRadius
      width=0
      color {colorUseRule=UseColor_Specified, specifiedColor=<rgba 0 0 0 0>}
      cornerRadii=<0 0 0 0>
     }
     tintColor {colorUseRule=UseColor_Specified, specifiedColor=<rgba 1 1 1 1>}
     uVRegion {bIsValid=false, max=<0 0>, min=<0 0>}
    }
    renderTransform {angle=180, scale=<1 1>, shear=<0 0>, translation=<0 0>}
   }
   slot {class=HorizontalBoxSlot, properties {horizontalAlignment=HAlign_Right, verticalAlignment=VAlign_Center}}
  }
  HorizontalBox_1 {
   widgetClassPath=&HorizontalBox
   parent=MainSwitcher
   widget=HorizontalBox_1
   properties {renderTransform {angle=0, scale=<1 1>, shear=<0 0>, translation=<0 8>}}
   slot {class=CommonVisibilitySwitcherSlot}
  }
  SizeBoxTopCombo1 {
   widgetClassPath=&SizeBox
   parent=HorizontalBox_1
   widget=SizeBoxTopCombo1
   properties {bOverride_HeightOverride=true, bOverride_WidthOverride=true, heightOverride=32, widthOverride=32}
   slot {class=HorizontalBoxSlot, properties {horizontalAlignment=HAlign_Left, verticalAlignment=VAlign_Center}}
  }
  KeybindWidgetCombo1 {
   widgetClassPath=&KeybindWidget
   bInherited=true
   parent=SizeBoxTopCombo1
   widget=KeybindWidgetCombo1
   properties {boundAction=GamepadAbility1}
   slot {class=SizeBoxSlot}
  }
  CommonTextBlock_0 {
   widgetClassPath=&CommonTextBlock
   parent=HorizontalBox_1
   widget=CommonTextBlock_0
   properties {
    style=/FortUILibrary/Styles/Typography/Foundation/TextStyle-Base-XS.TextStyle-Base-XS_C
    text=+
    font {
     bForceMonospaced=false
     bMaterialIsStencil=false
     fontMaterial=null
     fontObject=&NotoSans
     letterSpacing=0
     monospacedWidth=1
     size=16
     skewAmount=0
     typefaceFontName=NotoSans-Regular
     outlineSettings {
      bApplyOutlineToDropShadows=false
      bMiteredCorners=false
      bSeparateFillAlpha=false
      outlineMaterial=null
      outlineSize=0
      outlineColor=<rgba 0 0 0 1>
     }
    }
    shadowOffset=<0 0>
   }
   slot {class=HorizontalBoxSlot, properties {horizontalAlignment=HAlign_Center, verticalAlignment=VAlign_Center}}
  }
  SizeBoxTopCombo2 {
   widgetClassPath=&SizeBox
   parent=HorizontalBox_1
   widget=SizeBoxTopCombo2
   properties {bOverride_HeightOverride=true, bOverride_WidthOverride=true, heightOverride=32, widthOverride=32}
   slot {class=HorizontalBoxSlot, properties {horizontalAlignment=HAlign_Right, verticalAlignment=VAlign_Center}}
  }
  KeybindWidgetCombo2 {
   widgetClassPath=&KeybindWidget
   bInherited=true
   parent=SizeBoxTopCombo2
   widget=KeybindWidgetCombo2
   properties {boundAction=GamepadAbility2}
   slot {class=SizeBoxSlot}
  }
  HorizontalBoxTopDoubleBindArrows {
   widgetClassPath=&HorizontalBox
   parent=MainSwitcher
   widget=HorizontalBoxTopDoubleBindArrows
   properties {renderTransform {angle=0, scale=<1 1>, shear=<0 0>, translation=<0 8>}}
   slot {class=CommonVisibilitySwitcherSlot}
  }
  LeftArrowImage2 {
   widgetClassPath=&Image
   bInherited=true
   parent=HorizontalBoxTopDoubleBindArrows
   widget=LeftArrowImage2
   properties {
    visibility=Collapsed
    brush {
     bIsDynamicallyLoaded=false
     drawAs=Image
     imageType=NoImage
     mirroring=NoMirror
     resourceName=None
     resourceObject=&T-HUD-Slider-Arrow
     tiling=NoTile
     imageSize=<16 16>
     margin=<ltrb 0 0 0 0>
     outlineSettings {
      bUseBrushTransparency=false
      roundingType=HalfHeightRadius
      width=0
      color {colorUseRule=UseColor_Specified, specifiedColor=<rgba 0 0 0 0>}
      cornerRadii=<0 0 0 0>
     }
     tintColor {colorUseRule=UseColor_Specified, specifiedColor=<rgba 1 1 1 1>}
     uVRegion {bIsValid=false, max=<0 0>, min=<0 0>}
    }
   }
   slot {class=HorizontalBoxSlot, properties {horizontalAlignment=HAlign_Right, verticalAlignment=VAlign_Center}}
  }
  SizeBoxComboArrow1 {
   widgetClassPath=&SizeBox
   parent=HorizontalBoxTopDoubleBindArrows
   widget=SizeBoxComboArrow1
   properties {bOverride_HeightOverride=true, bOverride_WidthOverride=true, heightOverride=32, widthOverride=32}
   slot {class=HorizontalBoxSlot, properties {horizontalAlignment=HAlign_Left, verticalAlignment=VAlign_Center}}
  }
  KeybindWidgetComboArrow1 {
   widgetClassPath=&KeybindWidget
   bInherited=true
   parent=SizeBoxComboArrow1
   widget=KeybindWidgetComboArrow1
   properties {boundAction=GamepadAbility1, visibility=Collapsed}
   slot {class=SizeBoxSlot}
  }
  Spacer_1 {
   widgetClassPath=&Spacer
   parent=HorizontalBoxTopDoubleBindArrows
   widget=Spacer_1
   slot {class=HorizontalBoxSlot, properties {verticalAlignment=VAlign_Center}}
  }
  SizeBoxComboArrow2 {
   widgetClassPath=&SizeBox
   parent=HorizontalBoxTopDoubleBindArrows
   widget=SizeBoxComboArrow2
   properties {bOverride_HeightOverride=true, bOverride_WidthOverride=true, heightOverride=32, widthOverride=32}
   slot {class=HorizontalBoxSlot, properties {horizontalAlignment=HAlign_Right, verticalAlignment=VAlign_Center}}
  }
  KeybindWidgetComboArrow2 {
   widgetClassPath=&KeybindWidget
   bInherited=true
   parent=SizeBoxComboArrow2
   widget=KeybindWidgetComboArrow2
   properties {boundAction=GamepadAbility2, visibility=Collapsed}
   slot {class=SizeBoxSlot}
  }
  RightArrowImage2 {
   widgetClassPath=&Image
   bInherited=true
   parent=HorizontalBoxTopDoubleBindArrows
   widget=RightArrowImage2
   properties {
    visibility=Collapsed
    brush {
     bIsDynamicallyLoaded=false
     drawAs=Image
     imageType=NoImage
     mirroring=NoMirror
     resourceName=None
     resourceObject=&T-HUD-Slider-Arrow
     tiling=NoTile
     imageSize=<16 16>
     margin=<ltrb 0 0 0 0>
     outlineSettings {
      bUseBrushTransparency=false
      roundingType=HalfHeightRadius
      width=0
      color {colorUseRule=UseColor_Specified, specifiedColor=<rgba 0 0 0 0>}
      cornerRadii=<0 0 0 0>
     }
     tintColor {colorUseRule=UseColor_Specified, specifiedColor=<rgba 1 1 1 1>}
     uVRegion {bIsValid=false, max=<0 0>, min=<0 0>}
    }
    renderTransform {angle=180, scale=<1 1>, shear=<0 0>, translation=<0 0>}
   }
   slot {class=HorizontalBoxSlot, properties {horizontalAlignment=HAlign_Right, verticalAlignment=VAlign_Center}}
  }
  HoldText {widgetClassPath=&CommonTextBlock, bInherited=true, widget=null}
 ]
}
