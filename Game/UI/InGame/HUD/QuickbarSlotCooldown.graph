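# Quickbar slot cooldown HUD widget, built on FortItemCooldownWidget, layering radial or rectangular shading, cooldown images and countdown text.
# Initialize stores supported cooldown types and display style and picks the radial material, or hides the widget if none are given.
# Start/stop events switch on style and type to show or collapse shading; in Athena, unavailable ammo regeneration shows rectangular dimming.
# PreConstruct applies smaller, offset styling for mobile touch.
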WidgetBlueprint /Game/UI/InGame/HUD/QuickbarSlotCooldown
graphs [
 EventGraph {nodes [
   n1 = Event OnAmmoRegenerationUpdated!() -> (OutputDelegate, then, bOutAmmoRegenAvailableState:bool) {is=title K2Node_Event}
   > Branch(Condition=IsInAthena!() -> (ReturnValue:bool))
   > SetActiveWidgetIndex(self=Get SwitcherOpacityOverlay() -> (SwitcherOpacityOverlay:Common Widget Switcher Legacy Object Reference), Index=1)
   > SetVisibility(
    self=Get SwitcherOpacityOverlay()
    InVisibility=Select(Option 0=SelfHitTestInvisible, Option 1=Hidden, Index=@n1.bOutAmmoRegenAvailableState)
   )
   Event OnInitialized()
   > Branch()
   Event PreConstruct()
   > SetMobileStyle() {is=title K2Node_CallFunction}
   n2 = Event OnCooldownStarted!() -> (OutputDelegate, then, CooldownType:EFortItemCooldownType Enum=None, CooldownDuration:float) {is=title K2Node_Event}
   > n3 = Switch on ItemDisplayStyle!(execute, Selection:ItemDisplayStyle Enum=Get ItemDisplayStyle() -> (ItemDisplayStyle:ItemDisplayStyle Enum=NewEnumerator2)) -> (
    NewEnumerator2:Exec
    NewEnumerator3:Exec
    NewEnumerator4:Exec
    NewEnumerator5:Exec
    NewEnumerator6:Exec
    NewEnumerator7:Exec
    NewEnumerator8:Exec
   ) {is=title K2Node_SwitchEnum}
   n4 = Switch on EFortItemCooldownType!(execute=@n3.NewEnumerator3, Selection:EFortItemCooldownType Enum=@n2.CooldownType) -> (
    None:Exec
    AmmoRegeneration:Exec
    ItemActivation:Exec
    WeaponReloading:Exec
    Death:Exec
    AthenaWeaponFireCooldown:Exec
    AbilitySetActivateByInputAbility:Exec
    DirectProgressUpdate:Exec
   ) {is=title K2Node_SwitchEnum}
   n5 = SetActiveWidgetIndex(execute=[@n4.Death, @n4.DirectProgressUpdate, @n4.WeaponReloading], self=Get SwitcherOpacityOverlay(), Index=1)
   SetVisibility(execute=[@n5.then, @n6.then], self=Get SwitcherOpacityOverlay(), InVisibility=HitTestInvisible)
   n6 = SetActiveWidgetIndex(
    execute=[@n7.AbilitySetActivateByInputAbility, @n7.Death, @n7.DirectProgressUpdate, @n7.ItemActivation, @n4.ItemActivation]
    self=Get SwitcherOpacityOverlay()
    Index=0
   )
   n7 = Switch on EFortItemCooldownType(execute=@n3.NewEnumerator5, Selection=@n2.CooldownType)
   n8 = Event OnCooldownStopped!() -> (OutputDelegate, then, CooldownType:EFortItemCooldownType Enum=None) {is=title K2Node_Event}
   > n9 = Switch on ItemDisplayStyle(Selection=Get ItemDisplayStyle() -> (ItemDisplayStyle=NewEnumerator2))
   n10 = Switch on EFortItemCooldownType(execute=@n9.NewEnumerator3, Selection=@n8.CooldownType)
   SetVisibility(
    execute=[
     @n10.AthenaWeaponFireCooldown
     @n10.Death
     @n10.DirectProgressUpdate
     @n10.ItemActivation
     @n10.WeaponReloading
     @n11.AbilitySetActivateByInputAbility
     @n11.Death
     @n11.DirectProgressUpdate
     @n11.ItemActivation
    ]
    self=Get SwitcherOpacityOverlay()
    InVisibility=Collapsed
   )
   n11 = Switch on EFortItemCooldownType(execute=@n9.NewEnumerator5, Selection=@n8.CooldownType)
  ]}
 Initialize {nodes [
   n1 = Initialize!() -> (then, QuickbarItemDisplayStyle:ItemDisplayStyle Enum=NewEnumerator2, DesiredCooldownTypesSupported:Array of EFortItemCooldownType Enums=None) {is=title K2Node_FunctionEntry}
   > n2 = Set CooldownTypesSupported(CooldownTypesSupported:Array of EFortItemCooldownType Enums=@n1.DesiredCooldownTypesSupported) -> (Output_Get:Array of EFortItemCooldownType Enums=None)
   > n3 = Branch(Condition="integer > integer"(A=Length(TargetArray:Array of EFortItemCooldownType Enums=@n2), B=0) {is=title K2Node_CallFunction})
   > n4 = Set ItemDisplayStyle(ItemDisplayStyle:ItemDisplayStyle Enum=@n1) -> (Output_Get:ItemDisplayStyle Enum=NewEnumerator2)
   > SetBrushFromMaterial(
    self=Get ImageRadialOpacityLayer() -> (ImageRadialOpacityLayer:Image Object Reference)
    Material=Select!(
     NewEnumerator2:Material Interface Object Reference
     NewEnumerator3:Material Interface Object Reference=/Game/UI/Foundation/Textures/Icons/HUD/M-HUD-Slot-Cooldown
     NewEnumerator4:Material Interface Object Reference
     NewEnumerator5:Material Interface Object Reference=&M-HUD-Slot-Cooldown
     NewEnumerator6:Material Interface Object Reference
     NewEnumerator7:Material Interface Object Reference
     NewEnumerator8:Material Interface Object Reference
     Index:ItemDisplayStyle Enum=@n4
    ) -> (ReturnValue:Material Interface Object Reference)
   )
   > Branch(Condition=IsInAthena())
   > SetStretch!(execute, self:Scale Box Object Reference=Get ScaleBoxCDCount() -> (ScaleBoxCDCount:Scale Box Object Reference), InStretch:EStretch Enum=None) -> (then)
   SetVisibility(execute=@n3.else, InVisibility=Hidden)
  ]}
 InitializeAthenaCooldown {nodes [
   InitializeAthenaCooldown() {is=title K2Node_FunctionEntry}
   n1 = Get SizeBox_Cooldown() -> (SizeBox_Cooldown:Size Box Object Reference)
   n2 = SlotAsOverlaySlot(Widget=@n1)
   > SetHorizontalAlignment(self=@n2, InHorizontalAlignment=HAlign_Center)
   > SetVerticalAlignment(self=@n2, InVerticalAlignment=VAlign_Center)
   n3 = Get Athena_CooldownSizeOverride() -> (Athena_CooldownSizeOverride:float)
   > SetHeightOverride(self=@n1, InHeightOverride=@n3)
   > SetWidthOverride(self=@n1, InWidthOverride=@n3)
   n4 = Get ImageRadialOpacityLayer()
   > SetBrush(self=@n4, InBrush=Get Athena_RadialOpacityBrush() -> (Athena_RadialOpacityBrush:Slate Brush Structure))
   > n5 = Cast To WidgetSwitcherSlot!(execute, Object=Get Slot!(self:Widget Object Reference=@n4) -> (Slot:Panel Slot Object Reference)) -> (then, CastFailed, AsWidget Switcher Slot:Widget Switcher Slot Object Reference) {is=title K2Node_DynamicCast}
   > SetHorizontalAlignment(self:Widget Switcher Slot Object Reference=@n5, InHorizontalAlignment=HAlign_Center)
   > SetVerticalAlignment(self:Widget Switcher Slot Object Reference=@n5, InVerticalAlignment=VAlign_Center)
  ]}
 SetMobileStyle {nodes [
   SetMobileStyle() {is=title K2Node_FunctionEntry}
   n1 = GetOwningPlayer()
   > Branch(Condition=AND Boolean(A=IsMobileGame(), B=IsUsingTouch(PlayerController=@n1)))
   n2 = Get SizeBox_Cooldown()
   > SetMaxDesiredHeight!(execute, self:Size Box Object Reference=@n2, InMaxDesiredHeight:float=0) -> (then)
   > SetMaxDesiredWidth!(execute, self:Size Box Object Reference=@n2, InMaxDesiredWidth:float=0) -> (then)
   > SetRenderTransform!(
    execute
    self:Widget Object Reference=[@n2, Get ImageRadialOpacityLayer()]
    InTransform_Translation:Vector 2D Structure=(X=0,Y=0)
    InTransform_Scale:Vector 2D Structure=(X=0.65,Y=0.65)
    InTransform_Shear:Vector 2D Structure=(X=0,Y=0)
    InTransform_Angle:float=0
   ) -> (then)
   > SetRenderTranslation(
    self=[Get SizeBox_Cooldown(), Get SizeBox_0() -> (SizeBox_0:Size Box Object Reference), Get ImageRadialOpacityLayer()]
    Translation=(X=0,Y=6)
   )
  ]}
 Show Countdown {nodes [
   n1 = Show Countdown!() -> (then, Show:bool=~) {is=title K2Node_FunctionEntry}
   > SetVisibility(
    self=Get CooldownText() -> (CooldownText:Common Text Object Reference)
    InVisibility=Select!(Option 0:ESlateVisibility Enum=Collapsed, Option 1:ESlateVisibility Enum=SelfHitTestInvisible, Index:bool=@n1) -> (ReturnValue:ESlateVisibility Enum)
   )
  ]}
]
variables [
 Athena_CooldownSizeOverride {type=number, default=54}
 Athena_RadialOpacityBrush {
  type=SlateBrush
  default {
   bIsDynamicallyLoaded=false
   drawAs=Image
   imageType=FullColor
   mirroring=NoMirror
   resourceName=None
   resourceObject=&M-HUD-Slot-Cooldown
   tiling=NoTile
   imageSize=<54 54>
   margin=<ltrb 0 0 0 0>
   outlineSettings {
    bUseBrushTransparency=false
    roundingType=HalfHeightRadius
    width=0
    color {colorUseRule=UseColor_Specified, specifiedColor=<rgba 0 0 0 0>}
    cornerRadii=<0 0 0 0>
   }
   tintColor {colorUseRule=UseColor_Specified, specifiedColor=<rgba 1 1 1 1>}
   uVRegion {bIsValid=false, max=<0 0>, min=<0 0>}
  }
 }
 ItemDisplayStyle {type=ItemDisplayStyle, default=None}
]
widgets {
 info {inheritedWidgetCount=3, parentClass=/Script/FortniteUI.FortItemCooldownWidget, rootWidgetClass=&Overlay, widgetCount=10}
 widgets [
  Overlay_0 {widgetClassPath=&Overlay, bIsVariable=true, widget=Overlay_0}
  SwitcherOpacityOverlay {
   widgetClassPath=&CommonWidgetSwitcherLegacy
   bIsVariable=true
   parent=Overlay_0
   widget=SwitcherOpacityOverlay
   properties {transitionDuration=0, visibility=Collapsed}
   slot {class=OverlaySlot, properties {horizontalAlignment=HAlign_Fill, verticalAlignment=VAlign_Fill}}
  }
  ImageRadialOpacityLayer {
   widgetClassPath=&Image
   bIsVariable=true
   parent=SwitcherOpacityOverlay
   widget=ImageRadialOpacityLayer
   properties {brush {
     bIsDynamicallyLoaded=false
     drawAs=Image
     imageType=NoImage
     mirroring=NoMirror
     resourceName=None
     resourceObject=&M-HUD-Slot-Cooldown
     tiling=NoTile
     imageSize=<32 32>
     margin=<ltrb 0 0 0 0>
     outlineSettings {
      bUseBrushTransparency=false
      roundingType=HalfHeightRadius
      width=0
      color {colorUseRule=UseColor_Specified, specifiedColor=<rgba 0 0 0 0>}
      cornerRadii=<0 0 0 0>
     }
     tintColor {colorUseRule=UseColor_Specified, specifiedColor=<rgba 1 1 1 1>}
     uVRegion {bIsValid=false, max=<0 0>, min=<0 0>}
    }}
   slot {class=WidgetSwitcherSlot}
  }
  ImageRectangularOpacityLayer {
   widgetClassPath=&Image
   bIsVariable=true
   parent=SwitcherOpacityOverlay
   widget=ImageRectangularOpacityLayer
   properties {colorAndOpacity=<rgba 0 0 0 0.5>}
   slot {class=WidgetSwitcherSlot}
  }
  SizeBox_Cooldown {
   widgetClassPath=&SizeBox
   bIsVariable=true
   parent=Overlay_0
   widget=SizeBox_Cooldown
   properties {
    bOverride_MaxAspectRatio=true
    bOverride_MaxDesiredHeight=true
    bOverride_MaxDesiredWidth=true
    bOverride_MinAspectRatio=true
    bOverride_MinDesiredHeight=true
    bOverride_MinDesiredWidth=true
    maxDesiredHeight=92
    maxDesiredWidth=92
    minDesiredHeight=92
    minDesiredWidth=92
   }
   slot {class=OverlaySlot, properties {horizontalAlignment=HAlign_Fill, verticalAlignment=VAlign_Fill}}
  }
  CooldownImage {
   widgetClassPath=&Image
   bInherited=true
   bIsVariable=true
   parent=SizeBox_Cooldown
   widget=CooldownImage
   properties {brush {
     bIsDynamicallyLoaded=false
     drawAs=Image
     imageType=NoImage
     mirroring=NoMirror
     resourceName=None
     resourceObject=/Game/UI/InGame/HUD/HoldFeedbackRadial/M_Cooldown_Radial_Charges
     tiling=NoTile
     imageSize=<120 120>
     margin=<ltrb 0 0 0 0>
     outlineSettings {
      bUseBrushTransparency=false
      roundingType=HalfHeightRadius
      width=0
      color {colorUseRule=UseColor_Specified, specifiedColor=<rgba 0 0 0 0>}
      cornerRadii=<0 0 0 0>
     }
     tintColor {colorUseRule=UseColor_Specified, specifiedColor=<rgba 1 1 1 1>}
     uVRegion {bIsValid=false, max=<0 0>, min=<0 0>}
    }}
   slot {class=SizeBoxSlot}
  }
  CooldownChargingImage {
   widgetClassPath=&Image
   bInherited=true
   bIsVariable=true
   parent=Overlay_0
   widget=CooldownChargingImage
   properties {brush {
     bIsDynamicallyLoaded=false
     drawAs=Image
     imageType=NoImage
     mirroring=NoMirror
     resourceName=None
     resourceObject=/Game/UI/Foundation/Textures/Icons/Stats/T-Icon-HealthBattery-64
     tiling=NoTile
     imageSize=<64 64>
     margin=<ltrb 0 0 0 0>
     outlineSettings {
      bUseBrushTransparency=false
      roundingType=HalfHeightRadius
      width=0
      color {colorUseRule=UseColor_Specified, specifiedColor=<rgba 0 0 0 0>}
      cornerRadii=<0 0 0 0>
     }
     tintColor {colorUseRule=UseColor_Specified, specifiedColor=<rgba 1 1 1 1>}
     uVRegion {bIsValid=false, max=<0 0>, min=<0 0>}
    }}
   slot {class=OverlaySlot, properties {horizontalAlignment=HAlign_Center, verticalAlignment=VAlign_Center}}
  }
  SizeBox_0 {
   widgetClassPath=&SizeBox
   bIsVariable=true
   parent=Overlay_0
   widget=SizeBox_0
   properties {bOverride_WidthOverride=true, widthOverride=85}
   slot {class=OverlaySlot, properties {horizontalAlignment=HAlign_Center, verticalAlignment=VAlign_Center}}
  }
  ScaleBoxCDCount {widgetClassPath=&ScaleBox, bIsVariable=true, parent=SizeBox_0, widget=ScaleBoxCDCount, slot {class=SizeBoxSlot}}
  CooldownText {
   widgetClassPath=&CommonTextBlock
   bInherited=true
   bIsVariable=true
   parent=ScaleBoxCDCount
   widget=CooldownText
   properties {
    justification=Center
    minDesiredWidth=75
    mobileFontSizeMultiplier=0.7
    text=66
    font {
     bForceMonospaced=false
     bMaterialIsStencil=false
     fontMaterial=null
     fontObject=&Burbank
     letterSpacing=0
     monospacedWidth=1
     size=34
     skewAmount=0
     typefaceFontName=Burbank-PlayerPower
     outlineSettings {
      bApplyOutlineToDropShadows=true
      bMiteredCorners=false
      bSeparateFillAlpha=true
      outlineMaterial=&M-FontShadow
      outlineSize=2
      outlineColor=<rgba 0 0 0 0.8>
     }
    }
    renderTransform {angle=0, scale=<1 1>, shear=<-7 0>, translation=<0 0>}
    shadowOffset=<0 0>
   }
   slot {class=ScaleBoxSlot}
  }
 ]
}
