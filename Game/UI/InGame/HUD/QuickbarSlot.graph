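# HUD quickbar slot widget, built on FortQuickBarSlotBase, showing an item card, keybind, selection frame and cooldown.
# Maximize and minimize animations drive AnimatedBoxScaleFactor, whose setter lerps the SizeBox between minimized and maximized card dimensions plus padding.
# Selection toggles the frame and plays a lift animation, skipped on simplified HUD.
# Quickbar type picks the keybind position, and cooldown events can hide keybinds during death or item-activation cooldowns.
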WidgetBlueprint /Game/UI/InGame/HUD/QuickbarSlot
animations [
 AnimMinimize {
  binding_count=1
  display_label=""
  length_seconds=0.1
  object=&self:AnimMinimize
  sequence {bindings [QuickbarSlot_C_28 {
     id=#1
     possessed_class=QuickbarSlot_C
     tracks [{
       type=MovieSceneFloatTrack
       display_name=Animated Box Scale Factor
       sections [{type=MovieSceneFloatSection, end_frame=2, start_frame=0, channels [None {type=MovieSceneScriptingFloatChannel, keys [[0,1],[2,0]]}]}]
      }]
    }]}
 }
 AnimMaximize {
  binding_count=1
  display_label=""
  length_seconds=0.1
  object=&self:AnimMaximize
  sequence {bindings [QuickbarSlot_C_29 {
     id=#2
     possessed_class=QuickbarSlot_C
     tracks [{
       type=MovieSceneFloatTrack
       display_name=Animated Box Scale Factor
       sections [{type=MovieSceneFloatSection, end_frame=2, start_frame=0, channels [None {type=MovieSceneScriptingFloatChannel, keys [[0,0],[2,1]]}]}]
      }]
    }]}
 }
 Selected {
  binding_count=1
  display_label=""
  length_seconds=0.1
  object=&self:Selected
  sequence {bindings [Overlay_Content {
     id=#3
     possessed_class=Overlay
     tracks [{
       type=MovieScene2DTransformTrack
       sections [{
         type=MovieScene2DTransformSection
         end_frame=2
         start_frame=0
         channels [
          Translation.X {type=MovieSceneScriptingFloatChannel, const [0,2]}
          Translation.Y {type=MovieSceneScriptingFloatChannel, keys [[0,0],[2,-7]]}
          Angle {type=MovieSceneScriptingFloatChannel}
          Scale.X {type=MovieSceneScriptingFloatChannel}
          Scale.Y {type=MovieSceneScriptingFloatChannel}
          Shear.X {type=MovieSceneScriptingFloatChannel}
          Shear.Y {type=MovieSceneScriptingFloatChannel}
         ]
        }]
      }]
    }]}
 }
]
graphs [
 EventGraph {nodes [
   n1 = SetAnimatedBoxScaleFactor!() -> (OutputDelegate, then, AnimatedBoxScaleFactor:double=~) {is=title K2Node_CustomEvent}
   > Set AnimatedBoxScaleFactor(AnimatedBoxScaleFactor:double=@n1.AnimatedBoxScaleFactor) -> (Output_Get:double)
   > UpdateItemBoxSize()
   n2 = HandleMaximizeStarted() {is=title K2Node_CustomEvent}
   > Resize!(
    execute
    self:Fort Quick Bar Slot Base Object Reference
    CardSize:EFortItemCardSize Enum=GetCardSize!(self:Fort Multi Size Item Card Object Reference=Get ItemCardMaximized() -> (ItemCardMaximized:Fort Multi Size Item Card Object Reference)) -> (ReturnValue:EFortItemCardSize Enum=XXS)
   ) -> (then)
   n3 = HandleMinimizeFinished() {is=title K2Node_CustomEvent}
   > Resize(CardSize=GetCardSize(self=Get ItemCardMinimized() -> (ItemCardMinimized:Fort Multi Size Item Card Object Reference)) -> (ReturnValue=XXS))
   n4 = Event MaximizeSlot!() -> (OutputDelegate, then, bShouldSkipAnimation:bool) {is=title K2Node_Event}
   n5 = Get AnimMinimize() -> (AnimMinimize:Widget Animation Object Reference)
   > n6 = Branch(Condition=IsAnimationPlaying(InAnimation=@n5))
   > n7 = StopAnimation(InAnimation=@n5)
   n8 = Branch(execute=[@n7.then, @n6.else], Condition=@n4.bShouldSkipAnimation)
   > HandleMaximizeStarted() {is=title K2Node_CallFunction}
   > SetAnimatedBoxScaleFactor!(execute, self, AnimatedBoxScaleFactor:double=1) -> (then) {is=title K2Node_CallFunction}
   > HandleMaximizeFinished()
   PlayAnimation(execute=@n8.else, InAnimation=Get AnimMaximize() -> (AnimMaximize:Widget Animation Object Reference))
   > n9 = Cast To QuickbarSlotCooldown!(execute, Object=Get QuickbarSlotCooldown() -> (QuickbarSlotCooldown:Fort Item Cooldown Widget Object Reference)) -> (then, CastFailed, AsQuickbar Slot Cooldown:Quickbar Slot Cooldown Object Reference) {is=title K2Node_DynamicCast}
   > Show Countdown!(execute, self:Quickbar Slot Cooldown Object Reference=@n9, Show:bool=true) -> (then)
   n10 = Event MinimizeSlot!() -> (OutputDelegate, then, bShouldSkipAnimation:bool) {is=title K2Node_Event}
   n11 = Get AnimMaximize()
   > n12 = Branch(Condition=IsAnimationPlaying(InAnimation=@n11))
   > n13 = StopAnimation(InAnimation=@n11)
   n14 = Branch(execute=[@n13.then, @n12.else], Condition=@n10.bShouldSkipAnimation)
   > HandleMinimizeStarted()
   > SetAnimatedBoxScaleFactor(AnimatedBoxScaleFactor=0)
   > HandleMinimizeFinished() {is=title K2Node_CallFunction}
   PlayAnimation(execute=@n14.else, InAnimation=Get AnimMinimize())
   > n15 = Cast To QuickbarSlotCooldown(Object=Get QuickbarSlotCooldown())
   > Show Countdown(self=@n15, Show=false)
   Event PreConstruct()
   > n16 = Branch(Condition=Get ShouldUseDefaultsMinMaxCardSizes() -> (ShouldUseDefaultsMinMaxCardSizes:bool))
   > SetCardSize!(
    execute
    self:Fort Multi Size Item Card Object Reference=Get ItemCardMaximized()
    CardSize:EFortItemCardSize Enum=Get MaximumCardSize() -> (MaximumCardSize:EFortItemCardSize Enum=XXS)
   ) -> (then)
   > n17 = SetCardSize(self=Get ItemCardMinimized(), CardSize=Get MinimumCardSize() -> (MinimumCardSize:EFortItemCardSize Enum=XXS))
   SetShouldCollapseBorderPadding!(
    execute=[@n17.then, @n16.else]
    self:Fort Multi Size Item Card Object Reference=[Get ItemCardMaximized(), Get ItemCardMinimized()]
    Value:bool=Get ShouldCollapseItemWidgetBorder() -> (ShouldCollapseItemWidgetBorder:bool)
   ) -> (then)
   > UpdateItemBoxSize()
   > UpdateKeyBindingText(self:Fort Quick Bar Slot Base Object Reference)
   > n18 = Switch on EFortQuickBars!(execute, Selection:EFortQuickBars Enum=Get QuickBarType() -> (QuickBarType:EFortQuickBars Enum=Primary)) -> (Primary:Exec, Secondary:Exec, Creative:Exec, Max_None:Exec) {is=title K2Node_SwitchEnum}
   > SetVisibility(
    self=Get KeybindWidget!(self:Fort Quickbar Keybind Widget Object Reference=Get QuickbarKeybindWidgetTop() -> (QuickbarKeybindWidgetTop:Fort Quickbar Keybind Widget Object Reference)) -> (KeybindWidget:Fort Keybind Widget Object Reference)
    InVisibility=Collapsed
   )
   SetVisibility(
    execute=@n18.Secondary
    self=Get KeybindWidget(self=Get QuickbarKeybindWidgetBottom() -> (QuickbarKeybindWidgetBottom:Fort Quickbar Keybind Widget Object Reference))
    InVisibility=Collapsed
   )
   Event Construct()
   > InitializeInvalidation()
   > InitializeCooldowns()
   > RefreshItem(self:Fort Quick Bar Slot Base Object Reference)
   n19 = Get Selection() -> (Selection:Image Object Reference)
   > SetBrushFromTexture!(
    execute
    self:Image Object Reference=[Get EmptyImage() -> (EmptyImage:Image Object Reference), @n19]
    Texture:Texture 2D Object Reference=Select!(
     NewEnumerator2:Texture 2D Object Reference=/Game/UI/InGame/Quickbar/Quickbar-Frame-Default
     NewEnumerator3:Texture 2D Object Reference=&Quickbar-Frame-Default
     NewEnumerator4:Texture 2D Object Reference
     NewEnumerator5:Texture 2D Object Reference=/Game/UI/InGame/Quickbar/Quickbar-Frame-Ability
     NewEnumerator6:Texture 2D Object Reference
     NewEnumerator7:Texture 2D Object Reference
     NewEnumerator8:Texture 2D Object Reference
     Index:ItemDisplayStyle Enum=Get ItemDisplayStyle() -> (ItemDisplayStyle:ItemDisplayStyle Enum=NewEnumerator2)
    ) -> (ReturnValue:Texture 2D Object Reference)
    bMatchSize:bool=false
   ) -> (then)
   > SetVisibility(self=@n19, InVisibility=Hidden)
   n20 = Get AnimMaximize()
   > BindToAnimationStarted!(execute, self:User Widget Object Reference, Animation:Widget Animation Object Reference=@n20, Delegate:Delegate=@n2) -> (then)
   > BindToAnimationFinished(Animation=@n20, Delegate=HandleMaximizeFinished() {is=title K2Node_CustomEvent})
   n21 = Get AnimMinimize()
   > BindToAnimationStarted(Animation=@n21, Delegate=HandleMinimizeStarted() {is=title K2Node_CustomEvent})
   > BindToAnimationFinished(Animation=@n21, Delegate=@n3)
   n22 = Event SetSlotSelected!() -> (OutputDelegate, then, bSelected:bool) {is=title K2Node_Event}
   > SetVisibility(
    self=Get Selection()
    InVisibility=Select!(Option 0:ESlateVisibility Enum=Collapsed, Option 1:ESlateVisibility Enum=Visible, Index:bool=@n22.bSelected) -> (ReturnValue:ESlateVisibility Enum)
   )
   > n23 = Branch(Condition=@n22.bSelected)
   n24 = Get DoNotPlaySelectionAnimation() -> (DoNotPlaySelectionAnimation:bool)
   > n25 = Branch(Condition=@n24)
   PlayAnimation(execute=@n25.else, InAnimation=Get Selected() -> (Selected:Widget Animation Object Reference))
   n26 = Branch(execute=@n23.else, Condition=@n24)
   n27 = Get Selected()
   PlayAnimation(execute=@n26.else, InAnimation=@n27, StartAtTime=GetEndTime(self=@n27), PlayMode=Reverse)
   Event OnRefreshItem() {is=title K2Node_Event}
   > TryRefreshCooldowns()
  ]}
 InitializeCooldowns {nodes [
   InitializeCooldowns() {is=title K2Node_FunctionEntry}
   > n1 = Cast To QuickbarSlotCooldown(Object=Get QuickbarSlotCooldown())
   > Initialize!(
    execute
    self:Quickbar Slot Cooldown Object Reference=@n1
    QuickbarItemDisplayStyle:ItemDisplayStyle Enum=Get ItemDisplayStyle() -> (ItemDisplayStyle=NewEnumerator2)
    DesiredCooldownTypesSupported:Array of EFortItemCooldownType Enums=Get CooldownTypesSupported() -> (CooldownTypesSupported:Array of EFortItemCooldownType Enums=None)
   ) -> (then)
   > Branch(Condition=Get bHideKeybindsWhenAbilityUnavailable() -> (bHideKeybindsWhenAbilityUnavailable:bool))
   n2 = Get QuickbarSlotCooldown()
   > Bind Event to OnCooldownStarted(self:Fort Item Cooldown Widget Object Reference=@n2, Delegate=Create Event()) {is=title K2Node_AddDelegate}
   > Bind Event to OnCooldownStopped(self:Fort Item Cooldown Widget Object Reference=@n2, Delegate=Create Event()) {is=title K2Node_AddDelegate}
  ]}
 InitializeInvalidation {nodes [
   InitializeInvalidation() {is=title K2Node_FunctionEntry}
   > Is Simplified HUD!(Exec:Exec) -> (Simplified:Exec, Default:Exec) {is=title K2Node_MacroInstance}
   > Set DoNotPlaySelectionAnimation(DoNotPlaySelectionAnimation:bool=true) -> (Output_Get:bool)
  ]}
 OnCooldownStarted {nodes [
   n1 = OnCooldownStarted!() -> (then, CooldownType:EFortItemCooldownType Enum=None, CooldownDuration:float=~) {is=title K2Node_FunctionEntry}
   > n2 = Switch on EFortItemCooldownType!(execute, Selection:EFortItemCooldownType Enum=@n1) -> (
    None:Exec
    AmmoRegeneration:Exec
    ItemActivation:Exec
    WeaponReloading:Exec
    Death:Exec
    AthenaWeaponFireCooldown:Exec
    AbilitySetActivateByInputAbility:Exec
    DirectProgressUpdate:Exec
   ) {is=title K2Node_SwitchEnum}
   Set bKeybindsHidden(execute=[@n2.Death, @n2.ItemActivation], bKeybindsHidden:bool=true) -> (Output_Get:bool)
   > UpdateKeyBindingVisibility(self:Fort Quick Bar Slot Base Object Reference)
  ]}
 OnCooldownStopped {nodes [
   n1 = OnCooldownStopped!() -> (then, CooldownType:EFortItemCooldownType Enum=None) {is=title K2Node_FunctionEntry}
   > n2 = Switch on EFortItemCooldownType(Selection=@n1)
   Set bKeybindsHidden(execute=[@n2.Death, @n2.ItemActivation], bKeybindsHidden=false)
   > UpdateKeyBindingVisibility()
  ]}
 TryRefreshCooldowns {nodes [
   TryRefreshCooldowns() {is=title K2Node_FunctionEntry}
   > n1 = Cast To QuickbarSlotCooldown(Object=Get QuickbarSlotCooldown())
   > TryStartUpdateCooldown_BP(self:Fort Item Cooldown Widget Object Reference=@n1)
  ]}
 UpdateItemBoxSize {nodes [
   UpdateItemBoxSize() {is=title K2Node_FunctionEntry}
   n1 = Get Item() -> (Item:Fort Item Object Reference)
   > n2 = Is Valid(InputObject=@n1)
   > n3 = Set Item Type To Show(Item Type To Show:EFortItemType Enum=GetType!(self:Fort Item Object Reference=@n1) -> (ReturnValue:EFortItemType Enum=WorldItem)) -> (Output_Get:EFortItemType Enum=WorldItem)
   n4 = Sequence(execute=[@n3.then, @n2.Is Not Valid])
   n5 = Get ItemCardMaximized()
   n6 = Break Margin!(Margin:"Margin Structure (by ref)"=Get Padding!(self:Overlay Slot Object Reference=SlotAsOverlaySlot(Widget=@n5)) -> (Padding:Margin Structure)) -> (Left:float=~, Top:float=~, Right:float=~, Bottom:float=~) {is=title K2Node_BreakStruct}
   n7 = Get EmptyItemType!(self:Fort Multi Size Item Card Object Reference=@n5) -> (EmptyItemType:EFortItemType Enum=WorldItem)
   n8 = IsValid(Object=@n1)
   n9 = ShouldUseNewItemCards!() -> (ReturnValue:bool)
   n10 = BreakVector2D(InVec=GetCardDimensions!(
     ItemType:EFortItemType Enum=Select(Option 0:EFortItemType Enum=@n7, Option 1:EFortItemType Enum=@n3, Index=@n8) -> (ReturnValue:EFortItemType Enum)
     CardSize:EFortItemCardSize Enum=GetCardSize(self=@n5) -> (ReturnValue=XXS)
     IsReward:bool=Get ShouldDisplayItemAsReward!(self:Fort Multi Size Item Card Object Reference=@n5) -> (ShouldDisplayItemAsReward:bool)
     UseNewItemCardSizes:bool=@n9
    ) -> (ReturnValue:Vector 2D Structure))
   n11 = Get ItemCardMinimized()
   n12 = Break Margin(Margin=Get Padding(self=SlotAsOverlaySlot(Widget=@n11))) -> (Left=~, Top=~, Right=~, Bottom=~)
   n13 = Get EmptyItemType(self=@n11) -> (EmptyItemType=WorldItem)
   n14 = BreakVector2D(InVec=GetCardDimensions(
     ItemType=Select(Option 0=@n13, Option 1=@n3, Index=@n8)
     CardSize=GetCardSize(self=@n11) -> (ReturnValue=XXS)
     IsReward=Get ShouldDisplayItemAsReward(self=@n11)
     UseNewItemCardSizes=@n9
    ))
   n15 = Get AnimatedBoxScaleFactor() -> (AnimatedBoxScaleFactor:double)
   n16 = "Get [SizeBox] Item"() -> ("[SizeBox] Item":Size Box Object Reference)
   > SetWidthOverride(
    self=@n16
    InWidthOverride=Lerp(
     A=float + float!(A=@n12, B=@n12.Right, C:double=@n14) -> (ReturnValue) {is=title K2Node_CommutativeAssociativeBinaryOperator}
     B=float + float(A=@n6, B=@n6.Right, C=@n10)
     Alpha=@n15
    )
   )
   SetHeightOverride(
    execute=@n4.then_1
    self=@n16
    InHeightOverride=Lerp(A=float + float(A=@n12.Top, B=@n12.Bottom, C=@n14.Y), B=float + float(A=@n6.Top, B=@n6.Bottom, C=@n10.Y), Alpha=@n15)
   )
  ]}
]
variables [
 AnimatedBoxScaleFactor {type=number, default=1}
 CooldownTypesSupported {type="array<EFortItemCooldownType>", default []}
 ItemDisplayStyle {type=ItemDisplayStyle, default=World}
 MaximumCardSize {type=EFortItemCardSize, default=S}
 MinimumCardSize {type=EFortItemCardSize, default=XS}
 ShouldCollapseItemWidgetBorder {type=boolean}
 ShouldUseDefaultsMinMaxCardSizes {type=boolean}
]
widgets {
 info {inheritedWidgetCount=9, parentClass=/Script/FortniteUI.FortQuickBarSlotBase, rootWidgetClass=&Overlay, widgetCount=15}
 widgets [
  Overlay_0 {widgetClassPath=&Overlay, widget=Overlay_0}
  VerticalBox_0 {widgetClassPath=&VerticalBox, parent=Overlay_0, widget=VerticalBox_0, slot {class=OverlaySlot}}
  QuickbarKeybindWidgetTop {
   widgetClassPath=&/QuickbarKeybindTop
   bInherited=true
   bIsVariable=true
   parent=VerticalBox_0
   widget=QuickbarKeybindWidgetTop
   slot {class=VerticalBoxSlot, properties {horizontalAlignment=HAlign_Center, padding=<ltrb 0 0 0 16>}}
  }
  "[SizeBox] Item" {
   widgetClassPath=&SizeBox
   bIsVariable=true
   parent=VerticalBox_0
   widget="[SizeBox] Item"
   properties {heightOverride=80, widthOverride=80}
   slot {class=VerticalBoxSlot, properties {horizontalAlignment=HAlign_Center, verticalAlignment=VAlign_Center}}
  }
  Overlay_Content {widgetClassPath=&Overlay, parent="[SizeBox] Item", widget=Overlay_Content, slot {class=SizeBoxSlot}}
  Selection {
   widgetClassPath=&Image
   bIsVariable=true
   parent=Overlay_Content
   widget=Selection
   properties {brush {
     bIsDynamicallyLoaded=false
     drawAs=Image
     imageType=NoImage
     mirroring=NoMirror
     resourceName=None
     resourceObject=/Game/UI/Icons/Icon-BasicFrame-S
     tiling=NoTile
     imageSize=<24 24>
     margin=<ltrb 0.5 0.5 0.5 0.5>
     outlineSettings {
      bUseBrushTransparency=false
      roundingType=HalfHeightRadius
      width=0
      color {colorUseRule=UseColor_Specified, specifiedColor=<rgba 0 0 0 0>}
      cornerRadii=<0 0 0 0>
     }
     tintColor {colorUseRule=UseColor_Specified, specifiedColor=<rgba 1 1 1 1>}
     uVRegion {bIsValid=false, max=<0 0>, min=<0 0>}
    }}
   slot {class=OverlaySlot, properties {horizontalAlignment=HAlign_Fill, verticalAlignment=VAlign_Fill}}
  }
  EmptyImage {
   widgetClassPath=&Image
   bInherited=true
   bIsVariable=true
   parent=Overlay_Content
   widget=EmptyImage
   properties {
    brush {
     bIsDynamicallyLoaded=false
     drawAs=Image
     imageType=NoImage
     mirroring=NoMirror
     resourceName=None
     resourceObject=&Quickbar-Frame-Default
     tiling=NoTile
     imageSize=<80 80>
     margin=<ltrb 0.25 0.25 0.25 0.25>
     outlineSettings {
      bUseBrushTransparency=false
      roundingType=HalfHeightRadius
      width=0
      color {colorUseRule=UseColor_Specified, specifiedColor=<rgba 0 0 0 0>}
      cornerRadii=<0 0 0 0>
     }
     tintColor {colorUseRule=UseColor_Specified, specifiedColor=<rgba 1 1 1 1>}
     uVRegion {bIsValid=false, max=<0 0>, min=<0 0>}
    }
    colorAndOpacity=<rgba 1 1 1 0.2>
   }
   slot {class=OverlaySlot, properties {horizontalAlignment=HAlign_Fill, verticalAlignment=VAlign_Fill, padding=<ltrb 6 6 6 6>}}
  }
  ItemCardMaximized {
   widgetClassPath=/Script/FortniteUI.FortMultiSizeItemCard
   bInherited=true
   bIsVariable=true
   parent=Overlay_Content
   widget=ItemCardMaximized
   properties {bPItemCardSize=S, emptyItemType=WeaponMelee, shouldSuppressItemTypeIcon=true, shouldSuppressTierMeter=true}
   slot {class=OverlaySlot, properties {horizontalAlignment=HAlign_Center, verticalAlignment=VAlign_Center, padding=<ltrb 2 2 2 2>}}
  }
  ItemCardMinimized {
   widgetClassPath=&FortMultiSizeItemCard
   bInherited=true
   bIsVariable=true
   parent=Overlay_Content
   widget=ItemCardMinimized
   properties {bPItemCardSize=XS, emptyItemType=WeaponMelee, shouldSuppressItemTypeIcon=true, shouldSuppressTierMeter=true}
   slot {class=OverlaySlot, properties {horizontalAlignment=HAlign_Center, verticalAlignment=VAlign_Center, padding=<ltrb 2 2 2 2>}}
  }
  QuickbarSlotCooldown {
   widgetClassPath=&/QuickbarSlotCooldown
   bInherited=true
   bIsVariable=true
   parent=Overlay_Content
   widget=QuickbarSlotCooldown
   properties {visibility=HitTestInvisible}
   slot {class=OverlaySlot, properties {horizontalAlignment=HAlign_Center, verticalAlignment=VAlign_Center}}
  }
  SlotInteraction {
   widgetClassPath=&Image
   bInherited=true
   bIsVariable=true
   parent=Overlay_Content
   widget=SlotInteraction
   properties {
    brush {
     bIsDynamicallyLoaded=false
     drawAs=Box
     imageType=NoImage
     mirroring=NoMirror
     resourceName=None
     resourceObject=/Game/UI/InGame/Quickbar/Icon-QuickbarSlot-InvalidDrop
     tiling=NoTile
     imageSize=<64 64>
     margin=<ltrb 0.25 0.25 0.25 0.25>
     outlineSettings {
      bUseBrushTransparency=false
      roundingType=HalfHeightRadius
      width=0
      color {colorUseRule=UseColor_Specified, specifiedColor=<rgba 0 0 0 0>}
      cornerRadii=<0 0 0 0>
     }
     tintColor {colorUseRule=UseColor_Specified, specifiedColor=<rgba 1 1 1 1>}
     uVRegion {bIsValid=false, max=<0 0>, min=<0 0>}
    }
    colorAndOpacity=<rgba 1 1 1 0.1>
   }
   slot {class=OverlaySlot, properties {horizontalAlignment=HAlign_Fill, verticalAlignment=VAlign_Fill, padding=<ltrb 8 8 8 8>}}
  }
  ItemLocked {
   widgetClassPath=&Image
   bIsVariable=true
   parent=Overlay_Content
   widget=ItemLocked
   properties {
    visibility=Collapsed
    brush {
     bIsDynamicallyLoaded=false
     drawAs=Image
     imageType=NoImage
     mirroring=NoMirror
     resourceName=None
     resourceObject=/Game/UI/Foundation/Textures/Icons/Locks/T-Icon-Lock-32
     tiling=NoTile
     imageSize=<32 32>
     margin=<ltrb 0 0 0 0>
     outlineSettings {
      bUseBrushTransparency=false
      roundingType=HalfHeightRadius
      width=0
      color {colorUseRule=UseColor_Specified, specifiedColor=<rgba 0 0 0 0>}
      cornerRadii=<0 0 0 0>
     }
     tintColor {colorUseRule=UseColor_Specified, specifiedColor=<rgba 1 1 1 1>}
     uVRegion {bIsValid=false, max=<0 0>, min=<0 0>}
    }
   }
   slot {class=OverlaySlot}
  }
  QuickbarKeybindWidgetBottom {
   widgetClassPath=&/QuickbarKeybindBottom
   bInherited=true
   bIsVariable=true
   parent=VerticalBox_0
   widget=QuickbarKeybindWidgetBottom
   slot {class=VerticalBoxSlot, properties {horizontalAlignment=HAlign_Center}}
  }
  QuickbarSlotDurability {widgetClassPath=/Script/FortniteUI.AthenaItemElementWidgetBase, bInherited=true, widget=null}
  NativeCell {widgetClassPath=/Script/FortniteUI.AthenaQuickBarNativeCell, bInherited=true, widget=null}
 ]
}
