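# HUD widget that briefly shows status text near the reticle for failed actions and autofire pause/resume changes.
# HUD context and autofire delegates feed DisplayReticleStatusText, then retriggerable delays hide it; Party Rift failures get their own duration.
# A visibility switcher picks rich or plain text by searching for "/>" markup.
# Autofire messages respect a component flag and a HUD tag; touch input adjusts alignment and width.
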WidgetBlueprint /Game/UI/InGame/HUD/ReticleStatusWidget
graphs [
 DisplayReticleStatusText {nodes [
   n1 = DisplayReticleStatusText!() -> (then, TextToDisplay:"Text (by ref)", FailureReason:"Gameplay Tag Container Structure (by ref)") {is=title K2Node_FunctionEntry}
   > Set Reticle Status Text!(execute, self, FailureText:Text=@n1) -> (then, TextWidget:Widget Object Reference) {is=title K2Node_CallFunction}
   > n2 = GetStatusDisplayDurationFromFailureReason!(execute, self, FailureReason:Gameplay Tag Container Structure=@n1.FailureReason) -> (then, StatusDisplayDuration:double)
   > Return Node!(execute, DisplayDuration:double=@n2)
  ]}
 EventGraph {nodes [
   n1 = OnAutoFireChanged!() -> (OutputDelegate, then, bAutofireActive:bool=~) {is=title K2Node_CustomEvent}
   n2 = Get NoAutoFireableWeapons() -> (NoAutoFireableWeapons:bool)
   n3 = Cast To FortPlayerController(execute=@n9.then, Object=GetOwningPlayer())
   n4 = GetAutofireComponent!(self:Fort Player Controller Object Reference=@n3) -> (ReturnValue:Fort Controller Component Autofire Object Reference)
   Branch(
    execute=@n1.then
    Condition=AND Boolean(
     A=Get bAllowAutofireStatusText!(self:Fort Controller Component Autofire Object Reference=@n4) -> (bAllowAutofireStatusText:bool)
     B=NOT Boolean(A=@n2)
    )
   )
   > n5 = DisplayReticleStatusText!(
    execute
    self
    TextToDisplay:"Text (by ref)"=SelectText(A=Get AutoFireOnString() -> (AutoFireOnString:Text), B=Get AutoFireOffString() -> (AutoFireOffString:Text), bPickA=@n1.bAutofireActive)
    FailureReason:"Gameplay Tag Container Structure (by ref)"
   ) -> (then, DisplayDuration:double) {is=title K2Node_CallFunction}
   > RetriggerableDelay(Duration=@n5)
   > HideReticleStatusText()
   n6 = OnHudElementVisiblityChanged!() -> (OutputDelegate, then, HiddenHUDElementTags:"Gameplay Tag Container Structure (by ref)") {is=title K2Node_CustomEvent}
   > Set NoAutoFireableWeapons(NoAutoFireableWeapons:bool=HasTag!(
     TagContainer:"Gameplay Tag Container Structure (by ref)"=@n6.HiddenHUDElementTags
     Tag:Gameplay Tag Structure=(TagName="Gameplay.Status.NoAutoFireableWeapons")
     bExactMatch:bool=true
    ) -> (ReturnValue:bool)) -> (Output_Get:bool)
   n7 = OnUnableToPerformAction!() -> (OutputDelegate, then, FailedReason:Gameplay Tag Container Structure, FailureText:"Text (by ref)") {is=title K2Node_CustomEvent}
   > n8 = DisplayReticleStatusText(TextToDisplay=@n7.FailureText, FailureReason=@n7.FailedReason)
   > RetriggerableDelay(Duration=@n8)
   > HideReticleStatusText()
   Event OnInitialized()
   > n9 = Bind Event to On Unable to Perform Action(
    self:Fort HUDContext Object Reference=Get FortHUDContext!() -> (ReturnValue:Fort HUDContext Object Reference) {is=title K2Node_GetBlueprintContextEx}
    Delegate=@n7
   ) {is=title K2Node_AddDelegate}
   Event Construct()
   > SetTouchAlignment()
   > n10 = PlaylistDataReadyAsync!(execute:Exec) -> (
    then:Exec
    OnReady:Exec
    GameState:Fort Game State Athena Object Reference
    Playlist:Fort Playlist Object Reference
    PlaylistContextTags:"Gameplay Tag Container Structure (by ref)"
   ) {is=title K2Node_AsyncAction}
   n11 = GetOwningPlayer()
   n12 = IsDynamicTouchHUDActiveForPlaylist!(Playlist:Fort Playlist Object Reference=@n10.Playlist) -> (ReturnValue:bool)
   Branch(execute=@n10.OnReady, Condition=AND Boolean(A=NOT Boolean(A=@n12), B=IsUsingTouch(PlayerController=@n11)))
   n13 = Get SizeBox() -> (SizeBox:Size Box Object Reference)
   > SetWidthOverride(self=@n13, InWidthOverride=200)
   > ClearMinDesiredWidth(self:Size Box Object Reference=@n13)
   n14 = Event PreConstruct()
   > Parent: PreConstruct(IsDesignTime=@n14.IsDesignTime)
   > SetHorizontalAlignment(
    self=[
     SlotAsOverlaySlot(Widget=Get ReticleStatusRichTextBlock() -> (ReticleStatusRichTextBlock:Common Rich Text Block Object Reference))
     SlotAsOverlaySlot(Widget=Get ReticleStatusText() -> (ReticleStatusText:Common Text Object Reference))
    ]
    InHorizontalAlignment=Get TextHorizontalAlignment() -> (TextHorizontalAlignment:EHorizontalAlignment Enum=HAlign_Fill)
   )
   Bind Event to On HUDElement Visibility Changed(execute=@n15.then, self:Fort HUDContext Object Reference=Get FortHUDContext(), Delegate=@n6) {is=title K2Node_AddDelegate}
   n15 = Bind Event to On Autofire Value Changed(execute=@n16.Is Valid, self:Fort Controller Component Autofire Object Reference=@n4, Delegate=@n1) {is=title K2Node_AddDelegate}
   n16 = Is Valid(exec=@n3.then, InputObject=@n4)
  ]}
 GetStatusDisplayDurationFromFailureReason {nodes [
   n1 = GetStatusDisplayDurationFromFailureReason!() -> (then, FailureReason:Gameplay Tag Container Structure) {is=title K2Node_FunctionEntry}
   n2 = Get PartyRiftTags() -> (PartyRiftTags:Gameplay Tag Container Structure)
   > Return Node!(
    execute
    StatusDisplayDuration:double=Select(
     Option 0:double=Get DefaultStatusDisplayDuration() -> (DefaultStatusDisplayDuration:double)
     Option 1:double=Get PartyRiftStatusDisplayDuration() -> (PartyRiftStatusDisplayDuration:double)
     Index=HasAnyTags!(
      TagContainer:"Gameplay Tag Container Structure (by ref)"=@n1
      OtherContainer:"Gameplay Tag Container Structure (by ref)"=@n2
      bExactMatch:bool=false
     ) -> (ReturnValue:bool)
    ) -> (ReturnValue:double)
   )
  ]}
 HideReticleStatusText {nodes [
   HideReticleStatusText() {is=title K2Node_FunctionEntry}
   > Set Visibility(Visibility=Hidden, self=Get ReticleStatusRichTextBlock())
   > Set Visibility(Visibility=Hidden, self=Get ReticleStatusText())
  ]}
 Set Reticle Status Text {nodes [
   n1 = Set Reticle Status Text!() -> (then, FailureText:Text) {is=title K2Node_FunctionEntry}
   > n2 = Set bFailureTextIsRichText(bFailureTextIsRichText:bool=Contains(SearchIn="To String (Text)"(InText=@n1), Substring="/>")) -> (Output_Get:bool)
   > n3 = Branch(Condition=@n2)
   n4 = Get FailureText() -> (FailureText:Text)
   n5 = Get ReticleStatusRichTextBlock()
   > n6 = SetText(self=@n5, InText=@n4)
   n7 = Get ReticleStatusText()
   n8 = Set TextWidgetToUse(
    execute=[@n6.then, @n10.then]
    TextWidgetToUse:Widget Object Reference=Select(
     Option 0:Widget Object Reference=@n7
     Option 1:Widget Object Reference=@n5
     Index=Get bFailureTextIsRichText() -> (bFailureTextIsRichText:bool)
    ) -> (ReturnValue:Widget Object Reference)
   ) -> (Output_Get:Widget Object Reference)
   > SetActiveWidget(self=Get VisibleStatusText() -> (VisibleStatusText:Common Visibility Switcher Object Reference), Widget=@n8)
   n9 = Get TextWidgetToUse() -> (TextWidgetToUse:Widget Object Reference)
   > SetVisibility(self=@n9, InVisibility=HitTestInvisible)
   > Return Node!(execute, TextWidget:Widget Object Reference=@n9)
   n10 = "SetText (Text)"(execute=@n3.else, self=@n7, InText=@n4)
  ]}
 SetTouchAlignment {nodes [
   SetTouchAlignment() {is=title K2Node_FunctionEntry}
   n1 = GetOwningPlayer()
   > Branch(Condition=AND Boolean(A=IsUsingTouch(PlayerController=@n1), B=IsMobileGame()))
   > SetHorizontalAlignment(
    self:Size Box Slot Object Reference=SlotAsSizeBoxSlot!(Widget:Widget Object Reference=Get VisibleStatusText()) -> (ReturnValue:Size Box Slot Object Reference)
    InHorizontalAlignment=HAlign_Center
   )
  ]}
]
variables [
 AutoFireOffString {type=string, default=Auto Fire Paused}
 AutoFireOnString {type=string, default=Auto Fire Resumed}
 bIsNoBuildMode {type=boolean}
 DefaultStatusDisplayDuration {type=number, default=1.2}
 HUDElement Tags {type=GameplayTagContainer, default {gameplayTags [{tagName=Gameplay.Status.NoAutoFireableWeapons}]}}
 NoAutoFireableWeapons {type=boolean}
 PartyRiftStatusDisplayDuration {type=number, default=3}
 PartyRiftTags {type=GameplayTagContainer, default {gameplayTags [{tagName=Activate.Failed.PartyRift}]}}
 TextHorizontalAlignment {type=EHorizontalAlignment, default=HAlign_Left}
]
widgets {
 info {inheritedWidgetCount=1, parentClass=&FortHUDElementWidget, rootWidgetClass=&SizeBox, widgetCount=5}
 widgets [
  SizeBox {
   widgetClassPath=&SizeBox
   bIsVariable=true
   widget=SizeBox
   properties {bOverride_MinDesiredHeight=true, bOverride_MinDesiredWidth=true, minDesiredHeight=40, minDesiredWidth=600}
  }
  VisibleStatusText {
   widgetClassPath=&CommonVisibilitySwitcher
   bIsVariable=true
   parent=SizeBox
   widget=VisibleStatusText
   slot {class=SizeBoxSlot}
  }
  ReticleStatusText {
   widgetClassPath=&CommonTextBlock
   bIsVariable=true
   parent=VisibleStatusText
   widget=ReticleStatusText
   properties {
    justification=Center
    mobileFontSizeMultiplier=1.2
    text="{Reticle Status}"
    visibility=Hidden
    wrapTextAt=350
    font {
     bForceMonospaced=false
     bMaterialIsStencil=false
     fontMaterial=null
     fontObject=&BurbankSmall
     letterSpacing=0
     monospacedWidth=1
     size=18
     skewAmount=0
     typefaceFontName=Bold
     outlineSettings {
      bApplyOutlineToDropShadows=true
      bMiteredCorners=false
      bSeparateFillAlpha=true
      outlineMaterial=&M-FontShadow
      outlineSize=3
      outlineColor=<rgba 0 0 0 0.5>
     }
    }
    shadowOffset=<0 0>
   }
   slot {class=CommonVisibilitySwitcherSlot, properties {horizontalAlignment=HAlign_Left}}
  }
  ReticleStatusRichTextBlock {
   widgetClassPath=&CommonRichTextBlock
   bIsVariable=true
   parent=VisibleStatusText
   widget=ReticleStatusRichTextBlock
   properties {
    bOverrideDefaultStyle=true
    justification=Center
    mobileTextBlockScale=1.2
    text="{Reticle Status}"
    decoratorClasses [&/ReticleStatusImageDecorator.ReticleStatusImageDecorator_C]
    defaultTextStyleOverride {
     overflowPolicy=Clip
     transformPolicy=None
     colorAndOpacity {colorUseRule=UseColor_Specified, specifiedColor=<rgba 1 1 1 1>}
     font {
      bForceMonospaced=false
      bMaterialIsStencil=false
      fontMaterial=null
      fontObject=&BurbankSmall
      letterSpacing=0
      monospacedWidth=1
      size=18
      skewAmount=0
      typefaceFontName=Bold
      outlineSettings {
       bApplyOutlineToDropShadows=false
       bMiteredCorners=false
       bSeparateFillAlpha=false
       outlineMaterial=null
       outlineSize=0
       outlineColor=<rgba 0 0 0 1>
      }
     }
     highlightColor {colorUseRule=UseColor_Specified, specifiedColor=<rgba 0 0 0 1>}
     highlightShape {
      bIsDynamicallyLoaded=false
      drawAs=Image
      imageType=NoImage
      mirroring=NoMirror
      resourceName=None
      resourceObject=null
      tiling=NoTile
      imageSize=<32 32>
      margin=<ltrb 0 0 0 0>
      outlineSettings {
       bUseBrushTransparency=false
       roundingType=HalfHeightRadius
       width=0
       color {colorUseRule=UseColor_Specified, specifiedColor=<rgba 0 0 0 0>}
       cornerRadii=<0 0 0 0>
      }
      tintColor {colorUseRule=UseColor_Specified, specifiedColor=<rgba 1 1 1 1>}
      uVRegion {bIsValid=false, max=<0 0>, min=<0 0>}
     }
     selectedBackgroundColor {colorUseRule=UseColor_Foreground, specifiedColor=<rgba 1 0 1 1>}
     shadowColorAndOpacity=<rgba 0 0 0 1>
     shadowOffset=<0 0>
     strikeBrush {
      bIsDynamicallyLoaded=false
      drawAs=Image
      imageType=NoImage
      mirroring=NoMirror
      resourceName=None
      resourceObject=null
      tiling=NoTile
      imageSize=<32 32>
      margin=<ltrb 0 0 0 0>
      outlineSettings {
       bUseBrushTransparency=false
       roundingType=HalfHeightRadius
       width=0
       color {colorUseRule=UseColor_Specified, specifiedColor=<rgba 0 0 0 0>}
       cornerRadii=<0 0 0 0>
      }
      tintColor {colorUseRule=UseColor_Specified, specifiedColor=<rgba 1 1 1 1>}
      uVRegion {bIsValid=false, max=<0 0>, min=<0 0>}
     }
     underlineBrush {
      bIsDynamicallyLoaded=false
      drawAs=Image
      imageType=NoImage
      mirroring=NoMirror
      resourceName=None
      resourceObject=null
      tiling=NoTile
      imageSize=<32 32>
      margin=<ltrb 0 0 0 0>
      outlineSettings {
       bUseBrushTransparency=false
       roundingType=HalfHeightRadius
       width=0
       color {colorUseRule=UseColor_Specified, specifiedColor=<rgba 0 0 0 0>}
       cornerRadii=<0 0 0 0>
      }
      tintColor {colorUseRule=UseColor_Specified, specifiedColor=<rgba 1 1 1 1>}
      uVRegion {bIsValid=false, max=<0 0>, min=<0 0>}
     }
    }
    margin=<ltrb 0 3 0 0>
   }
   slot {class=CommonVisibilitySwitcherSlot, properties {horizontalAlignment=HAlign_Left}}
  }
  TutorialHighlight {widgetClassPath=&TutorialHighlight}
 ]
}
