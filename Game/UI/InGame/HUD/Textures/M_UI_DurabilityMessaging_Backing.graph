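# Additive UI material that, going by its name, draws a soft red circular backing for durability messaging on the in-game HUD.
# The SDF output of MF_UI_Circle is inverted and raised to a power to shape a radial falloff.
# The BackingOpacity parameter scales that falloff, alongside a constant red color.
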Material /Game/UI/InGame/HUD/Textures/M_UI_DurabilityMessaging_Backing
graph {
 properties {
  bAutomaticallySetUsageInEditor=false
  bUsedWithStaticMesh=false
  blendMode=BLEND_Additive
  materialDomain=MD_UI
  usedShadingModels=Default Lit
 }
 nodes [
  Constant3Vector() {constant=<rgba 1 0 0 0>}
  n1 = MaterialFunctionCall(Size=Constant()) {materialFunction=&MF_UI_Circle}
  Multiply(A=Power(Base=OneMinus(None=@n1.SDF)) {constExponent=3}, B=ScalarParameter() {defaultValue=0.3, parameterName=BackingOpacity})
 ]
}
