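# Translucent UI material for the QuickEdit HUD, drawing a 3x3 grid of square cells with selectable cross marks.
# Nine static switches, 1-1 to 3-3, place cross samples via ScaleUVsAroundPoint, merged with Max.
# SmoothStep thresholds the distance-field cross and square textures into sharp shapes with tunable edge softness.
# Square masks blend GridShadow color with the marks and combine outline, faded grid and crosses into an alpha-style mask.
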Material /Game/UI/InGame/HUD/Textures/M_UI_QuickEdit
graph {
 properties {
  bUsedWithStaticMesh=false
  blendMode=BLEND_Translucent
  materialDomain=MD_UI
  pixelDepthOffsetMode=PDOM_Legacy
  usedShadingModels=Default Lit
 }
 nodes [
  n1 = Constant3Vector() {constant=<rgba 0 0.2307 1 0>}
  n2 = VectorParameter() {parameterName=GridShadow, defaultValue=<rgba 0.0052 0.0052 0.0056 0.8>}
  n3 = MaterialFunctionCall!(
   Texture Scale=Add(A=ScalarParameter() {defaultValue=0.33, parameterName=GridScale}, B=ScalarParameter() {defaultValue=0.01, parameterName=SelectSize})
   UVs=TextureCoordinate()
   Offset XY Texturesize Z=Constant3Vector() {constant=<rgba 0 0 100 0>}
  ) {materialFunction=/Engine/Functions/Engine_MaterialFunctions01/Texturing/ScaleUVsAroundPoint}
  n4 = Constant()
  n5 = StaticSwitchParameter(
   True=Multiply(
    A=TextureSample(UVs=@n3.UVs) {samplerSource=SSM_Wrap_WorldGroupSettings, texture=/Game/Athena/HUD/Materials/Score/T_UI_Cross_DF}.R
    B=@n3.0-1 mask
   )
   False=@n4
  ) {parameterName=1-1}
  n6 = MaterialFunctionCall(
   Texture Scale=Add(A=ScalarParameter() {defaultValue=0.33, parameterName=GridScale}, B=ScalarParameter() {defaultValue=0.01, parameterName=SelectSize})
   UVs=TextureCoordinate()
   Offset XY Texturesize Z=Constant3Vector() {constant=<rgba 50 0 100 0>}
  ) {materialFunction=&ScaleUVsAroundPoint}
  n7 = StaticSwitchParameter(
   True=Multiply(
    A=TextureSample(UVs=@n6.UVs) {samplerSource=SSM_Wrap_WorldGroupSettings, texture=&T_UI_Cross_DF}.R
    B=@n6.0-1 mask
   )
   False=@n4
  ) {parameterName=1-2}
  n8 = MaterialFunctionCall(
   Texture Scale=Add(A=ScalarParameter() {defaultValue=0.33, parameterName=GridScale}, B=ScalarParameter() {defaultValue=0.01, parameterName=SelectSize})
   UVs=TextureCoordinate()
   Offset XY Texturesize Z=Constant3Vector() {constant=<rgba 100 0 100 0>}
  ) {materialFunction=&ScaleUVsAroundPoint}
  n9 = Multiply(
   A=TextureSample(UVs=@n8.UVs) {samplerSource=SSM_Wrap_WorldGroupSettings, texture=&T_UI_Cross_DF}.R
   B=@n8.0-1 mask
  )
  n10 = MaterialFunctionCall(
   Texture Scale=Add(A=ScalarParameter() {defaultValue=0.33, parameterName=GridScale}, B=ScalarParameter() {defaultValue=0.01, parameterName=SelectSize})
   UVs=TextureCoordinate()
   Offset XY Texturesize Z=Constant3Vector() {constant=<rgba 0 50 100 0>}
  ) {materialFunction=&ScaleUVsAroundPoint}
  n11 = TextureSample(UVs=@n10.UVs) {samplerSource=SSM_Wrap_WorldGroupSettings, texture=&T_UI_Cross_DF}
  n12 = Max(
   A=Max(A=Max(A=@n5, B=@n7), B=StaticSwitchParameter(True=@n9, False=@n4) {parameterName=1-3})
   B=StaticSwitchParameter(True=Multiply(A=@n11.R, B=@n10.0-1 mask), False=@n4) {parameterName=2-1}
  )
  n13 = MaterialFunctionCall(
   Texture Scale=Add(A=ScalarParameter() {defaultValue=0.33, parameterName=GridScale}, B=ScalarParameter() {defaultValue=0.01, parameterName=SelectSize})
   UVs=TextureCoordinate()
   Offset XY Texturesize Z=Constant3Vector() {constant=<rgba 50 50 100 0>}
  ) {materialFunction=&ScaleUVsAroundPoint}
  n14 = StaticSwitchParameter(
   True=Multiply(
    A=TextureSample(UVs=@n13.UVs) {samplerSource=SSM_Wrap_WorldGroupSettings, texture=&T_UI_Cross_DF}.R
    B=@n13.0-1 mask
   )
   False=@n4
  ) {parameterName=2-2}
  n15 = MaterialFunctionCall(
   Texture Scale=Add(A=ScalarParameter() {defaultValue=0.33, parameterName=GridScale}, B=ScalarParameter() {defaultValue=0.01, parameterName=SelectSize})
   UVs=TextureCoordinate()
   Offset XY Texturesize Z=Constant3Vector() {constant=<rgba 100 50 100 0>}
  ) {materialFunction=&ScaleUVsAroundPoint}
  n16 = Multiply(
   A=TextureSample(UVs=@n15.UVs) {samplerSource=SSM_Wrap_WorldGroupSettings, texture=&T_UI_Cross_DF}.R
   B=@n15.0-1 mask
  )
  n17 = MaterialFunctionCall(
   Texture Scale=Add(A=ScalarParameter() {defaultValue=0.33, parameterName=GridScale}, B=ScalarParameter() {defaultValue=0.01, parameterName=SelectSize})
   UVs=TextureCoordinate()
   Offset XY Texturesize Z=Constant3Vector() {constant=<rgba 0 100 100 0>}
  ) {materialFunction=&ScaleUVsAroundPoint}
  n18 = TextureSample(UVs=@n17.UVs) {samplerSource=SSM_Wrap_WorldGroupSettings, texture=&T_UI_Cross_DF}
  n19 = Max(
   A=Max(A=Max(A=@n12, B=@n14), B=StaticSwitchParameter(True=@n16, False=@n4) {parameterName=2-3})
   B=StaticSwitchParameter(True=Multiply(A=@n18.R, B=@n17.0-1 mask), False=@n4) {parameterName=3-1}
  )
  n20 = MaterialFunctionCall(
   Texture Scale=Add(A=ScalarParameter() {defaultValue=0.33, parameterName=GridScale}, B=ScalarParameter() {defaultValue=0.01, parameterName=SelectSize})
   UVs=TextureCoordinate()
   Offset XY Texturesize Z=Constant3Vector() {constant=<rgba 50 100 100 0>}
  ) {materialFunction=&ScaleUVsAroundPoint}
  n21 = StaticSwitchParameter(
   True=Multiply(
    A=TextureSample(UVs=@n20.UVs) {samplerSource=SSM_Wrap_WorldGroupSettings, texture=&T_UI_Cross_DF}.R
    B=@n20.0-1 mask
   )
   False=@n4
  ) {parameterName=3-2}
  n22 = MaterialFunctionCall(
   Texture Scale=Add(A=ScalarParameter() {defaultValue=0.33, parameterName=GridScale}, B=ScalarParameter() {defaultValue=0.01, parameterName=SelectSize})
   UVs=TextureCoordinate()
   Offset XY Texturesize Z=Constant3Vector() {constant=<rgba 100 100 100 0>}
  ) {materialFunction=&ScaleUVsAroundPoint}
  n23 = Multiply(
   A=TextureSample(UVs=@n22.UVs) {samplerSource=SSM_Wrap_WorldGroupSettings, texture=&T_UI_Cross_DF}.R
   B=@n22.0-1 mask
  )
  n24 = ScalarParameter() {defaultValue=0.2, parameterName=Edge}
  n25 = MaterialFunctionCall(
   Alpha=Max(A=Max(A=@n19, B=@n21), B=StaticSwitchParameter(True=@n23, False=@n4) {parameterName=3-3})
   Min=@n24
   Max=Add(A=@n24, B=ScalarParameter() {defaultValue=0.01, parameterName=EdgeSoftness})
  ) {materialFunction=&SmoothStep}
  n26 = ScalarParameter() {defaultValue=0.2, parameterName=Edge}
  n27 = TextureSample(UVs=MaterialFunctionCall(
    Texture Scale=ScalarParameter() {defaultValue=0.33, parameterName=GridScale}
    UVs=TextureCoordinate()
    Offset XY Texturesize Z=VectorParameter() {parameterName=GridScaleOffset, defaultValue=<rgba 0.5 0.5 1 1>}.RGB
   ) {materialFunction=&ScaleUVsAroundPoint}.UVs) {samplerSource=SSM_Wrap_WorldGroupSettings, texture=/Game/Athena/HUD/Materials/Score/T_UI_Square_DF}
  n28 = MaterialFunctionCall(Alpha=@n27.R, Min=@n26, Max=Add(A=@n26, B=ScalarParameter() {defaultValue=0.01, parameterName=EdgeSoftness})) {materialFunction=&SmoothStep}
  n29 = LinearInterpolate(A=@n2.RGB, B=@n25.Result, Alpha=@n28.Result)
  Add(A=@n1, B=@n29)
  n30 = Subtract(A=ScalarParameter() {defaultValue=0.2, parameterName=Edge}, B=ScalarParameter() {defaultValue=0.1, parameterName=EdgeOutline})
  n31 = MaterialFunctionCall(Alpha=@n27.R, Min=@n30, Max=Add(A=@n30, B=ScalarParameter() {defaultValue=0.01, parameterName=EdgeSoftness})) {materialFunction=&SmoothStep}
  n32 = ScalarParameter() {defaultValue=0.9, parameterName=FadeGrid}
  Max(A=Max(A=Multiply(A=@n31.Result, B=@n2.A), B=Multiply(A=@n28.Result, B=@n32)), B=@n25.Result)
  Multiply(A=@n1, B=@n29)
 ]
}
