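# Translucent UI material for the in-game HUD that draws a procedural ring with a configurable color.
# Two exponential radial gradients, one at OuterRadius and one at OuterRadius minus RingSize, each pass through SmoothStep.
# Edge and EdgeSoftness set the shared SmoothStep bounds that shape the edges.
# Subtracting the inner disc from the outer one leaves the ring mask, with no texture sampling.
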Material /Game/UI/InGame/HUD/Textures/M_UI_Ring
graph {
 properties {
  bUsedWithStaticMesh=false
  blendMode=BLEND_Translucent
  materialDomain=MD_UI
  pixelDepthOffsetMode=PDOM_Legacy
  usedShadingModels=Default Lit
 }
 nodes [
  n1 = ScalarParameter() {defaultValue=0.5, parameterName=OuterRadius, sliderMax=2}
  n2 = ScalarParameter() {parameterName=Edge}
  n3 = Subtract(A=@n2, B=ScalarParameter() {parameterName=EdgeSoftness})
  n4 = Subtract(A=@n1, B=ScalarParameter() {defaultValue=0.1, parameterName=RingSize})
  Subtract(
   A=MaterialFunctionCall(
    Alpha=MaterialFunctionCall(Radius=@n1) {materialFunction=&RadialGradientExponential}.RadialGradientExponential
    Min=@n2
    Max=@n3
   ) {materialFunction=&SmoothStep}.Result
   B=MaterialFunctionCall(
    Alpha=MaterialFunctionCall(Radius=@n4) {materialFunction=&RadialGradientExponential}.RadialGradientExponential
    Min=@n2
    Max=@n3
   ) {materialFunction=&SmoothStep}.Result
  )
  VectorParameter() {parameterName=Color, defaultValue=<rgba 1 1 1 1>}
 ]
}
