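# Translucent UI material for the HUD's durability messaging icon, built from the low-durability icon texture.
# The red and green channels each pass through the channel-packed distance field function, whose fills give the main shape and crack.
# The crack fill is dimmed by a constant and added to the main shape, alongside a constant white color.
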Material /Game/UI/InGame/HUD/Textures/M_UI_DurabilityMessaging_Icon
graph {
 properties {
  bAutomaticallySetUsageInEditor=false
  bUsedWithStaticMesh=false
  blendMode=BLEND_Translucent
  materialDomain=MD_UI
  usedShadingModels=Default Lit
 }
 nodes [
  n1 = Constant() {r=0.3}
  n2 = Constant() {r=0.5}
  n3 = TextureSample() {texture=&/T_UI_Icon_Durability_Low}
  n4 = MaterialFunctionCall!(RGBA Channels=@n3.G, Glow Min=@n1, Glow Max=@n2, Outline Thickness) {materialFunction=/FortUILibrary/MaterialFunctions/Basics/MF_UI_ChannelPackedDistanceField}
  Add(
   A=NamedRerouteDeclaration(None=MaterialFunctionCall(RGBA Channels=@n3.R, Glow Min=@n1, Glow Max=@n2) {materialFunction=&MF_UI_ChannelPackedDistanceField}.Fill) {name=IconMainShape, nodeColor=<rgba 0 0 0 1>}
   B=Multiply(A=NamedRerouteDeclaration(None=@n4.Fill) {name=Crack, nodeColor=<rgba 0 0 0 1>}, B=Constant() {r=0.4})
  )
  Constant3Vector() {constant=<rgba 1 1 1 0>}
 ]
}
