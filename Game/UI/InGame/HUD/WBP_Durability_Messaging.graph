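# HUD widget that shows a timed notification and plays a sound when the equipped item is low on durability or breaks.
# UpdateDurability compares viewmodel durability drops against three curve-table thresholds, warning once per band and resetting flags when the item changes.
# An async listener reports durability destruction to trigger the broken message.
# Text comes from the item definition with a formatted fallback, displayed through the shared backplate visuals widget.
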WidgetBlueprint /Game/UI/InGame/HUD/WBP_Durability_Messaging
graphs [
 ChangeBacking {nodes [
   n1 = ChangeBacking!() -> (then, MaterialInstance:Material Instance Object Reference) {is=title K2Node_FunctionEntry}
   > SetBackplateBrushFromSoftMaterial!(
    execute
    self:WBP Notification With Backplate Visuals Object Reference=Get WBP_NotificationWithBackplate() -> (WBP_NotificationWithBackplate:WBP Notification With Backplate Visuals Object Reference)
    SoftMat:Material Interface Soft Object Reference=Make Soft Reference!(Input:Material Instance Object Reference=@n1) -> (Output:Material Instance Soft Object Reference)
   ) -> (then)
  ]}
 DurabilityWarningMessage {nodes [
   DurabilityWarningMessage() {is=title K2Node_FunctionEntry}
   n1 = GetRepresentedItem!(self:Item Durability Viewmodel Object Reference=Get FortEquippedItemDurabilityVM() -> (FortEquippedItemDurabilityVM:Fort Equipped Item Durability VM Object Reference)) -> (ReturnValue:Fort World Item Object Reference)
   > Is Valid(InputObject=@n1)
   n2 = GetDisplayName!(self:Fort Item Object Reference=@n1) -> (ReturnValue:Text)
   n3 = GetDurabilityWarningMessage!(
    self:Fort Item Definition Object Reference=GetItemDefinition!(self:Fort Item Object Reference=@n1) -> (ReturnValue:Fort Item Definition Object Reference)
    InItemName:"Text (by ref)"=@n2
   ) -> (ReturnValue:Text)
   n4 = TextIsEmpty(InText=@n3)
   n5 = Format Text!(Format="LOCTEXT(\"<text color=\\\"FF5045EE\\\" size=\\\"18\\\">{ItemName}</> is damaged!\")", ItemName:Text=@n2) -> (Result)
   n6 = Get WBP_NotificationWithBackplate()
   > SetDetails!(
    execute
    self:WBP Notification With Backplate Visuals Object Reference=@n6
    InText:Text=TextToUpper!(InText:"Text (by ref)"=SelectText(A=@n5, B=@n3, bPickA=@n4)) -> (ReturnValue:Text)
   ) -> (then)
   > SetBrushFromMaterial(self:WBP Notification With Backplate Visuals Object Reference=@n6, Material=/Game/UI/InGame/HUD/Textures/M_UI_DurabilityMessaging_Icon)
   > TriggerStartAnimation()
   > PlaySound2D(Sound=/Game/Sounds/Fort_UI_Sounds/Gameplay/Inventory/UI_Inventory_Item_Durability_Low_Cue)
  ]}
 EventGraph {nodes [
   InitForNewItem() {is=title K2Node_CustomEvent}
   > Set ShowedLowNotification(ShowedLowNotification:bool=false) -> (Output_Get:bool)
   > Set ShowedLowNotification2(ShowedLowNotification2:bool=false) -> (Output_Get:bool)
   > Set ShowedLowNotification3(ShowedLowNotification3:bool=false) -> (Output_Get:bool)
   > n1 = Cast To FortPlayerController(Object=GetOwningPlayer())
   > UnregisterItemDestroyedHandle()
   > n2 = ListenForItemDestroyedByDurability!(
    execute:Exec
    OwningPlayerController:Fort Player Controller Object Reference=@n1
    ItemToMonitor:Fort Item Object Reference=GetRepresentedItem(self=Get FortEquippedItemDurabilityVM())
   ) -> (
    then:Exec
    AsyncTaskProxy:Fort Async Action Listen For Item Destroyed By Durability Object Reference
    OnItemDestroyed:Exec
    ItemGUID:"Guid Structure (by ref)"
    ItemDisplayName:"Text (by ref)"
   ) {is=title K2Node_AsyncAction}
   > Set ItemDestroyedHandle(ItemDestroyedHandle:Fort Async Action Listen For Item Destroyed By Durability Object Reference=@n2) -> (Output_Get:Fort Async Action Listen For Item Destroyed By Durability Object Reference)
   Set ItemDisplayName(execute=@n2.OnItemDestroyed, ItemDisplayName:Text=@n2.ItemDisplayName) -> (Output_Get:Text)
   > OnItemDestroyed()
   Event Destruct()
   > Parent: Destruct()
   > UnregisterItemDestroyedHandle()
   Event Construct()
   > Parent: Construct()
   > SetVisibility(InVisibility=Collapsed)
   n3 = Event PreConstruct()
   > Parent: PreConstruct(IsDesignTime=@n3.IsDesignTime)
   n4 = Get WBP_NotificationWithBackplate()
   > SetDetails(self=@n4)
   > SetBrushFromMaterial(self=@n4)
   > SetOutlineSize!(execute, self:WBP Notification With Backplate Visuals Object Reference=@n4, Outline Size:int=1) -> (then)
   > SetVisibility(
    self=Get Image_Icon!(self:WBP Notification With Backplate Visuals Object Reference=@n4) -> (Image_Icon:Image Object Reference)
    InVisibility=Collapsed
   )
   > SetVisibility(
    self=Get Text_Headline!(self:WBP Notification With Backplate Visuals Object Reference=@n4) -> (Text_Headline:Common Text Object Reference)
    InVisibility=Collapsed
   )
   > ChangeBacking!(execute, self, MaterialInstance:Material Instance Object Reference=/Game/UI/InGame/HUD/Textures/MI_UI_DurabilityMessaging_Backing) -> (then) {is=title K2Node_CallFunction}
  ]}
 OnItemDestroyed {nodes [
   OnItemDestroyed() {is=title K2Node_FunctionEntry}
   > TryShowBroken()
   > PlaySound2D(Sound=/Game/Sounds/Fort_UI_Sounds/Gameplay/Inventory/UI_Inventory_Item_Durability_Break_Cue)
  ]}
 OnNotificationTimerElapsed {nodes [
   OnNotificationTimerElapsed() {is=title K2Node_FunctionEntry}
   > TriggerStopAnimation()
  ]}
 TriggerStartAnimation {nodes [
   TriggerStartAnimation() {is=title K2Node_FunctionEntry}
   > SetVisibility(InVisibility=HitTestInvisible)
   > PlayIntroAnimation(self:WBP Notification With Backplate Visuals Object Reference=Get WBP_NotificationWithBackplate())
   > n1 = Set Timer by Event(
    Delegate=Create Event()
    Time=GetValueAtLevel!(Input:"Scalable Float Structure (by ref)"=Get MessagingDuration() -> (MessagingDuration:Scalable Float Structure), Level:float=0) -> (ReturnValue:float)
    bLooping=false
   )
   > Set HideNotificationTimerHandle(HideNotificationTimerHandle:Timer Handle Structure=@n1) -> (Output_Get:Timer Handle Structure)
  ]}
 TriggerStopAnimation {nodes [
   TriggerStopAnimation() {is=title K2Node_FunctionEntry}
   > n1 = Branch(Condition=Is Valid Timer Handle!(Handle:Timer Handle Structure=Get HideNotificationTimerHandle() -> (HideNotificationTimerHandle:Timer Handle Structure)) -> (ReturnValue:bool))
   > n2 = Clear and Invalidate Timer by Handle(Handle=Get HideNotificationTimerHandle())
   Branch(execute=[@n2.then, @n1.else], Condition=IsVisible())
   > PlayOutroAnimation(self:WBP Notification With Backplate Visuals Object Reference=Get WBP_NotificationWithBackplate())
  ]}
 TryShowBroken {nodes [
   TryShowBroken() {is=title K2Node_FunctionEntry}
   n1 = GetRepresentedItem(self=Get FortEquippedItemDurabilityVM())
   > Is Valid(InputObject=@n1)
   n2 = GetDisplayName(self=@n1)
   n3 = GetDurabilityBrokeMessage!(self:Fort Item Definition Object Reference=GetItemDefinition(self=@n1), InItemName:"Text (by ref)"=@n2) -> (ReturnValue:Text)
   n4 = TextIsEmpty(InText=@n3)
   n5 = Format Text(Format="LOCTEXT(\"<text color=\\\"FF5045EE\\\" size=\\\"18\\\">{ItemName}</> broke!\")", ItemName=@n2)
   n6 = Get WBP_NotificationWithBackplate()
   > SetDetails(self=@n6, InText=TextToUpper(InText=SelectText(A=@n5, B=@n3, bPickA=@n4)))
   > SetHeadling!(execute, self:WBP Notification With Backplate Visuals Object Reference=@n6, InText:Text="LOCTEXT(\"Broken!\")") -> (then)
   > SetBrushFromMaterial(self=@n6, Material=&M_UI_DurabilityMessaging_Icon)
   > TriggerStartAnimation() {is=title K2Node_CallFunction}
  ]}
 TryShowWarning {nodes [
   TryShowWarning() {is=title K2Node_FunctionEntry}
   > n1 = Branch(Condition=Get ShowedLowNotification() -> (ShowedLowNotification:bool))
   DurabilityWarningMessage(execute=@n1.else) {is=title K2Node_CallFunction}
   > Set ShowedLowNotification(ShowedLowNotification=true)
  ]}
 TryShowWarning_2 {nodes [
   TryShowWarning_2() {is=title K2Node_FunctionEntry}
   > n1 = Branch(Condition=Get ShowedLowNotification2() -> (ShowedLowNotification2:bool))
   DurabilityWarningMessage(execute=@n1.else)
   > Set ShowedLowNotification2(ShowedLowNotification2=true)
  ]}
 TryShowWarning_3 {nodes [
   TryShowWarning_3() {is=title K2Node_FunctionEntry}
   > n1 = Branch(Condition=Get ShowedLowNotification3() -> (ShowedLowNotification3:bool))
   DurabilityWarningMessage(execute=@n1.else)
   > Set ShowedLowNotification3(ShowedLowNotification3=true)
  ]}
 UnregisterItemDestroyedHandle {nodes [
   UnregisterItemDestroyedHandle() {is=title K2Node_FunctionEntry}
   > n1 = Get!(execute:Exec) -> (then:Exec, else:Exec, ItemDestroyedHandle:Fort Async Action Listen For Item Destroyed By Durability Object Reference)
   > StopListening(self:Fort Async Action Listen For Item Destroyed By Durability Object Reference=@n1)
   > Set ItemDestroyedHandle()
  ]}
 UpdateDurability {nodes [
   UpdateDurability!() -> (then, InDurability:double=~) {is=title K2Node_FunctionEntry}
   > n1 = Is Valid(InputObject=Get CachedWorldItem() -> (CachedWorldItem:Fort World Item Object Reference))
   n2 = GetRepresentedItem(self=Get FortEquippedItemDurabilityVM())
   n3 = Set CachedWorldItem(execute=@n1.Is Not Valid, CachedWorldItem:Fort World Item Object Reference=@n2) -> (Output_Get:Fort World Item Object Reference)
   Is Valid(exec=[@n1.Is Valid, @n3.then], InputObject=@n2)
   > n4 = Branch(Condition="Not Equal (Object)"(A:Object Reference=@n2, B:Object Reference=Get CachedWorldItem()) -> (ReturnValue:bool=~) {is=title K2Node_PromotableOperator})
   > n5 = InitForNewItem() {is=title K2Node_CallFunction}
   Set CachedDurability(
    execute=[@n7.then, @n6.then, @n5.then, @n13.then, @n19.then, @n21.then, @n20.else]
    CachedDurability:double=Get InDurability() -> (InDurability:double)
   ) -> (Output_Get:double)
   > Set CachedWorldItem(CachedWorldItem=GetRepresentedItem(self=Get FortEquippedItemDurabilityVM()))
   n6 = Branch(execute=@n4.else, Condition="float <= float"(A=Get InDurability(), B=-1) -> (ReturnValue=~))
   n7 = Branch(execute=@n6.else, Condition="float >= float"(A=Get InDurability(), B=Get CachedDurability() -> (CachedDurability:double)))
   n8 = Get First_DurabilityThreshold() -> (First_DurabilityThreshold:Scalable Float Structure)
   n9 = Get InDurability()
   n10 = Get Second_DurabilityThreshold() -> (Second_DurabilityThreshold:Scalable Float Structure)
   n11 = Get InDurability()
   n12 = Branch(
    execute=@n7.else
    Condition=AND Boolean(A="float < float"(A=@n9, B:Scalable Float Structure=@n8), B="float > float"(A=@n11, B:Scalable Float Structure=@n10))
   )
   > n13 = TryShowWarning() {is=title K2Node_CallFunction}
   n14 = Get InDurability()
   n15 = Get Second_DurabilityThreshold()
   n16 = Get Third_DurabilityThreshold() -> (Third_DurabilityThreshold:Scalable Float Structure)
   n17 = Get InDurability()
   n18 = Branch(execute=@n12.else, Condition=AND Boolean(A="float < float"(A=@n14, B=@n15), B="float > float"(A=@n17, B=@n16)))
   > n19 = TryShowWarning_2() {is=title K2Node_CallFunction}
   n20 = Branch(execute=@n18.else, Condition="float < float"(A=Get InDurability(), B=Get Third_DurabilityThreshold()))
   > n21 = TryShowWarning_3() {is=title K2Node_CallFunction}
  ]}
]
variables [
 CachedDurability {type=number}
 CachedWorldItem {type=FortWorldItem, default=null}
 First_DurabilityThreshold {
  type=ScalableFloat
  default {registryType=None, value=1, curve {curveTable=/Game/Athena/HUD/DurabilityUIData, rowName=Durability.FirstThreshold}}
 }
 HideNotificationTimerHandle {type=TimerHandle}
 ItemDestroyedHandle {type=FortAsyncAction_ListenForItemDestroyedByDurability, default=null}
 ItemDisplayName {type=string, default=Typhoon Blade}
 MessagingDuration {type=ScalableFloat, default {registryType=None, value=1, curve {curveTable=&DurabilityUIData, rowName=Durability.Duration}}}
 Second_DurabilityThreshold {
  type=ScalableFloat
  default {registryType=None, value=1, curve {curveTable=&DurabilityUIData, rowName=Durability.SecondThreshold}}
 }
 ShowedLowNotification {type=boolean}
 ShowedLowNotification2 {type=boolean}
 ShowedLowNotification3 {type=boolean}
 Third_DurabilityThreshold {
  type=ScalableFloat
  default {registryType=None, value=1, curve {curveTable=&DurabilityUIData, rowName=Durability.ThirdThreshold}}
 }
]
widgets {
 info {
  inheritedWidgetCount=0
  parentClass=&CommonUserWidget
  rootWidgetClass=/Game/Athena/HUD/Messaging/WBP_NotificationWithBackplate_Visuals.WBP_NotificationWithBackplate_Visuals_C
  widgetCount=1
 }
 widgets [WBP_NotificationWithBackplate {widgetClassPath=/Game/Athena/HUD/Messaging/WBP_NotificationWithBackplate_Visuals, bIsVariable=true, widget=WBP_NotificationWithBackplate}]
}
