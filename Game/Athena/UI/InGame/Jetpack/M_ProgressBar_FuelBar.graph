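# Translucent unlit UI material for a vertical jetpack fuel bar, driven by a Progress parameter.
# Progress blends the fill color from red to green, and comparisons against inverted UV height black out the area above the fill level within a box texture.
# Separate branches brighten color above a threshold just below that level for a top glow and build a vertical two-color background gradient.
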Material /Game/Athena/UI/InGame/Jetpack/M_ProgressBar_FuelBar
graph {
 properties {
  bUsedWithStaticMesh=false
  blendMode=BLEND_Translucent
  materialDomain=MD_UI
  pixelDepthOffsetMode=PDOM_Legacy
  shadingModel=MSM_Unlit
  usedShadingModels=Unlit
 }
 nodes [
  n1 = TextureCoordinate()
  n2 = ScalarParameter() {defaultValue=1, parameterName=Progress}
  n3 = Add(
   A=Multiply(A=OneMinus(None=@n2), B=Constant3Vector() {constant=<rgba 1 0 0 1>})
   B=Multiply(A=@n2, B=Constant3Vector() {constant=<rgba 0 1 0 1>})
  )
  If(
   A=OneMinus(None=MaterialFunctionCall(Float2=@n1) {materialFunction=&BreakOutFloat2Components}.G)
   B=Subtract(A=@n2) {constB=0.1, desc=Glow Thickness}
   "A > B"=Add(A=@n3) {constB=0.5, desc=Brightening for Glow at Top}
   "A == B"=@n3
   "A < B"=@n3
  )
  n4 = TextureCoordinate()
  LinearInterpolate(
   A=VectorParameter() {parameterName=BackgroundColorA, defaultValue=<rgba 0.1 0.1 0.1 1>}.RGB
   B=VectorParameter() {parameterName=BackgroundColorB, defaultValue=<rgba 0.4 0.4 0.4 1>}.RGB
   Alpha=OneMinus(None=MaterialFunctionCall(Float2=@n4) {materialFunction=&BreakOutFloat2Components}.G)
  )
  n5 = ComponentMask(None=TextureCoordinate()) {g=true}
  Multiply(
   A=TextureSample() {texture=&/TrivialBox}.RGB
   B=If(A=OneMinus(None=@n5), B=@n2, "A > B"=Constant(), "A == B"=@n3, "A < B"=@n3)
  )
 ]
}
