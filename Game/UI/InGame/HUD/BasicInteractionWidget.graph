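# HUD widget for interaction prompts, built on BasicInteractionWidgetBase.
# A size-limited vertical stack holds an uppercased rich-text description with an interaction-prompt decorator, a collapsed multi-interact display, and a bordered sub-description with an alert icon.
# On Construct it sets the description's wrap width, picking the shorter mobile length only when compact UI and touch input are both active.
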WidgetBlueprint /Game/UI/InGame/HUD/BasicInteractionWidget
graphs [
 EventGraph {nodes [
   Event Construct()
   > SetDescriptionWrapTextLenght()
  ]}
 SetDescriptionWrapTextLenght {nodes [
   SetDescriptionWrapTextLenght() {is=title K2Node_FunctionEntry}
   n1 = IsUsingCompactUI()
   n2 = IsUsingInputType!(InputType:ECommonInputType Enum=Touch, PlayerController:Player Controller Object Reference=GetOwningPlayer()) -> (ReturnValue:bool)
   > Set WrapTextAt!(
    execute
    WrapTextAt:float=SelectFloat(
     A=Get MobileDescriptionTextWrapLenght() -> (MobileDescriptionTextWrapLenght:double)
     B=Get DescriptionTextWrapLenght() -> (DescriptionTextWrapLenght:double)
     bPickA=AND Boolean(A=@n1, B=@n2)
    )
    self:Text Layout Widget Object Reference=Get RichText_Description() -> (RichText_Description:Common Rich Text Block Object Reference)
   ) -> (then, Output_Get:float)
  ]}
]
variables [
 DescriptionTextWrapLenght {type=number, default=400}
 MobileDescriptionTextWrapLenght {type=number, default=300}
]
widgets {
 info {inheritedWidgetCount=6, parentClass=/Script/FortniteUI.BasicInteractionWidgetBase, rootWidgetClass=&SizeBox, widgetCount=8}
 widgets [
  SizeBox_0 {
   widgetClassPath=&SizeBox
   widget=SizeBox_0
   properties {bOverride_MinDesiredWidth=true, maxDesiredWidth=300, minDesiredHeight=128, minDesiredWidth=50}
  }
  VerticalBox_0 {
   widgetClassPath=&VerticalBox
   parent=SizeBox_0
   widget=VerticalBox_0
   slot {class=SizeBoxSlot, properties {horizontalAlignment=HAlign_Left, verticalAlignment=VAlign_Center}}
  }
  RichText_Description {
   widgetClassPath=&CommonRichTextBlock
   bInherited=true
   bIsVariable=true
   parent=VerticalBox_0
   widget=RichText_Description
   properties {
    text="It's  text block of insanely big proportions.  I mean it's gigantic.  It's humongous.  It's the biggest piece of text you've ever seen in one place.  And better still, it doesn't need to be localized!"
    textStyleSet=/Game/Athena/HUD/TextStyles/AthenaInteractionStyles
    textTransformPolicy=ToUpper
    wrapTextAt=300
    decoratorClasses [&/RichTextDecorator_InteractionPrompt.RichTextDecorator_InteractionPrompt_C]
    defaultTextStyleOverride {
     overflowPolicy=Clip
     transformPolicy=None
     colorAndOpacity {colorUseRule=UseColor_Specified, specifiedColor=<rgba 1 1 1 1>}
     font {
      bForceMonospaced=false
      bMaterialIsStencil=false
      fontMaterial=null
      fontObject=&NotoSans
      letterSpacing=0
      monospacedWidth=1
      size=24
      skewAmount=0
      typefaceFontName=NotoSans-Regular
      outlineSettings {
       bApplyOutlineToDropShadows=false
       bMiteredCorners=false
       bSeparateFillAlpha=false
       outlineMaterial=null
       outlineSize=0
       outlineColor=<rgba 0 0 0 1>
      }
     }
     highlightColor {colorUseRule=UseColor_Specified, specifiedColor=<rgba 0 0 0 1>}
     highlightShape {
      bIsDynamicallyLoaded=false
      drawAs=Image
      imageType=NoImage
      mirroring=NoMirror
      resourceName=None
      resourceObject=null
      tiling=NoTile
      imageSize=<32 32>
      margin=<ltrb 0 0 0 0>
      outlineSettings {
       bUseBrushTransparency=false
       roundingType=HalfHeightRadius
       width=0
       color {colorUseRule=UseColor_Specified, specifiedColor=<rgba 0 0 0 0>}
       cornerRadii=<0 0 0 0>
      }
      tintColor {colorUseRule=UseColor_Specified, specifiedColor=<rgba 1 1 1 1>}
      uVRegion {bIsValid=false, max=<0 0>, min=<0 0>}
     }
     selectedBackgroundColor {colorUseRule=UseColor_Foreground, specifiedColor=<rgba 1 0 1 1>}
     shadowColorAndOpacity=<rgba 0 0 0 1>
     shadowOffset=<0 0>
     strikeBrush {
      bIsDynamicallyLoaded=false
      drawAs=Image
      imageType=NoImage
      mirroring=NoMirror
      resourceName=None
      resourceObject=null
      tiling=NoTile
      imageSize=<32 32>
      margin=<ltrb 0 0 0 0>
      outlineSettings {
       bUseBrushTransparency=false
       roundingType=HalfHeightRadius
       width=0
       color {colorUseRule=UseColor_Specified, specifiedColor=<rgba 0 0 0 0>}
       cornerRadii=<0 0 0 0>
      }
      tintColor {colorUseRule=UseColor_Specified, specifiedColor=<rgba 1 1 1 1>}
      uVRegion {bIsValid=false, max=<0 0>, min=<0 0>}
     }
     underlineBrush {
      bIsDynamicallyLoaded=false
      drawAs=Image
      imageType=NoImage
      mirroring=NoMirror
      resourceName=None
      resourceObject=null
      tiling=NoTile
      imageSize=<32 32>
      margin=<ltrb 0 0 0 0>
      outlineSettings {
       bUseBrushTransparency=false
       roundingType=HalfHeightRadius
       width=0
       color {colorUseRule=UseColor_Specified, specifiedColor=<rgba 0 0 0 0>}
       cornerRadii=<0 0 0 0>
      }
      tintColor {colorUseRule=UseColor_Specified, specifiedColor=<rgba 1 1 1 1>}
      uVRegion {bIsValid=false, max=<0 0>, min=<0 0>}
     }
    }
    margin=<ltrb 0 -2 0 0>
   }
   slot {class=VerticalBoxSlot}
  }
  MultiInteractDisplayWidget {
   widgetClassPath=/Game/UI/InGame/HUD/MultiInteract/MultiInteractDisplayWidget
   bInherited=true
   parent=VerticalBox_0
   widget=MultiInteractDisplayWidget
   properties {visibility=Collapsed}
   slot {class=VerticalBoxSlot, properties {horizontalAlignment=HAlign_Left, padding=<ltrb 5 5 0 5>}}
  }
  OuterBorder {
   widgetClassPath=&Border
   bInherited=true
   parent=VerticalBox_0
   widget=OuterBorder
   properties {visibility=SelfHitTestInvisible, brushColor=<rgba 0.0052 0.0052 0.0056 0.8>, padding=<ltrb 4 4 4 4>}
   slot {class=VerticalBoxSlot}
  }
  SubTextBox {
   widgetClassPath=&HorizontalBox
   bInherited=true
   parent=OuterBorder
   widget=SubTextBox
   properties {visibility=HitTestInvisible}
   slot {class=BorderSlot, properties {padding=<ltrb 4 4 4 4>}}
  }
  SubDescriptionAlertImage {
   widgetClassPath=&Image
   bInherited=true
   parent=SubTextBox
   widget=SubDescriptionAlertImage
   properties {
    visibility=HitTestInvisible
    brush {
     bIsDynamicallyLoaded=false
     drawAs=Image
     imageType=NoImage
     mirroring=NoMirror
     resourceName=None
     resourceObject=/Game/UI/Foundation/Textures/Icons/HUD/T-Icon-Alert2-32
     tiling=NoTile
     imageSize=<32 32>
     margin=<ltrb 0 0 0 0>
     outlineSettings {
      bUseBrushTransparency=false
      roundingType=HalfHeightRadius
      width=0
      color {colorUseRule=UseColor_Specified, specifiedColor=<rgba 0 0 0 0>}
      cornerRadii=<0 0 0 0>
     }
     tintColor {colorUseRule=UseColor_Specified, specifiedColor=<rgba 1 1 1 1>}
     uVRegion {bIsValid=false, max=<0 0>, min=<0 0>}
    }
   }
   slot {class=HorizontalBoxSlot, properties {verticalAlignment=VAlign_Center}}
  }
  TextSubDescription {
   widgetClassPath=&CommonRichTextBlock
   bInherited=true
   parent=SubTextBox
   widget=TextSubDescription
   properties {
    autoWrapText=true
    text="Sub Description <Rare>Sub</> Description Sub Description <Mythic>Sub Description</>"
    textStyleSet=&AthenaInteractionStyles
    wrapTextAt=256
    defaultTextStyleOverride {
     overflowPolicy=Clip
     transformPolicy=None
     colorAndOpacity {colorUseRule=UseColor_Specified, specifiedColor=<rgba 1 1 1 1>}
     font {
      bForceMonospaced=false
      bMaterialIsStencil=false
      fontMaterial=null
      fontObject=&NotoSans
      letterSpacing=0
      monospacedWidth=1
      size=24
      skewAmount=0
      typefaceFontName=NotoSans-Regular
      outlineSettings {
       bApplyOutlineToDropShadows=false
       bMiteredCorners=false
       bSeparateFillAlpha=false
       outlineMaterial=null
       outlineSize=0
       outlineColor=<rgba 0 0 0 1>
      }
     }
     highlightColor {colorUseRule=UseColor_Specified, specifiedColor=<rgba 0 0 0 1>}
     highlightShape {
      bIsDynamicallyLoaded=false
      drawAs=Image
      imageType=NoImage
      mirroring=NoMirror
      resourceName=None
      resourceObject=null
      tiling=NoTile
      imageSize=<32 32>
      margin=<ltrb 0 0 0 0>
      outlineSettings {
       bUseBrushTransparency=false
       roundingType=HalfHeightRadius
       width=0
       color {colorUseRule=UseColor_Specified, specifiedColor=<rgba 0 0 0 0>}
       cornerRadii=<0 0 0 0>
      }
      tintColor {colorUseRule=UseColor_Specified, specifiedColor=<rgba 1 1 1 1>}
      uVRegion {bIsValid=false, max=<0 0>, min=<0 0>}
     }
     selectedBackgroundColor {colorUseRule=UseColor_Foreground, specifiedColor=<rgba 1 0 1 1>}
     shadowColorAndOpacity=<rgba 0 0 0 1>
     shadowOffset=<0 0>
     strikeBrush {
      bIsDynamicallyLoaded=false
      drawAs=Image
      imageType=NoImage
      mirroring=NoMirror
      resourceName=None
      resourceObject=null
      tiling=NoTile
      imageSize=<32 32>
      margin=<ltrb 0 0 0 0>
      outlineSettings {
       bUseBrushTransparency=false
       roundingType=HalfHeightRadius
       width=0
       color {colorUseRule=UseColor_Specified, specifiedColor=<rgba 0 0 0 0>}
       cornerRadii=<0 0 0 0>
      }
      tintColor {colorUseRule=UseColor_Specified, specifiedColor=<rgba 1 1 1 1>}
      uVRegion {bIsValid=false, max=<0 0>, min=<0 0>}
     }
     underlineBrush {
      bIsDynamicallyLoaded=false
      drawAs=Image
      imageType=NoImage
      mirroring=NoMirror
      resourceName=None
      resourceObject=null
      tiling=NoTile
      imageSize=<32 32>
      margin=<ltrb 0 0 0 0>
      outlineSettings {
       bUseBrushTransparency=false
       roundingType=HalfHeightRadius
       width=0
       color {colorUseRule=UseColor_Specified, specifiedColor=<rgba 0 0 0 0>}
       cornerRadii=<0 0 0 0>
      }
      tintColor {colorUseRule=UseColor_Specified, specifiedColor=<rgba 1 1 1 1>}
      uVRegion {bIsValid=false, max=<0 0>, min=<0 0>}
     }
    }
   }
   slot {class=HorizontalBoxSlot, properties {verticalAlignment=VAlign_Center}}
  }
 ]
}
