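# Widget-less base HUD quickbar Blueprint on FortQuickbarBase that keeps its slot widgets in sync with inventory.
# Construct binds inventory and keybind events, plus death, revive and spawn outside Battle Royale; Destruct removes most bindings.
# Content changes refresh only changed slots on its own bar, skipping dead Battle Royale pawns in later phases and non-Campaign subgames otherwise.
# Optional event-router listeners track minigame volume entry and exit, updating slot capacity and visibility.
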WidgetBlueprint /Game/UI/InGame/HUD/QuickbarBase
graphs [
 EventGraph {nodes [
   n1 = OnClientEnterVolume!() -> (OutputDelegate, then, EnteredMinigame:Fort Minigame Object Reference) {is=title K2Node_CustomEvent}
   > Is Valid(InputObject=@n1.EnteredMinigame)
   > Bind Event to Minigame Started Delegate(self:Fort Minigame Object Reference=@n1.EnteredMinigame, Delegate=Create Event()) {is=title K2Node_AddDelegate}
   > Assign Minigame Ended Delegate!(execute:Exec, self:Fort Minigame Object Reference=@n1.EnteredMinigame, Delegate:"Delegate (by ref)"=Create Event()) -> (then:Exec) {is=title K2Node_AssignDelegate}
   > UpdateInventoryCapacity()
   > UpdateSlotVisibility()
   n2 = OnClientExitVolume!() -> (OutputDelegate, then, Volume:Actor Object Reference) {is=title K2Node_CustomEvent}
   > Is Valid(InputObject=@n2.Volume)
   n3 = GetMinigame!(ContextActor:Actor Object Reference=@n2.Volume) -> (ReturnValue:Fort Minigame Object Reference)
   > Is Valid(InputObject=@n3)
   > Unbind Event from Minigame Started Delegate(self:Fort Minigame Object Reference=@n3, Delegate=Create Event()) {is=title K2Node_RemoveDelegate}
   > Unbind Event from Minigame Ended Delegate(self:Fort Minigame Object Reference=@n3, Delegate=Create Event()) {is=title K2Node_RemoveDelegate}
   > Delay(Duration=0)
   > UpdateInventoryCapacity()
   > UpdateSlotVisibility()
   n4 = OnQuickbarForceFullUpdate!() -> (OutputDelegate, then, QuickbarIndex:EFortQuickBars Enum=Primary) {is=title K2Node_CustomEvent}
   > Branch(Condition="Equal (Enum)"(A:EFortQuickBars Enum=@n4.QuickbarIndex, B:EFortQuickBars Enum=Get MyQuickbarIndex() -> (MyQuickbarIndex:EFortQuickBars Enum=Primary)))
   > RefreshQuickbar!(execute, self, ForceItemsChanged:bool=true) -> (then)
   n5 = OnLocalPlayerDied!() -> (OutputDelegate, then, DeathReport:"Fort Player Death Report Structure (by ref)") {is=title K2Node_CustomEvent}
   RefreshQuickbar(execute=[@n5.then, @n7.then, @n8.then], ForceItemsChanged=false)
   n6 = OnQuickbarContentsChanged!() -> (OutputDelegate, then, QuickbarIndex:EFortQuickBars Enum=Primary, ChangedSlots:Array of Integers) {is=title K2Node_CustomEvent}
   > HandleQuickbarContentChanged!(execute, self, QuickbarType:EFortQuickBars Enum=@n6.QuickbarIndex, ChangedSlots:Array of Integers=@n6.ChangedSlots) -> (then)
   n7 = OnLocalPlayerRevived() {is=title K2Node_CustomEvent}
   n8 = OnLocalPlayerSpawned() {is=title K2Node_CustomEvent}
   Event Destruct()
   > n9 = Sequence!(execute) -> (then_0, then_1, then_2:Exec, then_3:Exec)
   n10 = Get FortInventoryContext!() -> (ReturnValue:Fort Inventory Context Object Reference) {is=title K2Node_GetBlueprintContextEx}
   > Is Valid(InputObject=@n10)
   > Unbind Event from On Quickbar Contents Changed(self:Fort Inventory Context Object Reference=@n10, Delegate=Create Event()) {is=title K2Node_RemoveDelegate}
   > Unbind Event from On World Item List Changed(self:Fort Inventory Context Object Reference=@n10, Delegate=Create Event()) {is=title K2Node_RemoveDelegate}
   > Unbind Event from On Quickbar Force Full Update(self:Fort Inventory Context Object Reference=@n10, Delegate=Create Event()) {is=title K2Node_RemoveDelegate}
   n11 = Get PlayerEnteredHandle() -> (PlayerEnteredHandle:Async Action Start Listening To Event Object Reference)
   n12 = Is Valid(exec=@n9.then_3, InputObject=@n11)
   n13 = Get ExitedEventHandle() -> (ExitedEventHandle:Async Action Start Listening To Event Object Reference)
   Is Valid(exec=[@n14.then, @n12.Is Not Valid], InputObject=@n13)
   > StopListeningToEvent(self:Async Action Start Listening To Event Object Reference=@n13)
   n14 = StopListeningToEvent(execute=@n12.Is Valid, self=@n11)
   n15 = Get FortGlobalUIContext!() -> (ReturnValue:Fort Global UIContext Object Reference) {is=title K2Node_GetBlueprintContextEx}
   Is Valid(exec=@n9.then_1, InputObject=@n15)
   > Unbind Event from On Keybinds Changed(self:Fort Global UIContext Object Reference=@n15, Delegate=Create Event()) {is=title K2Node_RemoveDelegate}
   n16 = Get FortHUDContext!() -> (ReturnValue:Fort HUDContext Object Reference) {is=title K2Node_GetBlueprintContextEx}
   Is Valid(exec=@n9.then_2, InputObject=@n16)
   > Unbind Event from On Local Player Died(self:Fort HUDContext Object Reference=@n16, Delegate=Create Event()) {is=title K2Node_RemoveDelegate}
   > Unbind Event from On Local Player Revived(self:Fort HUDContext Object Reference=@n16, Delegate=Create Event()) {is=title K2Node_RemoveDelegate}
   > Unbind Event from On Local Player Spawned(self:Fort HUDContext Object Reference=@n16, Delegate=Create Event()) {is=title K2Node_RemoveDelegate}
   Event Construct()
   n17 = GetGameState!() -> (ReturnValue:Game State Base Object Reference)
   > Set IsBattleRoyale(IsBattleRoyale:bool=ClassIsChildOf(TestClass=Get Class(Object=@n17) -> (ReturnValue:Game State Base Class Reference), ParentClass=/Script/FortniteGame.FortGameStateAthena)) -> (Output_Get:bool)
   > n18 = Sequence!(execute) -> (then_0, then_1, then_2, then_3, then_4:Exec, then_5:Exec, then_6:Exec, then_7:Exec, then_8:Exec)
   n19 = Get FortInventoryContext()
   Assign On Quickbar Force Full Update!(execute:Exec=@n18.then_4, self:Fort Inventory Context Object Reference=@n19, Delegate:"Delegate (by ref)"=@n4) -> (then:Exec) {is=title K2Node_AssignDelegate}
   Branch(execute=@n18.then_7, Condition=Get bBindMinigameStartAndEndEvents() -> (bBindMinigameStartAndEndEvents:bool))
   > n20 = StartListeningToEventBackwardCompatible!(
    execute:Exec
    Channel:Event Message Tag Structure=(TagName="EventMessage.Component.Playspace.PlayerEnteredSpatialActor.LegacyMinigame")
    Target:Gameplay Event Router Component Object Reference=GetGlobalEventRouter()
    EventType:Script Struct Object Reference=/Script/FortniteGame.PlayerEnterSpatialActorContextLegacyMinigame
    Context:Object Reference
    EventBubblingRule:EEventBubblingRule Enum=Down
   ) -> (
    then:Exec
    AsyncTaskProxy:Async Action Start Listening To Event Object Reference
    OnEventReceived:Exec
    Payload:Player Enter Spatial Actor Context Legacy Minigame Structure
   ) {is=title K2Node_StartListeningToEventAsyncAction}
   > Set PlayerEnteredHandle(PlayerEnteredHandle:Async Action Start Listening To Event Object Reference=@n20) -> (Output_Get:Async Action Start Listening To Event Object Reference)
   > n21 = StartListeningToEventBackwardCompatible(
    Channel=(TagName="EventMessage.Component.Playspace.PlayerExitedSpatialActor.LegacyMinigame")
    Target=GetGlobalEventRouter()
    EventType=/Script/FortniteGame.PlayerExitSpatialActorContextLegacyMinigame
    EventBubblingRule=Down
   ) -> (Payload:Player Exit Spatial Actor Context Legacy Minigame Structure)
   > Set ExitedEventHandle(ExitedEventHandle:Async Action Start Listening To Event Object Reference=@n21) -> (Output_Get:Async Action Start Listening To Event Object Reference)
   n22 = GetOwningPlayer()
   > Is Valid(InputObject=@n22)
   > n23 = Get!(execute:Exec, self:Controller Object Reference=@n22) -> (then:Exec, else:Exec, PlayerState:Player State Object Reference)
   > OnClientEnterVolume!(execute, self, EnteredMinigame:Fort Minigame Object Reference=GetMinigame(ContextActor=@n23)) -> (then) {is=title K2Node_CallFunction}
   OnClientExitVolume!(
    execute=@n21.OnEventReceived
    self
    Volume:Actor Object Reference=Break Player Exit Spatial Actor Context Legacy Minigame!(PlayerExitSpatialActorContextLegacyMinigame:"Player Exit Spatial Actor Context Legacy Minigame Structure (by ref)"=@n21.Payload) -> (SpatialGameplayInterfaceActor:Actor Object Reference, ExitingPlayerState:Player State Object Reference) {is=title K2Node_BreakStruct}
   ) -> (then) {is=title K2Node_CallFunction}
   OnClientEnterVolume(
    execute=@n20.OnEventReceived
    EnteredMinigame=GetMinigame(ContextActor=Break Player Enter Spatial Actor Context Legacy Minigame!(PlayerEnterSpatialActorContextLegacyMinigame:"Player Enter Spatial Actor Context Legacy Minigame Structure (by ref)"=@n20.Payload) -> (SpatialGameplayInterfaceActor:Actor Object Reference, EnteringPlayerState:Player State Object Reference) {is=title K2Node_BreakStruct})
   )
   Assign On World Item List Changed!(
    execute:Exec=@n18.then_1
    self:Fort Inventory Context Object Reference=@n19
    Delegate:"Delegate (by ref)"=OnWorldItemListChanged!() -> (OutputDelegate, then, ItemsAdded:Array of Fort World Item Object References, ItemsRemoved:Array of Fort World Item Object References) {is=title K2Node_CustomEvent}
   ) -> (then:Exec) {is=title K2Node_AssignDelegate}
   n24 = Branch(execute=@n18.then_6, Condition=Get IsBattleRoyale() -> (IsBattleRoyale:bool))
   n25 = Get FortHUDContext()
   Bind Event to On Local Player Died(execute=@n24.else, self:Fort HUDContext Object Reference=@n25, Delegate=@n5) {is=title K2Node_AddDelegate}
   > Assign On Local Player Revived!(execute:Exec, self:Fort HUDContext Object Reference=@n25, Delegate:"Delegate (by ref)"=@n7) -> (then:Exec) {is=title K2Node_AssignDelegate}
   > Assign On Local Player Spawned!(execute:Exec, self:Fort HUDContext Object Reference=@n25, Delegate:"Delegate (by ref)"=@n8) -> (then:Exec) {is=title K2Node_AssignDelegate}
   Bind Event to On Keybinds Changed(execute=@n18.then_5, self:Fort Global UIContext Object Reference=Get FortGlobalUIContext(), Delegate=Create Event()) {is=title K2Node_AddDelegate}
   > HandleKeybindsChanged()
   Assign On Quickbar Contents Changed!(execute:Exec=@n18.then_0, self:Fort Inventory Context Object Reference=@n19, Delegate:"Delegate (by ref)"=@n6) -> (then:Exec) {is=title K2Node_AssignDelegate}
  ]}
 HandleKeybindsChanged {nodes [
   HandleKeybindsChanged() {is=title K2Node_FunctionEntry}
   > n1 = For Each Loop(Array:Array of Fort Quick Bar Slot Base Object References=Get QuickbarSlots() -> (QuickbarSlots:Array of Fort Quick Bar Slot Base Object References)) -> (Array Element:Fort Quick Bar Slot Base Object Reference)
   > RefreshItem(self:Fort Quick Bar Slot Base Object Reference=@n1)
  ]}
 HandleQuickbarContentChanged {nodes [
   HandleQuickbarContentChanged!() -> (then, QuickbarType:EFortQuickBars Enum=Primary, ChangedSlots:Array of Integers) {is=title K2Node_FunctionEntry}
   > n1 = Branch(Condition=Get IsBattleRoyale())
   n2 = Branch(
    execute=@n1.else
    Condition="Equal (Enum)"(A:ESubGame Enum=GetCurrentSubGame!() -> (ReturnValue:ESubGame Enum=Campaign), B:ESubGame Enum=Campaign)
   )
   Branch(
    execute=[@n2.then, @n8.else, @n7.CastFailed, @n5.Is Not Valid, @n6.Aircraft, @n6.Count, @n6.None, @n6.Setup, @n6.Warmup]
    Condition="Equal (Enum)"(
     A:EFortQuickBars Enum=Get MyQuickbarIndex() -> (MyQuickbarIndex=Primary)
     B:EFortQuickBars Enum=Get QuickbarType() -> (QuickbarType:EFortQuickBars Enum=Primary)
    )
   )
   > n3 = For Each Loop(Array:Array of Integers=Get ChangedSlots() -> (ChangedSlots:Array of Integers)) -> (Array Element:int)
   > Branch(Condition=Is Valid Index!(TargetArray:Array of Fort Quick Bar Slot Base Object References=Get QuickbarSlots(), IndexToTest:int=@n3) -> (ReturnValue:bool))
   > RefreshItem(self="Get (a copy)"(Array:Array of Fort Quick Bar Slot Base Object References=Get QuickbarSlots(), Dimension 1=@n3) -> (Output:Fort Quick Bar Slot Base Object Reference))
   n4 = GetBattleRoyaleGamePhaseLogicComponent!() -> (ReturnValue:Fort Game State Component Battle Royale Game Phase Logic Object Reference)
   n5 = Is Valid(InputObject=@n4)
   > n6 = Switch on EAthenaGamePhase!(
    execute
    Selection:EAthenaGamePhase Enum=GetGamePhase!(self:Fort Game State Component Battle Royale Game Phase Logic Object Reference=@n4) -> (ReturnValue:EAthenaGamePhase Enum=None)
   ) -> (None:Exec, Setup:Exec, Warmup:Exec, Aircraft:Exec, SafeZones:Exec, EndGame:Exec, Count:Exec) {is=title K2Node_SwitchEnum}
   n7 = Cast To FortPlayerPawnAthena(execute=[@n6.EndGame, @n6.SafeZones], Object=GetOwningPlayerPawn!(self:User Widget Object Reference) -> (ReturnValue:Pawn Object Reference))
   > n8 = Branch(Condition=IsDead!(self:Fort Pawn Object Reference=@n7) -> (ReturnValue:bool))
  ]}
 HandleQuickbarSlotFocusChanged {nodes [
   n1 = HandleQuickbarSlotFocusChanged!() -> (then, QuickbarIndex:EFortQuickBars Enum=Primary, Slot:int) {is=title K2Node_FunctionEntry}
   > n2 = Branch(Condition="Equal (Enum)"(A=@n1, B=Get MyQuickbarIndex() -> (MyQuickbarIndex=Primary)))
   > Maximize(self:Fort Quickbar Base Object Reference)
   Minimize(execute=@n2.else, self:Fort Quickbar Base Object Reference)
  ]}
 RefreshQuickbar {nodes [
   RefreshQuickbar!() -> (then, ForceItemsChanged:bool=~) {is=title K2Node_FunctionEntry}
   > n1 = For Each Loop(Array:Array of Fort Quick Bar Slot Base Object References=Get QuickbarSlots()) -> (Array Element:Fort Quick Bar Slot Base Object Reference)
   > n2 = Branch(Condition=Get ForceItemsChanged() -> (ForceItemsChanged:bool))
   > BroadcastOnItemChanged!(
    execute
    self:Fort Item Object Reference=Get Item!(self:Fort Quick Bar Slot Base Object Reference=@n1) -> (Item:Fort Item Object Reference)
    bItemChanged:bool=false
    bAmmoChanged:bool=false
    bIngredientsChanged:bool=false
    bLogBroadcasts:bool=false
   ) -> (then)
   RefreshItem(execute=@n2.else, self=@n1)
  ]}
 ShouldRefreshQuickbarItem {nodes [
   n1 = ShouldRefreshQuickbarItem!() -> (then, RequestedSlot:int, CurrentSlot:int) {is=title K2Node_FunctionEntry}
   n2 = GetQuickbarFocus!(self:Fort Inventory Context Object Reference=Get FortInventoryContext()) -> (OutQuickBar:EFortQuickBars Enum=Primary, OutSlot:int=0, OutSecondarySlot:int=0, OutPreviousFocusedSlot:int=0)
   > Return Node!(
    execute
    Result:bool=OR Boolean(
     A="Equal (Integer)"(A=@n2.OutPreviousFocusedSlot, B=@n1.CurrentSlot) {is=title K2Node_CallFunction}
     B="Equal (Integer)"(A=@n1, B=@n1.CurrentSlot)
    )
   )
  ]}
 StopUsingAllSlots {nodes [
   StopUsingAllSlots() {is=title K2Node_FunctionEntry}
   > n1 = For Each Loop(Array=Get QuickbarSlots())
   > n2 = Cast To BacchusQuickbarSlot!(execute, Object=@n1) -> (then, CastFailed, AsBacchus Quickbar Slot:Bacchus Quickbar Slot Object Reference) {is=title K2Node_DynamicCast}
   > Cancel Interaction(self:Bacchus Quickbar Slot Object Reference=@n2)
  ]}
 UpdateCurrentFocusedQuickbar {nodes [
   UpdateCurrentFocusedQuickbar() {is=title K2Node_FunctionEntry}
   > n1 = Set PrevFocusedQuickbarIndex(PrevFocusedQuickbarIndex:EFortQuickBars Enum=Get FocusedQuickbarIndex() -> (FocusedQuickbarIndex:EFortQuickBars Enum=Primary)) -> (Output_Get:EFortQuickBars Enum=Primary)
   > n2 = Set FocusedQuickbarIndex(FocusedQuickbarIndex:EFortQuickBars Enum=GetQuickbarFocus(self=Get FortInventoryContext()) -> (OutQuickBar=Primary, OutSlot=0, OutSecondarySlot=0, OutPreviousFocusedSlot=0)) -> (Output_Get:EFortQuickBars Enum=Primary)
   > Return Node!(execute, PreviousFocusedQuickbar:EFortQuickBars Enum=@n1, CurrentFocusedQuickbar:EFortQuickBars Enum=@n2)
  ]}
 UpdateInventoryCapacity {nodes [
   UpdateInventoryCapacity() {is=title K2Node_FunctionEntry}
   > n1 = For Each Loop(Array=Get QuickbarSlots())
   > Is Valid(InputObject=@n1)
   > RefreshInventoryCapacity(self:Fort Quick Bar Slot Base Object Reference=@n1)
  ]}
 UpdateSlotVisibility {nodes [
   UpdateSlotVisibility() {is=title K2Node_FunctionEntry}
   > n1 = For Each Loop(Array=Get QuickbarSlots())
   > Is Valid(InputObject=@n1)
   > RefreshItem(self=@n1)
  ]}
]
variables [
 bBindMinigameStartAndEndEvents {type=boolean}
 ExitedEventHandle {type=AsyncAction_StartListeningToEvent, default=null}
 IsBattleRoyale {type=boolean}
 IsMaximized {type=boolean, default=true}
 PlayerEnteredHandle {type=AsyncAction_StartListeningToEvent, default=null}
]
widgets {info {inheritedWidgetCount=0, parentClass=/Script/FortniteUI.FortQuickbarBase, rootWidgetClass=null, widgetCount=0}, widgets []}
