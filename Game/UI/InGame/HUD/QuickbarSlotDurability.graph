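# HUD widget for a quickbar slot that draws the item's durability as a material progress bar.
# On item change it feeds world items to the durability viewmodel and collapses for non-world items, items without durability, or infinite-durability owners.
# Helper functions pick visibility and set two material color parameters from high, mid and low thresholds.
# OnDurabilityChange pulses the material's IsDamaged parameter by animation, skipping negative, full or unchanged values.
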WidgetBlueprint /Game/UI/InGame/HUD/QuickbarSlotDurability
animations [AnimDurabilityUsed {
  binding_count=1
  length_seconds=0.5
  object=&self:AnimDurabilityUsed
  sequence {bindings [Image_ProgressBar {
     id=#1
     possessed_class=FortLazyImage
     tracks [{
       type=MovieSceneWidgetMaterialTrack
       sections [{
         type=MovieSceneParameterSection
         end_frame=10
         start_frame=0
         channels [IsDamaged {type=MovieSceneScriptingFloatChannel, keys [[0,0],[5,0.3],[10,0]]}]
        }]
      }]
    }]}
 }]
graphs [
 DurabilityToColorA {nodes [
   n1 = DurabilityToColorA!() -> (then, InDurability:double=~) {is=title K2Node_FunctionEntry}
   > n2 = GetDynamicMaterial(self=Get Image_ProgressBar() -> (Image_ProgressBar:Fort Lazy Image Object Reference))
   > Set ImageProgressBarMaterial(ImageProgressBarMaterial:Material Instance Dynamic Object Reference=@n2) -> (Output_Get:Material Instance Dynamic Object Reference)
   > n3 = Branch(Condition="float > float"(A=@n1, B=Get LowDurabilityThreshold() -> (LowDurabilityThreshold:double)))
   n4 = Get ImageProgressBarMaterial() -> (ImageProgressBarMaterial:Material Instance Dynamic Object Reference)
   > SetVectorParameterValue(self=@n4, ParameterName=ColorA, Value=Get ColorAHighDurability() -> (ColorAHighDurability:Linear Color Structure))
   n5 = Branch(execute=@n3.else, Condition="float > float"(A=@n1, B=Get VeryLowDurabilityThreshold() -> (VeryLowDurabilityThreshold:double)))
   > SetVectorParameterValue(self=@n4, ParameterName=ColorA, Value=Get ColorAMidDurability() -> (ColorAMidDurability:Linear Color Structure))
   SetVectorParameterValue(execute=@n5.else, self=@n4, ParameterName=ColorA, Value=Get ColorALowDurability() -> (ColorALowDurability:Linear Color Structure))
  ]}
 DurabilityToColorB {nodes [
   n1 = DurabilityToColorB!() -> (then, InDurability:double=~) {is=title K2Node_FunctionEntry}
   > n2 = GetDynamicMaterial(self=Get Image_ProgressBar())
   > Set ImageProgressMaterialLocal(ImageProgressMaterialLocal:Material Instance Dynamic Object Reference=@n2) -> (Output_Get:Material Instance Dynamic Object Reference)
   > n3 = Branch(Condition="float > float"(A=@n1, B=Get LowDurabilityThreshold()))
   n4 = Get ImageProgressMaterialLocal() -> (ImageProgressMaterialLocal:Material Instance Dynamic Object Reference)
   > SetVectorParameterValue(self=@n4, ParameterName=ColorB, Value=Get ColorBHighDurability() -> (ColorBHighDurability:Linear Color Structure))
   n5 = Branch(execute=@n3.else, Condition="float > float"(A=@n1, B=Get VeryLowDurabilityThreshold()))
   > SetVectorParameterValue(self=@n4, ParameterName=ColorB, Value=Get ColorBMidDurability() -> (ColorBMidDurability:Linear Color Structure))
   SetVectorParameterValue(execute=@n5.else, self=@n4, ParameterName=ColorB, Value=Get ColorBLowDurability() -> (ColorBLowDurability:Linear Color Structure))
  ]}
 DurabilityToVisibility {nodes [
   n1 = DurabilityToVisibility!() -> (then, InDurability:double=~, bOwnerHasInfiniteDurability:bool=~) {is=title K2Node_FunctionEntry}
   n2 = "float < float"(A=@n1)
   > Return Node!(
    execute
    OutVisibility:ESlateVisibility Enum=Select(Option 0=HitTestInvisible, Option 1=Collapsed, Index=OR Boolean(A=@n2, B=@n1.bOwnerHasInfiniteDurability))
   )
  ]}
 EventGraph {nodes [
   n1 = Event OnPostItemToRepresentChanged!() -> (OutputDelegate, then, ItemToRepresent:Fort Item Object Reference) {is=title K2Node_Event}
   > n2 = Cast To FortWorldItem!(execute, Object=@n1.ItemToRepresent) -> (then, CastFailed, AsFort World Item:Fort World Item Object Reference) {is=title K2Node_DynamicCast}
   n3 = Get FortItemDurabilityVM() -> (FortItemDurabilityVM:Item Durability Viewmodel Object Reference)
   > Is Valid(InputObject=@n3)
   > SetWorldItem!(execute, self:Item Durability Viewmodel Object Reference=@n3, InWorldItem:Fort World Item Object Reference=@n2, bCheckInitialized:bool=false) -> (then)
   n4 = Get FortItemDurabilityVM()
   n5 = ItemHasDurability!(self:Item Durability Viewmodel Object Reference=@n4) -> (ReturnValue:bool)
   n6 = NOT Boolean(A=GetOwnerHasInfiniteDurability!(self:Item Durability Viewmodel Object Reference=@n4) -> (ReturnValue:bool))
   > SetVisibility(InVisibility=Select(Option 0=Collapsed, Option 1=HitTestInvisible, Index=AND Boolean(A=@n5, B=@n6)))
   SetVisibility(execute=@n2.CastFailed, InVisibility=Collapsed)
  ]}
 OnDurabilityChange {nodes [
   n1 = OnDurabilityChange!() -> (then, Durability:double=~) {is=title K2Node_FunctionEntry}
   n2 = Get TempDurability() -> (TempDurability:double)
   > n3 = Branch(Condition=OR Boolean!(A="float < float"(A=@n1), B="Equal (Float)"(A=@n1, B=1), C:bool="Equal (Float)"(A=@n1, B=@n2)) -> (ReturnValue))
   > StopAnimation(InAnimation=Get AnimDurabilityUsed() -> (AnimDurabilityUsed:Widget Animation Object Reference))
   n4 = Get AnimDurabilityUsed()
   n5 = Branch(execute=@n3.else, Condition=IsAnimationPlaying(InAnimation=@n4))
   PlayAnimation(execute=@n5.else, InAnimation=@n4, bRestoreState=true)
   > Set TempDurability(TempDurability:double=@n1) -> (Output_Get:double)
  ]}
]
variables [
 ColorAHighDurability {type=LinearColor, default=<rgba 0 0.3176 0.0314 1>}
 ColorALowDurability {type=LinearColor, default=<rgba 1 0.0314 0.0314 1>}
 ColorAMidDurability {type=LinearColor, default=<rgba 1 0.2275 0.0314 1>}
 ColorBHighDurability {type=LinearColor, default=<rgba 0.1333 0.6039 0.0314 1>}
 ColorBLowDurability {type=LinearColor, default=<rgba 1 0.1255 0.0314 1>}
 ColorBMidDurability {type=LinearColor, default=<rgba 1 0.6039 0 1>}
 LowDurabilityThreshold {type=number, default=0.5}
 TempDurability {type=number}
 VeryLowDurabilityThreshold {type=number, default=0.25}
]
widgets {
 info {inheritedWidgetCount=0, parentClass=/Script/FortniteUI.AthenaItemElementWidgetBase, rootWidgetClass=&Overlay, widgetCount=7}
 widgets [
  Overlay_20 {widgetClassPath=&Overlay, widget=Overlay_20, properties {visibility=HitTestInvisible}}
  SizeBox {
   widgetClassPath=&SizeBox
   parent=Overlay_20
   widget=SizeBox
   properties {bOverride_HeightOverride=true, bOverride_WidthOverride=true, heightOverride=92, visibility=HitTestInvisible, widthOverride=92}
   slot {class=OverlaySlot, properties {horizontalAlignment=HAlign_Fill, verticalAlignment=VAlign_Fill}}
  }
  VerticalBox_132 {widgetClassPath=&VerticalBox, parent=SizeBox, widget=VerticalBox_132, slot {class=SizeBoxSlot}}
  Overlay_131 {
   widgetClassPath=&Overlay
   parent=VerticalBox_132
   widget=Overlay_131
   slot {class=VerticalBoxSlot, properties {padding=<ltrb 7 7 7 7>, size {sizeRule=Fill, value=1}}}
  }
  Image_Outline {
   widgetClassPath=&Image
   parent=Overlay_131
   widget=Image_Outline
   properties {colorAndOpacity=<rgba 1 1 1 0.7>}
   slot {class=OverlaySlot, properties {horizontalAlignment=HAlign_Fill, verticalAlignment=VAlign_Fill, padding=<ltrb -1 -1 -1 -1>}}
  }
  Image_ProgressBar {
   widgetClassPath=&FortLazyImage
   bIsVariable=true
   parent=Overlay_131
   widget=Image_ProgressBar
   properties {brush {
     bIsDynamicallyLoaded=false
     drawAs=Image
     imageType=FullColor
     mirroring=NoMirror
     resourceName=None
     resourceObject=/Game/Athena/HUD/Team/MI_ProgressBar_Durability_Inst
     tiling=NoTile
     imageSize=<32 32>
     margin=<ltrb 0 0 0 0>
     outlineSettings {
      bUseBrushTransparency=false
      roundingType=HalfHeightRadius
      width=0
      color {colorUseRule=UseColor_Specified, specifiedColor=<rgba 0 0 0 0>}
      cornerRadii=<0 0 0 0>
     }
     tintColor {colorUseRule=UseColor_Specified, specifiedColor=<rgba 1 1 1 1>}
     uVRegion {bIsValid=false, max=<0 0>, min=<0 0>}
    }}
   slot {class=OverlaySlot, properties {horizontalAlignment=HAlign_Fill, verticalAlignment=VAlign_Fill}}
  }
  Spacer_43 {
   widgetClassPath=&Spacer
   parent=VerticalBox_132
   widget=Spacer_43
   slot {class=VerticalBoxSlot, properties {size {sizeRule=Fill, value=10}}}
  }
 ]
}
