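# MF_UI_Ring is a material function in the HUD textures folder that outputs a procedural ring mask.
# It builds two exponential radial gradients, one at OuterRadius and one at OuterRadius minus RingSize.
# Both pass through SmoothStep between Edge and Edge plus EdgeSoftness, and the inner disc is subtracted from the outer.
# Each input falls back to a preview scalar parameter as its default.
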MaterialFunction /Game/UI/InGame/HUD/Textures/MF_UI_Ring
graph {nodes [
  n1 = FunctionInput(Preview=ScalarParameter() {defaultValue=0.5, parameterName=OuterRadius, sliderMax=2}) {bUsePreviewValueAsDefault=true, inputName=OuterRadius, inputType=FunctionInput_Scalar}
  n2 = MaterialFunctionCall(Radius=@n1) {materialFunction=&RadialGradientExponential}
  n3 = FunctionInput(Preview=ScalarParameter() {parameterName=Edge}) {bUsePreviewValueAsDefault=true, inputName=Edge, inputType=FunctionInput_Scalar}
  n4 = Add(
   A=@n3
   B=FunctionInput(Preview=ScalarParameter() {defaultValue=0.002, parameterName=EdgeSoftness}) {bUsePreviewValueAsDefault=true, inputName=EdgeSoftness, inputType=FunctionInput_Scalar}
  )
  n5 = ScalarParameter() {defaultValue=0.04, parameterName=RingSize}
  n6 = MaterialFunctionCall(Radius=Subtract(A=@n1, B=FunctionInput(Preview=@n5) {bUsePreviewValueAsDefault=true, inputName=RingSize, inputType=FunctionInput_Scalar})) {materialFunction=&RadialGradientExponential}
  FunctionOutput(None=Subtract(
    A=MaterialFunctionCall(Alpha=@n2.RadialGradientExponential, Min=@n3, Max=@n4) {materialFunction=&SmoothStep}.Result
    B=MaterialFunctionCall(Alpha=@n6.RadialGradientExponential, Min=@n3, Max=@n4) {materialFunction=&SmoothStep}.Result
   ))
 ]}
